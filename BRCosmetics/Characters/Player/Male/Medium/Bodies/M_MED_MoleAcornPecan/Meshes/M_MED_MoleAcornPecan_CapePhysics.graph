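# Physics asset for the MoleAcornPecan male medium body's cape, defining the bodies and joints for its secondary motion.
# Four cape strands each chain six capsules from a master bone, with narrow twist and swing limits that loosen on one axis lower down.
# Free-rotation cross-links join neighboring strands at the fifth segment, keeping the cape together.
# Torso, arm and leg capsules plus a large ground box provide collision shapes.
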PhysicsAsset /BRCosmetics/Characters/Player/Male/Medium/Bodies/M_MED_MoleAcornPecan/Meshes/M_MED_MoleAcornPecan_CapePhysics
bodies {
 calf_l [{
   shapeType=Capsule
   length=41.4223
   radius=14.9678
   shapeName=calf_l_capsule
   center=<-25.405 1.365 2.7764>
   rotation=<rot 86.9523 179.9998 179.9999>
  }]
 calf_r [{
   shapeType=Capsule
   length=41.4323
   radius=15.1007
   shapeName=calf_r_capsule
   center=<25.3928 -0.5698 -2.6776>
   rotation=<rot 89.3024 179.9994 179.9995>
  }]
 clavicle_l [{shapeType=Capsule, length=11.136, radius=10.8288, shapeName=clavicle_l_capsule, center=<5.1544 0 0>, rotation=<rot 90 0 0>}]
 clavicle_r [{shapeType=Capsule, length=11.0408, radius=10.9595, shapeName=clavicle_r_capsule, center=<-4.3594 0 0>, rotation=<rot 90 0 0>}]
 dyn_cape_ground_collision [{shapeType=Box, extentX=318.8156, extentY=318.8156, extentZ=283.5656, shapeName=dyn_skirt_ground_collision_box, center=<0 0 -140>}]
 dyn_cape_l_01 [{shapeType=Capsule, length=11.663, radius=14.8873, shapeName=dyn_cape_l_01_capsule, center=<16.3712 0 16.7>, rotation=<rot 90 0 0>}]
 dyn_cape_l_02 [{shapeType=Capsule, length=7.8705, radius=14.8873, shapeName=dyn_cape_l_02_capsule, center=<16.3712 0 16.3>, rotation=<rot 90 0 0>}]
 dyn_cape_l_03 [{shapeType=Capsule, length=19.3701, radius=15.4132, shapeName=dyn_cape_l_03_capsule, center=<10.1586 0 13.3>, rotation=<rot 90 0 0>}]
 dyn_cape_l_04 [{shapeType=Capsule, length=18.8586, radius=15.4132, shapeName=dyn_cape_l_04_capsule, center=<7.116 0 10.3>, rotation=<rot 90 0 0>}]
 dyn_cape_l_05 [{shapeType=Capsule, length=11.284, radius=15.4132, shapeName=dyn_cape_l_05_capsule, center=<3.8558 0 9.8>, rotation=<rot 90 0 0>}]
 dyn_cape_l_06 [{shapeType=Capsule, length=7.3744, radius=15.4132, shapeName=dyn_cape_l_06_capsule, center=<3.8558 0 9.8>, rotation=<rot 90 0 0>}]
 dyn_cape_l_MSTR [{shapeType=Capsule, length=1, radius=1, shapeName=dyn_cape_l_MSTR_capsule, rotation=<rot 0 0 90>}]
 dyn_cape_mid_l_01 [{shapeType=Capsule, length=11.663, radius=14.8873, shapeName=dyn_cape_mid_l_01_capsule, center=<16.3712 0 16.7>, rotation=<rot 90 0 0>}]
 dyn_cape_mid_l_02 [{shapeType=Capsule, length=7.8705, radius=14.8873, shapeName=dyn_cape_mid_l_02_capsule, center=<16.3712 0 16.3>, rotation=<rot 90 0 0>}]
 dyn_cape_mid_l_03 [{shapeType=Capsule, length=19.3701, radius=15.4132, shapeName=dyn_cape_mid_l_03_capsule, center=<10.1586 0 13.3>, rotation=<rot 90 0 0>}]
 dyn_cape_mid_l_04 [{shapeType=Capsule, length=18.8586, radius=15.4132, shapeName=dyn_cape_mid_l_04_capsule, center=<7.116 0 10.3>, rotation=<rot 90 0 0>}]
 dyn_cape_mid_l_05 [{shapeType=Capsule, length=11.284, radius=15.4132, shapeName=dyn_cape_mid_l_05_capsule, center=<3.8558 0 9.8>, rotation=<rot 90 0 0>}]
 dyn_cape_mid_l_06 [{shapeType=Capsule, length=7.3744, radius=15.4132, shapeName=dyn_cape_mid_l_06_capsule, center=<3.8558 0 9.8>, rotation=<rot 90 0 0>}]
 dyn_cape_mid_l_MSTR [{shapeType=Capsule, length=1, radius=1, shapeName=dyn_cape_mid_l_MSTR_capsule, rotation=<rot 0 0 90>}]
 dyn_cape_mid_r_01 [{shapeType=Capsule, length=11.663, radius=14.8873, shapeName=dyn_cape_mid_r_01_capsule, center=<16.7726 0 16.7>, rotation=<rot 90 0 0>}]
 dyn_cape_mid_r_02 [{shapeType=Capsule, length=7.8705, radius=14.8873, shapeName=dyn_cape_mid_r_02_capsule, center=<16.7726 0 16.3>, rotation=<rot 90 0 0>}]
 dyn_cape_mid_r_03 [{shapeType=Capsule, length=19.3701, radius=15.4132, shapeName=dyn_cape_mid_r_03_capsule, center=<10.56 0 13.3>, rotation=<rot 90 0 0>}]
 dyn_cape_mid_r_04 [{shapeType=Capsule, length=18.8586, radius=15.4132, shapeName=dyn_cape_mid_r_04_capsule, center=<7.5175 0 10.3>, rotation=<rot 90 0 0>}]
 dyn_cape_mid_r_05 [{shapeType=Capsule, length=11.284, radius=15.4132, shapeName=dyn_cape_mid_r_05_capsule, center=<4.2573 0 9.8>, rotation=<rot 90 0 0>}]
 dyn_cape_mid_r_06 [{shapeType=Capsule, length=7.3744, radius=15.4132, shapeName=dyn_cape_mid_r_06_capsule, center=<4.2573 0 9.8>, rotation=<rot 90 0 0>}]
 dyn_cape_mid_r_MSTR [{shapeType=Capsule, length=1, radius=1, shapeName=dyn_cape_mid_r_MSTR_capsule, rotation=<rot 0 0 90>}]
 dyn_cape_r_01 [{shapeType=Capsule, length=11.663, radius=14.8873, shapeName=dyn_cape_r_01_capsule, center=<16.7726 0 16.7>, rotation=<rot 90 0 0>}]
 dyn_cape_r_02 [{shapeType=Capsule, length=7.8705, radius=14.8873, shapeName=dyn_cape_r_02_capsule, center=<16.7726 0 16.3>, rotation=<rot 90 0 0>}]
 dyn_cape_r_03 [{shapeType=Capsule, length=19.3701, radius=15.4132, shapeName=dyn_cape_r_03_capsule, center=<10.56 0 13.3>, rotation=<rot 90 0 0>}]
 dyn_cape_r_04 [{shapeType=Capsule, length=18.8586, radius=15.4132, shapeName=dyn_cape_r_04_capsule, center=<7.5175 0 10.3>, rotation=<rot 90 0 0>}]
 dyn_cape_r_05 [{shapeType=Capsule, length=11.284, radius=15.4132, shapeName=dyn_cape_r_05_capsule, center=<4.2573 0 9.8>, rotation=<rot 90 0 0>}]
 dyn_cape_r_06 [{shapeType=Capsule, length=7.3744, radius=15.4132, shapeName=dyn_cape_r_06_capsule, center=<4.2573 0 9.8>, rotation=<rot 90 0 0>}]
 dyn_cape_r_MSTR [{shapeType=Capsule, length=1, radius=1, shapeName=dyn_cape_r_MSTR_capsule, rotation=<rot 0 0 90>}]
 foot_l [{shapeType=Capsule, length=16.0692, radius=14.6715, shapeName=foot_l_capsule, center=<-5.0209 6.671 0>, rotation=<rot 0 0 90>}]
 foot_r [{shapeType=Capsule, length=16.0692, radius=14.6715, shapeName=foot_l_capsule, center=<2.686 -7.6029 0>, rotation=<rot 0 0 90>}]
 lowerarm_l [{shapeType=Capsule, length=30.4824, radius=8.9735, shapeName=lowerarm_l_capsule, center=<13.5156 0 0>, rotation=<rot 83.0534 -0.0003 -0.0003>}]
 lowerarm_r [{
   shapeType=Capsule
   length=25.7174
   radius=8.5652
   shapeName=lowerarm_r_capsule
   center=<-12.9098 0 0>
   rotation=<rot 80.2665 -0.0005 -0.0006>
  }]
 pelvis [{shapeType=Capsule, length=11.1392, radius=18.0812, shapeName=pelvis_capsule, center=<-5.7674 -0.3857 0>}]
 spine_01 [{shapeType=Capsule, length=8.6247, radius=19.1324, shapeName=spine_01_capsule, center=<0 2.2895 0>}]
 spine_02 [{shapeType=Capsule, length=5.6285, radius=19.8496, shapeName=spine_02_capsule, center=<0 1.0403 0>}]
 spine_03 [{shapeType=Capsule, length=4.8369, radius=19.9858, shapeName=spine_03_capsule, center=<0 1.6707 0>}]
 spine_04 [{shapeType=Capsule, length=1.1195, radius=20.0633, shapeName=spine_04_capsule, center=<0 3.004 0>}]
 spine_05 [{shapeType=Capsule, length=10.6159, radius=16.93, shapeName=spine_05_capsule, center=<0 3.8008 0>}]
 thigh_l [{shapeType=Capsule, length=36.0536, radius=15.2, shapeName=thigh_l_capsule, center=<-13.8103 1.5202 0>, rotation=<rot 90 0 0>}]
 thigh_r [{shapeType=Capsule, length=36.0535, radius=15.1991, shapeName=thigh_r_capsule, center=<14.142 -1.7894 0>, rotation=<rot 90 0 0>}]
 upperarm_l [{
   shapeType=Capsule
   length=28.7561
   radius=8.6793
   shapeName=upperarm_l_capsule
   center=<8.5216 0 0>
   rotation=<rot 84.6531 -179.9994 -179.9994>
  }]
 upperarm_r [{
   shapeType=Capsule
   length=25.8492
   radius=8.3208
   shapeName=upperarm_r_capsule
   center=<-7.9214 0 0>
   rotation=<rot 85.4467 -179.9985 -179.9986>
  }]
}
constraints [
 {bone1Name=dyn_cape_mid_l_01, bone2Name=dyn_cape_mid_l_MSTR, swing1LimitDegrees=10, swing2LimitDegrees=20, twistLimitDegrees=5}
 {bone1Name=dyn_cape_mid_l_02, bone2Name=dyn_cape_mid_l_01, swing1LimitDegrees=10, swing2LimitDegrees=20, twistLimitDegrees=5}
 {bone1Name=dyn_cape_mid_l_03, bone2Name=dyn_cape_mid_l_02, swing1LimitDegrees=10, swing2LimitDegrees=45, twistLimitDegrees=5}
 {bone1Name=dyn_cape_mid_l_04, bone2Name=dyn_cape_mid_l_03, swing1LimitDegrees=10, swing2LimitDegrees=45, twistLimitDegrees=5}
 {bone1Name=dyn_cape_mid_l_05, bone2Name=dyn_cape_mid_l_04, swing1LimitDegrees=10, swing2LimitDegrees=45, twistLimitDegrees=5}
 {bone1Name=dyn_cape_mid_l_06, bone2Name=dyn_cape_mid_l_05, swing1LimitDegrees=10, swing2LimitDegrees=45, twistLimitDegrees=5}
 {bone1Name=dyn_cape_mid_r_01, bone2Name=dyn_cape_mid_r_MSTR, swing1LimitDegrees=10, swing2LimitDegrees=20, twistLimitDegrees=5}
 {bone1Name=dyn_cape_mid_r_02, bone2Name=dyn_cape_mid_r_01, swing1LimitDegrees=10, swing2LimitDegrees=20, twistLimitDegrees=5}
 {bone1Name=dyn_cape_mid_r_03, bone2Name=dyn_cape_mid_r_02, swing1LimitDegrees=10, swing2LimitDegrees=45, twistLimitDegrees=5}
 {bone1Name=dyn_cape_mid_r_04, bone2Name=dyn_cape_mid_r_03, swing1LimitDegrees=10, swing2LimitDegrees=45, twistLimitDegrees=5}
 {bone1Name=dyn_cape_mid_r_05, bone2Name=dyn_cape_mid_r_04, swing1LimitDegrees=10, swing2LimitDegrees=45, twistLimitDegrees=5}
 {bone1Name=dyn_cape_mid_r_06, bone2Name=dyn_cape_mid_r_05, swing1LimitDegrees=10, swing2LimitDegrees=45, twistLimitDegrees=5}
 {bone1Name=dyn_cape_l_01, bone2Name=dyn_cape_l_MSTR, swing1LimitDegrees=10, swing2LimitDegrees=20, twistLimitDegrees=5}
 {bone1Name=dyn_cape_l_02, bone2Name=dyn_cape_l_01, swing1LimitDegrees=10, swing2LimitDegrees=20, twistLimitDegrees=5}
 {bone1Name=dyn_cape_l_03, bone2Name=dyn_cape_l_02, swing1LimitDegrees=10, swing2LimitDegrees=45, twistLimitDegrees=5}
 {bone1Name=dyn_cape_l_04, bone2Name=dyn_cape_l_03, swing1LimitDegrees=10, swing2LimitDegrees=45, twistLimitDegrees=5}
 {bone1Name=dyn_cape_l_05, bone2Name=dyn_cape_l_04, swing1LimitDegrees=10, swing2LimitDegrees=45, twistLimitDegrees=5}
 {bone1Name=dyn_cape_l_06, bone2Name=dyn_cape_l_05, swing1LimitDegrees=10, swing2LimitDegrees=45, twistLimitDegrees=5}
 {bone1Name=dyn_cape_r_01, bone2Name=dyn_cape_r_MSTR, swing1LimitDegrees=10, swing2LimitDegrees=20, twistLimitDegrees=5}
 {bone1Name=dyn_cape_r_02, bone2Name=dyn_cape_r_01, swing1LimitDegrees=10, swing2LimitDegrees=20, twistLimitDegrees=5}
 {bone1Name=dyn_cape_r_03, bone2Name=dyn_cape_r_02, swing1LimitDegrees=10, swing2LimitDegrees=45, twistLimitDegrees=5}
 {bone1Name=dyn_cape_r_04, bone2Name=dyn_cape_r_03, swing1LimitDegrees=10, swing2LimitDegrees=45, twistLimitDegrees=5}
 {bone1Name=dyn_cape_r_05, bone2Name=dyn_cape_r_04, swing1LimitDegrees=10, swing2LimitDegrees=45, twistLimitDegrees=5}
 {bone1Name=dyn_cape_r_06, bone2Name=dyn_cape_r_05, swing1LimitDegrees=10, swing2LimitDegrees=45, twistLimitDegrees=5}
 {
  bone1Name=dyn_cape_mid_l_05
  bone2Name=dyn_cape_mid_r_05
  swing1LimitDegrees=45
  swing1Motion=Free
  swing2LimitDegrees=45
  swing2Motion=Free
  twistLimitDegrees=45
  twistMotion=Free
 }
 {
  bone1Name=dyn_cape_l_05
  bone2Name=dyn_cape_mid_l_05
  swing1LimitDegrees=45
  swing1Motion=Free
  swing2LimitDegrees=45
  swing2Motion=Free
  twistLimitDegrees=45
  twistMotion=Free
 }
 {
  bone1Name=dyn_cape_r_05
  bone2Name=dyn_cape_mid_r_05
  swing1LimitDegrees=45
  swing1Motion=Free
  swing2LimitDegrees=45
  swing2Motion=Free
  twistLimitDegrees=45
  twistMotion=Free
 }
]
