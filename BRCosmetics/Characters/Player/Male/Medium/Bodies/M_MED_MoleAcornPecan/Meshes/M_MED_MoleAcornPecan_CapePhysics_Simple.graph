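# Simplified physics asset for the cape on the M_MED_MoleAcornPecan male medium player body.
# Four cape chains (left, mid-left, mid-right, right) each hang two large, offset capsules from a tiny MSTR capsule, using identical tight swing and twist limits.
# No constraints link neighboring chains or join them to the body.
# Pelvis, spine, clavicles and upper arms carry body capsules, while thigh, calf and foot bodies have no shapes.
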PhysicsAsset /BRCosmetics/Characters/Player/Male/Medium/Bodies/M_MED_MoleAcornPecan/Meshes/M_MED_MoleAcornPecan_CapePhysics_Simple
bodies {
 calf_l []
 calf_r []
 clavicle_l [{shapeType=Capsule, length=11.136, radius=10.8288, shapeName=clavicle_l_capsule, center=<5.1544 0 0>, rotation=<rot 90 0 0>}]
 clavicle_r [{shapeType=Capsule, length=11.0408, radius=10.9595, shapeName=clavicle_r_capsule, center=<-4.3594 0 0>, rotation=<rot 90 0 0>}]
 dyn_cape_l_01 [{shapeType=Capsule, length=11.663, radius=14.8873, shapeName=dyn_cape_l_01_capsule, center=<16.3712 0 16.7>, rotation=<rot 90 0 0>}]
 dyn_cape_l_02 [{shapeType=Capsule, length=7.8705, radius=14.8873, shapeName=dyn_cape_l_02_capsule, center=<16.3712 0 16.3>, rotation=<rot 90 0 0>}]
 dyn_cape_l_MSTR [{shapeType=Capsule, length=1, radius=1, shapeName=dyn_cape_l_MSTR_capsule, rotation=<rot 0 0 90>}]
 dyn_cape_mid_l_01 [{shapeType=Capsule, length=11.663, radius=14.8873, shapeName=dyn_cape_mid_l_01_capsule, center=<16.3712 0 16.7>, rotation=<rot 90 0 0>}]
 dyn_cape_mid_l_02 [{shapeType=Capsule, length=7.8705, radius=14.8873, shapeName=dyn_cape_mid_l_02_capsule, center=<16.3712 0 16.3>, rotation=<rot 90 0 0>}]
 dyn_cape_mid_l_MSTR [{shapeType=Capsule, length=1, radius=1, shapeName=dyn_cape_mid_l_MSTR_capsule, rotation=<rot 0 0 90>}]
 dyn_cape_mid_r_01 [{shapeType=Capsule, length=11.663, radius=14.8873, shapeName=dyn_cape_mid_r_01_capsule, center=<16.7726 0 16.7>, rotation=<rot 90 0 0>}]
 dyn_cape_mid_r_02 [{shapeType=Capsule, length=7.8705, radius=14.8873, shapeName=dyn_cape_mid_r_02_capsule, center=<16.7726 0 16.3>, rotation=<rot 90 0 0>}]
 dyn_cape_mid_r_MSTR [{shapeType=Capsule, length=1, radius=1, shapeName=dyn_cape_mid_r_MSTR_capsule, rotation=<rot 0 0 90>}]
 dyn_cape_r_01 [{shapeType=Capsule, length=11.663, radius=14.8873, shapeName=dyn_cape_r_01_capsule, center=<16.7726 0 16.7>, rotation=<rot 90 0 0>}]
 dyn_cape_r_02 [{shapeType=Capsule, length=7.8705, radius=14.8873, shapeName=dyn_cape_r_02_capsule, center=<16.7726 0 16.3>, rotation=<rot 90 0 0>}]
 dyn_cape_r_MSTR [{shapeType=Capsule, length=1, radius=1, shapeName=dyn_cape_r_MSTR_capsule, rotation=<rot 0 0 90>}]
 foot_l []
 foot_r []
 pelvis [{shapeType=Capsule, length=11.1392, radius=18.0812, shapeName=pelvis_capsule, center=<-5.7674 1.6628 0>}]
 spine_01 [{shapeType=Capsule, length=8.6247, radius=19.1324, shapeName=spine_01_capsule, center=<0 2.2895 0>}]
 spine_02 [{shapeType=Capsule, length=6.8028, radius=18.6753, shapeName=spine_02_capsule, center=<0 1.0403 0>}]
 spine_03 [{shapeType=Capsule, length=5.7231, radius=19.0995, shapeName=spine_03_capsule, center=<0 1.6707 0>}]
 spine_04 [{shapeType=Capsule, length=1.4015, radius=19.7813, shapeName=spine_04_capsule, center=<0 3.004 0>}]
 thigh_l []
 thigh_r []
 upperarm_l [{
   shapeType=Capsule
   length=28.7561
   radius=8.6793
   shapeName=upperarm_l_capsule
   center=<8.5216 0 0>
   rotation=<rot 84.6531 -179.9994 -179.9994>
  }]
 upperarm_r [{
   shapeType=Capsule
   length=25.8492
   radius=8.3208
   shapeName=upperarm_r_capsule
   center=<-7.9214 0 0>
   rotation=<rot 85.4467 -179.9985 -179.9986>
  }]
}
constraints [
 {bone1Name=dyn_cape_mid_l_01, bone2Name=dyn_cape_mid_l_MSTR, swing1LimitDegrees=10, swing2LimitDegrees=20, twistLimitDegrees=5}
 {bone1Name=dyn_cape_mid_l_02, bone2Name=dyn_cape_mid_l_01, swing1LimitDegrees=10, swing2LimitDegrees=20, twistLimitDegrees=5}
 {bone1Name=dyn_cape_mid_r_01, bone2Name=dyn_cape_mid_r_MSTR, swing1LimitDegrees=10, swing2LimitDegrees=20, twistLimitDegrees=5}
 {bone1Name=dyn_cape_mid_r_02, bone2Name=dyn_cape_mid_r_01, swing1LimitDegrees=10, swing2LimitDegrees=20, twistLimitDegrees=5}
 {bone1Name=dyn_cape_l_01, bone2Name=dyn_cape_l_MSTR, swing1LimitDegrees=10, swing2LimitDegrees=20, twistLimitDegrees=5}
 {bone1Name=dyn_cape_l_02, bone2Name=dyn_cape_l_01, swing1LimitDegrees=10, swing2LimitDegrees=20, twistLimitDegrees=5}
 {bone1Name=dyn_cape_r_01, bone2Name=dyn_cape_r_MSTR, swing1LimitDegrees=10, swing2LimitDegrees=20, twistLimitDegrees=5}
 {bone1Name=dyn_cape_r_02, bone2Name=dyn_cape_r_01, swing1LimitDegrees=10, swing2LimitDegrees=20, twistLimitDegrees=5}
]
