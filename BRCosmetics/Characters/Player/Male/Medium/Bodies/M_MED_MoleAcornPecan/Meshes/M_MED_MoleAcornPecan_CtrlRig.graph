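# Control Rig for the MoleAcornPecan medium male player body, correcting deform, toe, middle-finger, loincloth and cape bones.
# It runs a shared deform-rig function, toe aim and foot-yaw correctives, finger constraints, then weighted loincloth parent constraints.
# Spherical pose readers on spine and thighs drive cape and loincloth offsets, with interpolated thigh weights.
# Disable Floor Slope pushes the cape's ground-collision bone far down through a Kalman-smoothed weight.
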ControlRigBlueprint /BRCosmetics/Characters/Player/Male/Medium/Bodies/M_MED_MoleAcornPecan/Meshes/M_MED_MoleAcornPecan_CtrlRig
graphs [
 New Function_ContainedGraph {nodes [
   RigVMFunctionEntryNode()
   > &RigUnit_AimConstraintLocalSpaceOffset(
    Child=(Type=Bone,Name="deform_toe_l_outer")
    Filter=(bX=True,bY=True,bZ=True)
    Parents=((Item=(Type=Control,Name="ball_l_ctrl"),Weight=1))
    Weight=1
    WorldUp=(Target=(X=0,Y=0,Z=1),Kind=Location,Space=(Type=Control,Name="ball_l_ctrl"))
   )
   > &RigUnit_AimConstraintLocalSpaceOffset(
    Child=(Type=Bone,Name="deform_toe_l_mid")
    Filter=(bX=True,bY=True,bZ=True)
    Parents=((Item=(Type=Control,Name="ball_l_ctrl"),Weight=1))
    Weight=1
    WorldUp=(Target=(X=0,Y=0,Z=1),Kind=Location,Space=(Type=Control,Name="ball_l_ctrl"))
   )
   > &RigUnit_AimConstraintLocalSpaceOffset(
    Child=(Type=Bone,Name="deform_toe_l_inner")
    Filter=(bX=True,bY=True,bZ=True)
    Parents=((Item=(Type=Control,Name="ball_l_ctrl"),Weight=1))
    Weight=1
    WorldUp=(Target=(X=0,Y=0,Z=1),Kind=Location,Space=(Type=Control,Name="ball_l_ctrl"))
   )
   > &RigUnit_AimConstraintLocalSpaceOffset(
    Child=(Type=Bone,Name="deform_toe_r_outer")
    Filter=(bX=True,bY=True,bZ=True)
    Parents=((Item=(Type=Control,Name="ball_r_ctrl"),Weight=1))
    Weight=1
    WorldUp=(Target=(X=0,Y=0,Z=1),Kind=Location,Space=(Type=Control,Name="ball_r_ctrl"))
   )
   > &RigUnit_AimConstraintLocalSpaceOffset(
    Child=(Type=Bone,Name="deform_toe_r_mid")
    Filter=(bX=True,bY=True,bZ=True)
    Parents=((Item=(Type=Control,Name="ball_r_ctrl"),Weight=1))
    Weight=1
    WorldUp=(Target=(X=0,Y=0,Z=1),Kind=Location,Space=(Type=Control,Name="ball_r_ctrl"))
   )
   > &RigUnit_AimConstraintLocalSpaceOffset(
    Child=(Type=Bone,Name="deform_toe_r_inner")
    Filter=(bX=True,bY=True,bZ=True)
    Parents=((Item=(Type=Control,Name="ball_r_ctrl"),Weight=1))
    Weight=1
    WorldUp=(Target=(X=0,Y=0,Z=1),Kind=Location,Space=(Type=Control,Name="ball_r_ctrl"))
   )
   > Get Delta Ref = RigVMFunctionReferenceNode(Item=(Type=Bone,Name="foot_l"))
   > Get Delta Ref_1 = RigVMFunctionReferenceNode(Item=(Type=Bone,Name="foot_r"))
   Remap = &RigVMFunction_MathDoubleRemap(SourceMaximum=50, SourceMinimum=0, TargetMaximum=1, TargetMinimum=0, Value=@"Get Delta Ref".Result.Rotation.Yaw)
   > Modify Transforms with Mirror = RigVMFunctionReferenceNode(
    Base Modify Data="((Item=(Type=Bone,Name=\"deform_toe_l_inner\"),Transform=(Rotation=(X=0,Y=0.2971,Z=0,W=0.9549),Translation=(X=-0.6923,Y=0,Z=-1.3044),Scale3D=(X=1,Y=1,Z=1))),(Item=(Type=Bone,Name=\"deform_toe_l_outer\"),Transform=(Rotation=(X=0,Y=-0.2039,Z=0,W=0.979),Translati ...(truncated)"
    Mirror Modify Data="((Item=(Type=Bone,Name=\"deform_toe_r_inner\"),Transform=(Rotation=(X=0,Y=0,Z=0,W=1),Translation=(X=0,Y=0,Z=0),Scale3D=(X=1,Y=1,Z=1))),(Item=(Type=Bone,Name=\"deform_toe_r_outer\"),Transform=(Rotation=(X=0,Y=0,Z=0,W=1),Translation=(X=0,Y=0,Z=0),Scale3D=(X=1,Y= ...(truncated)"
    Weight=@Remap.Result
    "Mirror Weight "=&RigVMFunction_MathDoubleRemap(SourceMaximum=50, SourceMinimum=0, TargetMaximum=1, TargetMinimum=0, Value=@"Get Delta Ref_1".Result.Rotation.Yaw).Result
   )
   > Get Delta Ref_2 = RigVMFunctionReferenceNode(Item=(Type=Bone,Name="foot_l"))
   > Get Delta Ref_3 = RigVMFunctionReferenceNode(Item=(Type=Bone,Name="foot_r"))
   > Modify Transforms with Mirror_1 = RigVMFunctionReferenceNode(
    Base Modify Data="((Item=(Type=Bone,Name=\"deform_toe_l_inner\"),Transform=(Rotation=(X=0,Y=-0.2164,Z=0,W=0.9763),Translation=(X=0,Y=0,Z=0),Scale3D=(X=1,Y=1,Z=1))),(Item=(Type=Bone,Name=\"deform_toe_l_outer\"),Transform=(Rotation=(X=0,Y=0.2164,Z=0,W=0.9763),Translation=(X=0,Y=0 ...(truncated)"
    Mirror Modify Data="((Item=(Type=Bone,Name=\"deform_toe_r_inner\"),Transform=(Rotation=(X=0,Y=0,Z=0,W=1),Translation=(X=0,Y=0,Z=0),Scale3D=(X=1,Y=1,Z=1))),(Item=(Type=Bone,Name=\"deform_toe_r_outer\"),Transform=(Rotation=(X=0,Y=0,Z=0,W=1),Translation=(X=0,Y=0,Z=0),Scale3D=(X=1,Y= ...(truncated)"
    Weight=&RigVMFunction_MathDoubleRemap(SourceMaximum=-20, SourceMinimum=0, TargetMaximum=1, TargetMinimum=0, Value=@"Get Delta Ref_2".Result.Rotation.Yaw).Result
    "Mirror Weight "=&RigVMFunction_MathDoubleRemap(SourceMaximum=-20, SourceMinimum=0, TargetMaximum=1, TargetMinimum=0, Value=@"Get Delta Ref_3".Result.Rotation.Yaw).Result
   )
   > RigVMFunctionReturnNode()
  ]}
 New Function_ContainedGraph {nodes [
   RigVMFunctionEntryNode()
   > For_Each = RigVMDispatchNode(Array=RigVMVariableNode() {variable=Source Middle Fingers}.Value) {template="DISPATCH_RigVMDispatch_ArrayIterator(in Array,out Element,out Index,out Count,out Ratio)"}
   > &RigUnit_ParentConstraint(
    AdvancedSettings=(InterpolationType=Average,RotationOrderForFilter=XZY)
    Parents=((Item=(Type=Bone,Name="None"),Weight=1))
    Weight=1
    Child=RigVMDispatchNode(Index=@For_Each.Index, Array=RigVMVariableNode() {variable=Target Middle Fingers}.Value) {template="DISPATCH_RigVMDispatch_ArrayGetAtIndex(in Array,in Index,out Element)"}.Element
    Parents.0.Item=@For_Each.Element
   )
   RigVMFunctionReturnNode(ExecuteContext=@For_Each.Completed)
  ]}
 RigVMFunctionLibrary {nodes [
   Deform Middle Finger = RigVMCollapseNode(ExecuteContext=())
   Deform Toes = RigVMCollapseNode(ExecuteContext=())
  ]}
 RigVMFunction_Sequence_1_ContainedGraph {nodes [
   n1 = RigVMFunctionEntryNode()
   RigVMFunction_Sequence = &RigVMFunction_Sequence(ExecuteContext=@n1.ExecuteContext)
   n2 = &RigVMFunction_Sequence(ExecuteContext=@RigVMFunction_Sequence.B)
   RigVMFunctionReturnNode(A=@RigVMFunction_Sequence.A, B=@n2.A, C=@n2.B)
  ]}
 RigVMModel {nodes [
   RigUnit_BeginExecution = &RigUnit_BeginExecution()
   RigVMFunction_Sequence = RigVMAggregateNode(ExecuteContext=@RigUnit_BeginExecution.ExecutePin)
   FN_DeformRig_FNC = RigVMFunctionReferenceNode(
    DeformRig_ Curve="(deformRig_l_upperarm_up_alpha_72_#1=False,deformRig_r_upperarm_up_alpha_73_#2=False,deformRig_l_upperarm_fr_alpha_74_#3=False,deformRig_r_upperarm_fr_alpha_75_CB51B8 ...(truncated)"
    DeformRig_Bool="(upperarm_1_#4=True,elbow_3_#5=True,wrist_7_#6=True,groin_8_#7=True,glute_17_#8=True,knee_9_40C408324CA261 ...(truncated)"
    DeformRig_Debug="(upperarm_debug=False,wrist_debug=False,lat_debug=False,groinGlute_debug=False,upperarm_debug_18_#4=False,wrist_debug_20_#6=False,groinGlute_debug_27_#7=False,lat_deb ...(truncated)"
    DeformRig_Value="(upperarm_up_18_#9=(Translation=(X=0,Y=0,Z=8),Scale3D=(X=0.7,Y=1,Z=1)),upperarm_fwd_fr_19_#10=(Translation=(X=0,Y=10,Z=0),Scale3D=(X=0.5,Y=1,Z=1)),upperarm_fwd_bk_20_9973AAD84EE0A1E349BC9E85067D4A3 ...(truncated)"
    ExecuteContext=@RigVMFunction_Sequence.A
   )
   > Deform Toes = RigVMFunctionReferenceNode()
   > Deform Middle Finger = RigVMFunctionReferenceNode()
   > &RigUnit_ParentConstraint(
    AdvancedSettings=(RotationOrderForFilter=XZY)
    Child=(Type=Bone,Name="dyn_loincloth_fr_l_MSTR")
    Parents=((Item=(Type=Bone,Name="spine_04"),Weight=0.0181),(Item=(Type=Bone,Name="spine_03"),Weight=0.1458),(Item=(Type=Bone,Name="spine_01"),Weight=0.1902),(Item=(Type=Bone,Name="spine_02"),Weight=0.2362),(Item=(Type=Bone,Name="pelvis"),Weight=0.4098))
    Weight=1
   )
   > &RigUnit_ParentConstraint(
    AdvancedSettings=(RotationOrderForFilter=XZY)
    Child=(Type=Bone,Name="dyn_loincloth_fr_r_MSTR")
    Parents=((Item=(Type=Bone,Name="spine_04"),Weight=0.018),(Item=(Type=Bone,Name="spine_03"),Weight=0.1457),(Item=(Type=Bone,Name="spine_01"),Weight=0.1901),(Item=(Type=Bone,Name="spine_02"),Weight=0.236),(Item=(Type=Bone,Name="pelvis"),Weight=0.4102))
    Weight=1
   )
   > &RigUnit_ParentConstraint(
    AdvancedSettings=(RotationOrderForFilter=XZY)
    Child=(Type=Bone,Name="dyn_loincloth_side_fr_l_MSTR")
    Parents=((Item=(Type=Bone,Name="spine_03"),Weight=0.1298),(Item=(Type=Bone,Name="spine_01"),Weight=0.1812),(Item=(Type=Bone,Name="spine_02"),Weight=0.2196),(Item=(Type=Bone,Name="pelvis"),Weight=0.4693))
    Weight=1
   )
   > &RigUnit_ParentConstraint(
    AdvancedSettings=(RotationOrderForFilter=XZY)
    Child=(Type=Bone,Name="dyn_loincloth_side_bk_l_MSTR")
    Parents=((Item=(Type=Bone,Name="deform_hip_l"),Weight=0.0086),(Item=(Type=Bone,Name="spine_03"),Weight=0.0491),(Item=(Type=Bone,Name="spine_01"),Weight=0.0734),(Item=(Type=Bone,Name="spine_02"),Weight=0.0852),(Item=(Type=Bone,Name="pelvis"),Weight=0.7837))
    Weight=1
   )
   > &RigUnit_ParentConstraint(
    AdvancedSettings=(RotationOrderForFilter=XZY)
    Child=(Type=Bone,Name="dyn_loincloth_side_fr_r_MSTR")
    Parents=((Item=(Type=Bone,Name="spine_03"),Weight=0.1295),(Item=(Type=Bone,Name="spine_01"),Weight=0.181),(Item=(Type=Bone,Name="spine_02"),Weight=0.2192),(Item=(Type=Bone,Name="pelvis"),Weight=0.4702))
    Weight=1
   )
   > &RigUnit_ParentConstraint(
    AdvancedSettings=(RotationOrderForFilter=XZY)
    Child=(Type=Bone,Name="dyn_loincloth_side_bk_r_MSTR")
    Parents=((Item=(Type=Bone,Name="deform_hip_r"),Weight=0.0091),(Item=(Type=Bone,Name="spine_03"),Weight=0.0484),(Item=(Type=Bone,Name="spine_01"),Weight=0.0723),(Item=(Type=Bone,Name="spine_02"),Weight=0.0839),(Item=(Type=Bone,Name="pelvis"),Weight=0.7863))
    Weight=1
   )
   > &RigUnit_ParentConstraint(
    AdvancedSettings=(RotationOrderForFilter=XZY)
    Child=(Type=Bone,Name="dyn_loincloth_bk_r_01")
    Parents=((Item=(Type=Bone,Name="spine_03"),Weight=0.0096),(Item=(Type=Bone,Name="spine_02"),Weight=0.0147),(Item=(Type=Bone,Name="spine_01"),Weight=0.0151),(Item=(Type=Bone,Name="deform_glute_r"),Weight=0.023),(Item=(Type=Bone,Name="pelvis"),Weight=0.9376))
    Weight=1
   )
   > &RigUnit_ParentConstraint(
    AdvancedSettings=(RotationOrderForFilter=XZY)
    Child=(Type=Bone,Name="dyn_loincloth_bk_l_01")
    Parents=((Item=(Type=Bone,Name="spine_03"),Weight=0.0096),(Item=(Type=Bone,Name="spine_02"),Weight=0.0147),(Item=(Type=Bone,Name="spine_01"),Weight=0.0151),(Item=(Type=Bone,Name="deform_glute_l"),Weight=0.023),(Item=(Type=Bone,Name="pelvis"),Weight=0.9376))
    Weight=1
   )
   > &RigUnit_ModifyTransforms(
    ItemToModify="((Item=(Type=Bone,Name=\"dyn_cape_l_MSTR\"),Transform=(Rotation=(X=0,Y=-0.0175,Z=0,W=0.9998),Translation=(X=0,Y=0,Z=2),Scale3D=(X=1,Y=1,Z=1))),(Item=(Type=Bone,Name=\"dyn_cape_mid_l_MSTR\"),Transform=(Rotation=(X=0,Y=-0.0175,Z=0,W=0.9998),Translation=(X=0,Y=0, ...(truncated)"
    Weight=1
    WeightMinimum=0
   )
   > n1 = &RigUnit_SphericalPoseReader(
    ActiveRegionScaleFactors=(PositiveWidth=3,NegativeWidth=1,PositiveHeight=3,NegativeHeight=3)
    ActiveRegionSize=0.05
    Debug=(bDrawDebug=True,bDraw2D=False,bDrawLocalAxes=False,DebugScale=25,DebugSegments=20,DebugThickness=0.25)
    DriverAxis=(X=0,Y=1,Z=0)
    DriverItem=(Type=Bone,Name="spine_05")
    FalloffRegionScaleFactors=(PositiveWidth=1,NegativeWidth=1,PositiveHeight=1,NegativeHeight=1)
    FalloffSize=0.25
    FlipHeightScaling=False
    FlipWidthScaling=False
    OptionalParentItem=(Type=Bone,Name="pelvis")
    RotationOffset=(X=90,Y=0,Z=-50)
   )
   > &RigUnit_ModifyTransforms(
    ItemToModify="((Item=(Type=Bone,Name=\"dyn_cape_l_MSTR\"),Transform=(Rotation=(X=0,Y=-0.342,Z=0,W=0.9397),Translation=(X=0,Y=0,Z=0),Scale3D=(X=1,Y=1,Z=1))),(Item=(Type=Bone,Name=\"dyn_cape_mid_l_MSTR\"),Transform=(Rotation=(X=0,Y=-0.342,Z=0,W=0.9397),Translation=(X=0,Y=0,Z= ...(truncated)"
    WeightMinimum=0
    Weight=@n1.OutputParam
   )
   n2 = &RigVMFunction_Sequence(ExecuteContext=@RigVMFunction_Sequence.B)
   n3 = &RigUnit_SphericalPoseReader(
    ActiveRegionScaleFactors=(PositiveWidth=3,NegativeWidth=1,PositiveHeight=1.2,NegativeHeight=1)
    ActiveRegionSize=0.1
    Debug=(bDrawDebug=false,bDraw2D=False,bDrawLocalAxes=False,DebugScale=25,DebugSegments=20,DebugThickness=0.25)
    DriverAxis=(X=-1,Y=0,Z=0)
    DriverItem=(Type=Bone,Name="thigh_l")
    FalloffRegionScaleFactors=(PositiveWidth=1,NegativeWidth=1,PositiveHeight=0.75,NegativeHeight=0.75)
    FalloffSize=0.58
    FlipHeightScaling=False
    FlipWidthScaling=False
    OptionalParentItem=(Type=Bone,Name="pelvis")
    RotationOffset=(X=90,Y=0,Z=-15)
    ExecutePin=@n2.A
   )
   RigVMFunction_AlphaInterp = &RigVMFunction_AlphaInterp(InterpSpeedDecreasing=10, InterpSpeedIncreasing=10, Scale=1, bClampResult=False, bInterpResult=True, Value=@n3.OutputParam)
   > &RigUnit_ModifyTransforms(
    ItemToModify="((Item=(Type=Bone,Name=\"dyn_loincloth_fr_l_MSTR\"),Transform=(Rotation=(X=0.406,Y=0.1266,Z=0.1322,W=0.8953),Translation=(X=0,Y=0,Z=0),Scale3D=(X=1,Y=1,Z=1))),(Item=(Type=Bone,Name=\"dyn_loincloth_fr_r_MSTR\"),Transform=(Rotation=(X=0.1522,Y=-0.1314,Z=-0.1522, ...(truncated)"
    WeightMinimum=0
    Weight=@RigVMFunction_AlphaInterp.Result
   )
   > &RigUnit_SphericalPoseReader(
    ActiveRegionScaleFactors=(PositiveWidth=3,NegativeWidth=1,PositiveHeight=1.3,NegativeHeight=1.2)
    ActiveRegionSize=0.1
    Debug=(bDrawDebug=false,bDraw2D=False,bDrawLocalAxes=False,DebugScale=25,DebugSegments=20,DebugThickness=0.25)
    DriverAxis=(X=-1,Y=0,Z=0)
    DriverItem=(Type=Bone,Name="thigh_l")
    FalloffRegionScaleFactors=(PositiveWidth=1,NegativeWidth=1,PositiveHeight=0.5,NegativeHeight=0.5)
    FalloffSize=0.57
    FlipHeightScaling=False
    FlipWidthScaling=False
    OptionalParentItem=(Type=Bone,Name="pelvis")
    RotationOffset=(X=130,Y=5,Z=-12)
   )
   > &RigUnit_SphericalPoseReader(
    ActiveRegionScaleFactors=(PositiveWidth=3,NegativeWidth=1,PositiveHeight=1,NegativeHeight=1)
    ActiveRegionSize=0.1
    Debug=(bDrawDebug=false,bDraw2D=False,bDrawLocalAxes=False,DebugScale=25,DebugSegments=20,DebugThickness=0.25)
    DriverAxis=(X=-1,Y=0,Z=0)
    DriverItem=(Type=Bone,Name="thigh_l")
    FalloffRegionScaleFactors=(PositiveWidth=1,NegativeWidth=1,PositiveHeight=1,NegativeHeight=1)
    FalloffSize=0.5
    FlipHeightScaling=False
    FlipWidthScaling=False
    OptionalParentItem=(Type=Bone,Name="pelvis")
    RotationOffset=(X=180,Y=0,Z=0)
   )
   > &RigUnit_SphericalPoseReader(
    ActiveRegionScaleFactors=(PositiveWidth=3,NegativeWidth=1,PositiveHeight=1.5,NegativeHeight=1.5)
    ActiveRegionSize=0.1
    Debug=(bDrawDebug=false,bDraw2D=False,bDrawLocalAxes=False,DebugScale=25,DebugSegments=20,DebugThickness=0.25)
    DriverAxis=(X=-1,Y=0,Z=0)
    DriverItem=(Type=Bone,Name="thigh_l")
    FalloffRegionScaleFactors=(PositiveWidth=1,NegativeWidth=1,PositiveHeight=1,NegativeHeight=1)
    FalloffSize=0.5
    FlipHeightScaling=False
    FlipWidthScaling=False
    OptionalParentItem=(Type=Bone,Name="pelvis")
    RotationOffset=(X=-125,Y=0,Z=0)
   )
   > &RigUnit_SphericalPoseReader(
    ActiveRegionScaleFactors=(PositiveWidth=3,NegativeWidth=1,PositiveHeight=3,NegativeHeight=3)
    ActiveRegionSize=0.1
    Debug=(bDrawDebug=false,bDraw2D=False,bDrawLocalAxes=False,DebugScale=25,DebugSegments=20,DebugThickness=0.25)
    DriverAxis=(X=-1,Y=0,Z=0)
    DriverItem=(Type=Bone,Name="thigh_l")
    FalloffRegionScaleFactors=(PositiveWidth=1,NegativeWidth=0.95,PositiveHeight=1,NegativeHeight=1)
    FalloffSize=0.5
    FlipHeightScaling=False
    FlipWidthScaling=False
    OptionalParentItem=(Type=Bone,Name="pelvis")
    RotationOffset=(X=-90,Y=0,Z=0)
   )
   > n4 = &RigUnit_SphericalPoseReader(
    ActiveRegionScaleFactors=(PositiveWidth=3,NegativeWidth=1,PositiveHeight=1.2,NegativeHeight=1)
    ActiveRegionSize=0.1
    Debug=(bDrawDebug=false,bDraw2D=False,bDrawLocalAxes=False,DebugScale=25,DebugSegments=20,DebugThickness=0.25)
    DriverAxis=(X=-1,Y=0,Z=0)
    DriverItem=(Type=Bone,Name="thigh_l")
    FalloffRegionScaleFactors=(PositiveWidth=1,NegativeWidth=1,PositiveHeight=0.3,NegativeHeight=0.3)
    FalloffSize=0.58
    FlipHeightScaling=False
    FlipWidthScaling=False
    OptionalParentItem=(Type=Bone,Name="pelvis")
    RotationOffset=(X=45,Y=0,Z=-15)
   )
   n5 = &RigVMFunction_AlphaInterp(InterpSpeedDecreasing=10, InterpSpeedIncreasing=10, Scale=1, bClampResult=False, bInterpResult=True, Value=@n4.OutputParam)
   > &RigUnit_ModifyTransforms(
    ItemToModify="((Item=(Type=Bone,Name=\"dyn_loincloth_fr_l_MSTR\"),Transform=(Rotation=(X=0.1886,Y=0.3538,Z=0.0731,W=0.9132),Translation=(X=0,Y=0,Z=0),Scale3D=(X=1,Y=1,Z=1))),(Item=(Type=Bone,Name=\"dyn_loincloth_fr_r_MSTR\"),Transform=(Rotation=(X=0.2379,Y=-0.3586,Z=0.095,W ...(truncated)"
    WeightMinimum=0
    Weight=@n5.Result
   )
   n6 = &RigUnit_SphericalPoseReader(
    ActiveRegionScaleFactors=(PositiveWidth=3,NegativeWidth=1,PositiveHeight=1.2,NegativeHeight=1)
    ActiveRegionSize=0.1
    Debug=(bDrawDebug=true,bDraw2D=False,bDrawLocalAxes=False,DebugScale=25,DebugSegments=20,DebugThickness=0.25)
    DriverAxis=(X=1,Y=0,Z=0)
    DriverItem=(Type=Bone,Name="thigh_r")
    FalloffRegionScaleFactors=(PositiveWidth=1,NegativeWidth=1,PositiveHeight=0.75,NegativeHeight=0.75)
    FalloffSize=0.58
    FlipHeightScaling=true
    FlipWidthScaling=true
    OptionalParentItem=(Type=Bone,Name="None")
    RotationOffset=(X=-90,Y=0,Z=-15)
    ExecutePin=@n2.B
   )
   n7 = &RigVMFunction_AlphaInterp(InterpSpeedDecreasing=10, InterpSpeedIncreasing=10, Scale=1, bClampResult=False, bInterpResult=True, Value=@n6.OutputParam)
   > &RigUnit_ModifyTransforms(
    ItemToModify="((Item=(Type=Bone,Name=\"dyn_loincloth_fr_l_MSTR\"),Transform=(Rotation=(X=0.4,Y=-0.1392,Z=-0.2205,W=0.8786),Translation=(X=0,Y=0,Z=0),Scale3D=(X=1,Y=1,Z=1))),(Item=(Type=Bone,Name=\"dyn_loincloth_fr_r_MSTR\"),Transform=(Rotation=(X=0.1747,Y=0.0188,Z=0.2594,W= ...(truncated)"
    WeightMinimum=0
    Weight=@n7.Result
   )
   > &RigUnit_SphericalPoseReader(
    ActiveRegionScaleFactors=(PositiveWidth=3,NegativeWidth=1,PositiveHeight=1.3,NegativeHeight=1.2)
    ActiveRegionSize=0.1
    Debug=(bDrawDebug=false,bDraw2D=False,bDrawLocalAxes=False,DebugScale=25,DebugSegments=20,DebugThickness=0.25)
    DriverAxis=(X=1,Y=0,Z=0)
    DriverItem=(Type=Bone,Name="thigh_r")
    FalloffRegionScaleFactors=(PositiveWidth=1,NegativeWidth=1,PositiveHeight=0.5,NegativeHeight=0.5)
    FalloffSize=0.57
    FlipHeightScaling=true
    FlipWidthScaling=true
    OptionalParentItem=(Type=Bone,Name="None")
    RotationOffset=(X=-50,Y=5,Z=-12)
   )
   > &RigUnit_SphericalPoseReader(
    ActiveRegionScaleFactors=(PositiveWidth=3,NegativeWidth=1,PositiveHeight=1,NegativeHeight=1)
    ActiveRegionSize=0.1
    Debug=(bDrawDebug=false,bDraw2D=False,bDrawLocalAxes=False,DebugScale=25,DebugSegments=20,DebugThickness=0.25)
    DriverAxis=(X=1,Y=0,Z=0)
    DriverItem=(Type=Bone,Name="thigh_r")
    FalloffRegionScaleFactors=(PositiveWidth=1,NegativeWidth=1,PositiveHeight=1,NegativeHeight=1)
    FalloffSize=0.5
    FlipHeightScaling=true
    FlipWidthScaling=true
    OptionalParentItem=(Type=Bone,Name="None")
    RotationOffset=(X=0,Y=0,Z=0)
   )
   > &RigUnit_SphericalPoseReader(
    ActiveRegionScaleFactors=(PositiveWidth=3,NegativeWidth=1,PositiveHeight=1.5,NegativeHeight=1.5)
    ActiveRegionSize=0.1
    Debug=(bDrawDebug=false,bDraw2D=False,bDrawLocalAxes=False,DebugScale=25,DebugSegments=20,DebugThickness=0.25)
    DriverAxis=(X=1,Y=0,Z=0)
    DriverItem=(Type=Bone,Name="thigh_r")
    FalloffRegionScaleFactors=(PositiveWidth=1,NegativeWidth=1,PositiveHeight=1,NegativeHeight=1)
    FalloffSize=0.5
    FlipHeightScaling=true
    FlipWidthScaling=true
    OptionalParentItem=(Type=Bone,Name="None")
    RotationOffset=(X=55,Y=0,Z=0)
   )
   > &RigUnit_SphericalPoseReader(
    ActiveRegionScaleFactors=(PositiveWidth=3,NegativeWidth=1,PositiveHeight=3,NegativeHeight=3)
    ActiveRegionSize=0.1
    Debug=(bDrawDebug=false,bDraw2D=False,bDrawLocalAxes=False,DebugScale=25,DebugSegments=20,DebugThickness=0.25)
    DriverAxis=(X=1,Y=0,Z=0)
    DriverItem=(Type=Bone,Name="thigh_r")
    FalloffRegionScaleFactors=(PositiveWidth=1,NegativeWidth=0.95,PositiveHeight=1,NegativeHeight=1)
    FalloffSize=0.5
    FlipHeightScaling=true
    FlipWidthScaling=true
    OptionalParentItem=(Type=Bone,Name="None")
    RotationOffset=(X=90,Y=0,Z=0)
   )
   > n8 = &RigUnit_SphericalPoseReader(
    ActiveRegionScaleFactors=(PositiveWidth=3,NegativeWidth=1,PositiveHeight=1.2,NegativeHeight=1)
    ActiveRegionSize=0.1
    Debug=(bDrawDebug=false,bDraw2D=False,bDrawLocalAxes=False,DebugScale=25,DebugSegments=20,DebugThickness=0.25)
    DriverAxis=(X=1,Y=0,Z=0)
    DriverItem=(Type=Bone,Name="thigh_r")
    FalloffRegionScaleFactors=(PositiveWidth=1,NegativeWidth=1,PositiveHeight=0.3,NegativeHeight=0.3)
    FalloffSize=0.58
    FlipHeightScaling=true
    FlipWidthScaling=true
    OptionalParentItem=(Type=Bone,Name="None")
    RotationOffset=(X=-135,Y=0,Z=-15)
   )
   n9 = &RigVMFunction_AlphaInterp(InterpSpeedDecreasing=10, InterpSpeedIncreasing=10, Scale=1, bClampResult=False, bInterpResult=True, Value=@n8.OutputParam)
   > &RigUnit_ModifyTransforms(
    ItemToModify="((Item=(Type=Bone,Name=\"dyn_loincloth_fr_l_MSTR\"),Transform=(Rotation=(X=0.3761,Y=-0.4427,Z=-0.2119,W=0.7859),Translation=(X=0,Y=0,Z=0),Scale3D=(X=1,Y=1,Z=1))),(Item=(Type=Bone,Name=\"dyn_loincloth_fr_r_MSTR\"),Transform=(Rotation=(X=0.3313,Y=0.4169,Z=0.1664 ...(truncated)"
    WeightMinimum=0
    Weight=@n9.Result
   )
   RigVMFunction_ControlFlowBranch = &RigVMFunction_ControlFlowBranch(Condition=RigVMVariableNode() {variable=Disable Floor Slope}.Value, ExecuteContext=@RigVMFunction_Sequence.C)
   KalmanFilter = &RigVMFunction_KalmanFloat(BufferSize=16, Value=RigVMVariableNode() {variable=Push Down Weight}.Value) {template="KalmanFilter::Execute(in Value,in BufferSize,out Result)"}
   &RigUnit_ModifyTransforms(
    ItemToModify=((Item=(Type=Bone,Name="dyn_cape_ground_collision"),Transform=(Rotation=(X=0,Y=0,Z=0,W=1),Translation=(X=0,Y=0,Z=-500),Scale3D=(X=1,Y=1,Z=1))))
    WeightMinimum=0
    Weight=@KalmanFilter.Result
    ExecutePin=@RigVMFunction_ControlFlowBranch.Completed
   )
   RigVMVariableNode(Value=0, ExecuteContext=@RigVMFunction_ControlFlowBranch.False) {variable=Push Down Weight}
   RigVMVariableNode(Value=1, ExecuteContext=@RigVMFunction_ControlFlowBranch.True) {variable=Push Down Weight}
   &RigUnit_AimConstraintLocalSpaceOffset(
    Child=(Type=Bone,Name="attach_cape")
    ExecutePin=()
    Filter=(bX=true,bY=false,bZ=false)
    Parents=((Item=(Type=Bone,Name="cape_aim"),Weight=1))
    Weight=1
    WorldUp=(Target=(X=0,Y=0,Z=1),Kind=Direction,Space=(Type=None,Name="None"))
   )
   &RigUnit_ModifyTransforms(
    ExecutePin=()
    ItemToModify=((Item=(Type=Bone,Name="attach_cape"),Transform=(Rotation=(X=0,Y=0,Z=0,W=1),Translation=(X=0,Y=0,Z=0),Scale3D=(X=1,Y=1,Z=1))))
    Weight=1
    WeightMinimum=0
   )
  ]}
]
hierarchy [
 root {loc=<0 0 0>, rot=<0 0 0 1>}
 pelvis {parent=root, loc=<0 -0.7071 93.1191>, rot=<0.0127 0.707 -0.0127 -0.707>}
 spine_01 {parent=pelvis, loc=<6.7494 -0.6388 0>, rot=<0 0 0.034 -0.9994>}
 spine_02 {parent=spine_01, loc=<7.2518 -0.035 0>, rot=<0 0 0.0238 -0.9997>}
 spine_03 {parent=spine_02, loc=<6.0104 0 0>, rot=<0 0 0.0149 -0.9999>}
 spine_04 {parent=spine_03, loc=<8.1024 0 0>, rot=<0 0 -0.0132 -0.9999>}
 spine_05 {parent=spine_04, loc=<10.0146 0 0>, rot=<0 0 -0.0015 -1>}
 neck_01 {parent=spine_05, loc=<11.3138 0.1529 0>, rot=<0 0 -0.2003 -0.9797>}
 neck_02 {parent=neck_01, loc=<4.5627 0 0>, rot=<0 0 0.0504 -0.9987>}
 head {parent=neck_02, loc=<5.3914 0 0>, rot=<0 0 0.115 -0.9934>}
 attach_cape {parent=spine_05, loc=<9.017 -11.714 0>, rot=<0.0616 -0.7044 -0.0616 -0.7044>}
 dyn_cape_mid_l_MSTR {parent=attach_cape, loc=<7.5797 1.4158 0.5324>, rot=<-0.4896 -0.4896 0.5102 -0.5102>}
 dyn_cape_mid_l_01 {parent=dyn_cape_mid_l_MSTR, loc=<2.1125 0 0>, rot=<0 0 0 1>}
 dyn_cape_mid_l_02 {parent=dyn_cape_mid_l_01, loc=<13.1286 0 0>, rot=<0 -0.0189 0 -0.9998>}
 dyn_cape_mid_l_03 {parent=dyn_cape_mid_l_02, loc=<26.1011 0 0>, rot=<0 -0.0195 0 -0.9998>}
 dyn_cape_mid_l_04 {parent=dyn_cape_mid_l_03, loc=<30.7601 0 0>, rot=<0 -0.0047 0 -1>}
 dyn_cape_mid_l_05 {parent=dyn_cape_mid_l_04, loc=<26.8136 0 0>, rot=<0 -0.0145 0 -0.9999>}
 dyn_cape_mid_l_06 {parent=dyn_cape_mid_l_05, loc=<16.0111 0 0>, rot=<0 0 0 1>}
 dyn_cape_l_MSTR {parent=attach_cape, loc=<22.9844 3.7946 0.792>, rot=<-0.4814 -0.4814 0.5179 -0.5179>}
 dyn_cape_l_01 {parent=dyn_cape_l_MSTR, loc=<2.1577 0 0>, rot=<0 0 0 1>}
 dyn_cape_l_02 {parent=dyn_cape_l_01, loc=<13.1923 0 0>, rot=<0 -0.0346 0 -0.9994>}
 dyn_cape_l_03 {parent=dyn_cape_l_02, loc=<26.2131 0 0>, rot=<0 -0.0173 0 -0.9999>}
 dyn_cape_l_04 {parent=dyn_cape_l_03, loc=<30.8468 0 0>, rot=<0 -0.001 0 -1>}
 dyn_cape_l_05 {parent=dyn_cape_l_04, loc=<26.8826 0 0>, rot=<0 -0.0055 0 -1>}
 dyn_cape_l_06 {parent=dyn_cape_l_05, loc=<16.0708 0 0>, rot=<0 0 0 1>}
 dyn_cape_r_MSTR {parent=attach_cape, loc=<-22.9844 3.7946 0.792>, rot=<-0.4814 -0.4814 0.5179 -0.5179>}
 dyn_cape_r_01 {parent=dyn_cape_r_MSTR, loc=<2.1577 0 0>, rot=<0 0 0 1>}
 dyn_cape_r_02 {parent=dyn_cape_r_01, loc=<13.1923 0 0>, rot=<0 -0.0346 0 -0.9994>}
 dyn_cape_r_03 {parent=dyn_cape_r_02, loc=<26.2131 0 0>, rot=<0 -0.0173 0 -0.9999>}
 dyn_cape_r_04 {parent=dyn_cape_r_03, loc=<30.8468 0 0>, rot=<0 -0.001 0 -1>}
 dyn_cape_r_05 {parent=dyn_cape_r_04, loc=<26.8826 0 0>, rot=<0 -0.0055 0 -1>}
 dyn_cape_r_06 {parent=dyn_cape_r_05, loc=<16.0708 0 0>, rot=<0 0 0 1>}
 cape {parent=attach_cape, type=Null, loc=<0 0 0>, rot=<0 0 0 1>}
 dyn_cape_mid_r_MSTR {parent=attach_cape, loc=<-7.462 1.4158 0.5324>, rot=<-0.4896 -0.4896 0.5102 -0.5102>}
 dyn_cape_mid_r_01 {parent=dyn_cape_mid_r_MSTR, loc=<2.1125 0 0>, rot=<0 0 0 1>}
 dyn_cape_mid_r_02 {parent=dyn_cape_mid_r_01, loc=<13.1286 0 0>, rot=<0 -0.0189 0 -0.9998>}
 dyn_cape_mid_r_03 {parent=dyn_cape_mid_r_02, loc=<26.1011 0 0>, rot=<0 -0.0292 0 -0.9996>}
 dyn_cape_mid_r_04 {parent=dyn_cape_mid_r_03, loc=<30.731 0 0>, rot=<0 0.0049 0 -1>}
 dyn_cape_mid_r_05 {parent=dyn_cape_mid_r_04, loc=<26.8136 0 0>, rot=<0 -0.0145 0 -0.9999>}
 dyn_cape_mid_r_06 {parent=dyn_cape_mid_r_05, loc=<16.0111 0 0>, rot=<0 0 0 1>}
 attach_backpack {parent=spine_05, loc=<5.0475 -23.424 1.5655>, rot=<-0.011 -0.707 0.0112 -0.707>}
 Backpack_BR {parent=attach_backpack, type=Null, loc=<0 -4.2317 -0.2709>, rot=<-0.123 0 0 0.9924>}
 clavicle_l {parent=spine_05, loc=<4.3575 -0.1466 -2.3223>, rot=<0.0465 -0.7289 -0.009 -0.6829>}
 upperarm_l {parent=clavicle_l, loc=<15.5111 0 0>, rot=<0.0226 -0.3234 -0.0088 -0.9459>}
 lowerarm_l {parent=upperarm_l, loc=<26.6514 0 0>, rot=<0.0391 -0.0099 -0.2453 -0.9686>}
 hand_l {parent=lowerarm_l, loc=<22.3491 0 0>, rot=<-0.7081 -0.0507 0.0975 -0.6975>}
 thumb_01_l {parent=hand_l, loc=<1.3118 -0.3191 -2.459>, rot=<0.6546 -0.4042 0.0664 -0.6354>}
 thumb_02_l {parent=thumb_01_l, loc=<5.9336 -0.0448 0.0401>, rot=<-0.0037 0.0213 0.298 -0.9543>}
 thumb_03_l {parent=thumb_02_l, loc=<2.2223 0.3682 -0.0907>, rot=<0.0118 -0.013 0.0066 -0.9998>}
 index_metacarpal_l {parent=hand_l, loc=<1.6535 0.0183 -2.96>, rot=<0.0158 -0.0124 -0.0173 -0.9996>}
 index_01_l {parent=index_metacarpal_l, loc=<8.3008 0.0357 -0.0524>, rot=<-0.0044 0.0485 0.0924 -0.9945>}
 index_02_l {parent=index_01_l, loc=<3.3587 -0.3004 0.0452>, rot=<-0.002 0.0027 0.3225 -0.9466>}
 index_03_l {parent=index_02_l, loc=<2.5578 -0.123 -0.0252>, rot=<-0.002 0.0067 0.2558 -0.9667>}
 middle_metacarpal_l {parent=hand_l, loc=<2.1856 -0.1725 -0.5984>, rot=<0.006 0.0168 -0.0329 -0.9993>}
 middle_01_l {parent=middle_metacarpal_l, loc=<7.3275 -0.2447 0.0129>, rot=<-0.0208 0.0042 0.0858 -0.9961>}
 middle_02_l {parent=middle_01_l, loc=<4.0123 0.1401 0.1175>, rot=<-0.0218 0.0258 0.3512 -0.9357>}
 middle_03_l {parent=middle_02_l, loc=<2.6077 0.0534 0.1373>, rot=<0.0082 0.0188 0.1708 -0.9851>}
 deform_middle_01_l {parent=middle_metacarpal_l, loc=<7.3571 -0.0483 0.9391>, rot=<-0.021 0.0029 0.1452 -0.9892>}
 deform_middle_02_l {parent=deform_middle_01_l, loc=<4.0123 0.1401 0.1175>, rot=<-0.0215 0.026 0.341 -0.9394>}
 deform_middle_03_l {parent=deform_middle_02_l, loc=<2.8714 -0.0186 0.148>, rot=<0.0089 0.0185 0.1329 -0.9909>}
 ring_metacarpal_l {parent=hand_l, loc=<2.2074 -0.3625 1.3638>, rot=<-0.0284 0.0722 -0.0458 -0.9959>}
 ring_01_l {parent=ring_metacarpal_l, loc=<7.0295 -0.6197 -0.1446>, rot=<-0.0489 -0.0252 0.1892 -0.9804>}
 ring_02_l {parent=ring_01_l, loc=<3.6174 0.4647 -0.1027>, rot=<0 0 0.2806 -0.9598>}
 ring_03_l {parent=ring_02_l, loc=<2.4152 0.114 0.0199>, rot=<0 0 0.1189 -0.9929>}
 pinky_metacarpal_l {parent=hand_l, loc=<1.7616 -0.9753 3.3814>, rot=<-0.0397 0.0814 -0.069 -0.9935>}
 pinky_01_l {parent=pinky_metacarpal_l, loc=<7.1709 -0.4831 -0.1094>, rot=<-0.0654 -0.0035 0.2242 -0.9723>}
 pinky_02_l {parent=pinky_01_l, loc=<2.8274 0.0893 -0.1043>, rot=<0.0037 0.0011 0.2854 -0.9584>}
 pinky_03_l {parent=pinky_02_l, loc=<2.2046 0.0307 -0.0078>, rot=<-0.0002 0.0098 0.1445 -0.9895>}
 deform_wrist_dn_base_l {parent=hand_l, loc=<0 -2.2342 0>, rot=<0 0 0 -1>}
 deform_wrist_dn_l {parent=deform_wrist_dn_base_l, loc=<0 0 0>, rot=<0 0 0 1>}
 deform_wrist_up_base_l {parent=hand_l, loc=<0 2.2682 0>, rot=<0 0 0 -1>}
 deform_wrist_up_l {parent=deform_wrist_up_base_l, loc=<0 0 0>, rot=<0 0 0 1>}
 lowerarm_twist_02_l {parent=lowerarm_l, loc=<9.6199 0 0>, rot=<0 0 0 -1>}
 lowerarm_twist_01_l {parent=lowerarm_l, loc=<17.8549 0 0>, rot=<0 0 0 -1>}
 deform_elbow_in_base_l {parent=lowerarm_l, loc=<0.2112 5.259 0.8274>, rot=<0 0 0 -1>}
 deform_elbow_in_l {parent=deform_elbow_in_base_l, loc=<0 0 0>, rot=<0 0 0 1>}
 deform_elbow_out_base_l {parent=lowerarm_l, loc=<0 -4 0>, rot=<0 0 0 -1>}
 deform_elbow_out_l {parent=deform_elbow_out_base_l, loc=<0 0 0>, rot=<0 0 0 1>}
 upperarm_twist_01_l {parent=upperarm_l, loc=<0.0041 0 0>, rot=<0 0 0 -1>}
 deform_upperarm_bk_base_l {parent=upperarm_twist_01_l, loc=<1.0813 -7.2193 -1.9453>, rot=<0 0 0 1>}
 deform_upperarm_bk_l {parent=deform_upperarm_bk_base_l, loc=<0 0 0>, rot=<0 0 0 1>}
 deform_upperarm_fr_base_l {parent=upperarm_twist_01_l, loc=<-0.3691 7.0506 -2.765>, rot=<0 0 0 1>}
 deform_upperarm_fr_l {parent=deform_upperarm_fr_base_l, loc=<0 0 0>, rot=<0 0 0 1>}
 deform_upperarm_up_base_l {parent=upperarm_twist_01_l, loc=<-3.2216 -1.1391 3.3311>, rot=<0 0 0 1>}
 deform_upperarm_up_l {parent=deform_upperarm_up_base_l, loc=<0 0 0>, rot=<0 0 0 1>}
 upperarm_twist_02_l {parent=upperarm_l, loc=<13.7941 0 0>, rot=<0 0 0 -1>}
 clavicle_r {parent=spine_05, loc=<4.3575 -0.1466 2.3223>, rot=<0.7289 0.0465 -0.6829 0.009>}
 upperarm_r {parent=clavicle_r, loc=<-15.5111 0 0>, rot=<0.0226 -0.3234 -0.0088 -0.9459>}
 lowerarm_r {parent=upperarm_r, loc=<-26.6514 0 0>, rot=<0.0391 -0.0099 -0.2453 -0.9686>}
 hand_r {parent=lowerarm_r, loc=<-22.3491 0 0>, rot=<-0.7081 -0.0507 0.0975 -0.6975>}
 thumb_01_r {parent=hand_r, loc=<-1.3118 0.3191 2.459>, rot=<0.6546 -0.4042 0.0664 -0.6354>}
 thumb_02_r {parent=thumb_01_r, loc=<-5.9336 0.0448 -0.0401>, rot=<-0.0037 0.0213 0.298 -0.9543>}
 thumb_03_r {parent=thumb_02_r, loc=<-2.2223 -0.3682 0.0907>, rot=<0.0118 -0.013 0.0066 -0.9998>}
 index_metacarpal_r {parent=hand_r, loc=<-1.6535 -0.0183 2.96>, rot=<0.0158 -0.0124 -0.0173 -0.9996>}
 index_01_r {parent=index_metacarpal_r, loc=<-8.3008 -0.0357 0.0524>, rot=<-0.0044 0.0485 0.0924 -0.9945>}
 index_02_r {parent=index_01_r, loc=<-3.3587 0.3004 -0.0452>, rot=<-0.002 0.0027 0.3225 -0.9466>}
 index_03_r {parent=index_02_r, loc=<-2.5578 0.123 0.0252>, rot=<-0.002 0.0067 0.2558 -0.9667>}
 middle_metacarpal_r {parent=hand_r, loc=<-2.1856 0.1725 0.5984>, rot=<0.006 0.0168 -0.0329 -0.9993>}
 middle_01_r {parent=middle_metacarpal_r, loc=<-7.3275 0.2447 -0.0129>, rot=<-0.0208 0.0042 0.0858 -0.9961>}
 middle_02_r {parent=middle_01_r, loc=<-4.0123 -0.1401 -0.1175>, rot=<-0.0218 0.0258 0.3512 -0.9357>}
 middle_03_r {parent=middle_02_r, loc=<-2.6077 -0.0534 -0.1373>, rot=<0.0082 0.0188 0.1708 -0.9851>}
 deform_middle_01_r {parent=middle_metacarpal_r, loc=<-7.3572 0.0484 -0.9391>, rot=<-0.021 0.0029 0.1452 -0.9892>}
 deform_middle_02_r {parent=deform_middle_01_r, loc=<-4.0123 -0.1401 -0.1175>, rot=<-0.0215 0.026 0.341 -0.9394>}
 deform_middle_03_r {parent=deform_middle_02_r, loc=<-2.8713 0.0186 -0.148>, rot=<0.0089 0.0185 0.1329 -0.9909>}
 ring_metacarpal_r {parent=hand_r, loc=<-2.2074 0.3625 -1.3638>, rot=<-0.0284 0.0722 -0.0458 -0.9959>}
 ring_01_r {parent=ring_metacarpal_r, loc=<-7.0295 0.6197 0.1446>, rot=<-0.0489 -0.0252 0.1892 -0.9804>}
 ring_02_r {parent=ring_01_r, loc=<-3.6174 -0.4647 0.1027>, rot=<0 0 0.2806 -0.9598>}
 ring_03_r {parent=ring_02_r, loc=<-2.4152 -0.114 -0.0199>, rot=<0 0 0.1189 -0.9929>}
 pinky_metacarpal_r {parent=hand_r, loc=<-1.7616 0.9753 -3.3814>, rot=<-0.0397 0.0814 -0.069 -0.9935>}
 pinky_01_r {parent=pinky_metacarpal_r, loc=<-7.1709 0.4831 0.1094>, rot=<-0.0654 -0.0035 0.2242 -0.9723>}
 pinky_02_r {parent=pinky_01_r, loc=<-2.8274 -0.0893 0.1043>, rot=<0.0037 0.0011 0.2854 -0.9584>}
 pinky_03_r {parent=pinky_02_r, loc=<-2.2046 -0.0307 0.0078>, rot=<-0.0002 0.0098 0.1445 -0.9895>}
 deform_wrist_dn_base_r {parent=hand_r, loc=<-0.0004 2.2345 -0.0001>, rot=<0 0 0 -1>}
 deform_wrist_dn_r {parent=deform_wrist_dn_base_r, loc=<0 0 0>, rot=<0 0 0 1>}
 deform_wrist_up_base_r {parent=hand_r, loc=<0.0003 -2.2684 0.0001>, rot=<0 0 0 -1>}
 deform_wrist_up_r {parent=deform_wrist_up_base_r, loc=<0 0 0>, rot=<0 0 0 1>}
 lowerarm_twist_02_r {parent=lowerarm_r, loc=<-9.6199 0 0>, rot=<0 0 0 -1>}
 lowerarm_twist_01_r {parent=lowerarm_r, loc=<-17.8549 0 0>, rot=<0 0 0 -1>}
 deform_elbow_in_base_r {parent=lowerarm_r, loc=<-0.2111 -5.2591 -0.8275>, rot=<0 0 0 -1>}
 deform_elbow_in_r {parent=deform_elbow_in_base_r, loc=<-0.0001 0 0.0002>, rot=<0 0 0 1>}
 deform_elbow_out_base_r {parent=lowerarm_r, loc=<0 4 -0.0001>, rot=<0 0 0 -1>}
 deform_elbow_out_r {parent=deform_elbow_out_base_r, loc=<0 0 0>, rot=<0 0 0 1>}
 upperarm_twist_01_r {parent=upperarm_r, loc=<-0.0041 0 0>, rot=<0 0 0 -1>}
 deform_upperarm_bk_base_r {parent=upperarm_twist_01_r, loc=<-1.0814 7.2192 1.9455>, rot=<0 0 0 1>}
 deform_upperarm_bk_r {parent=deform_upperarm_bk_base_r, loc=<0 0 0>, rot=<0 0 0 1>}
 deform_upperarm_fr_base_r {parent=upperarm_twist_01_r, loc=<0.3689 -7.0506 2.7652>, rot=<0 0 0 1>}
 deform_upperarm_fr_r {parent=deform_upperarm_fr_base_r, loc=<0 0 0>, rot=<0 0 0 1>}
 deform_upperarm_up_base_r {parent=upperarm_twist_01_r, loc=<3.2214 1.1391 -3.3308>, rot=<0 0 0 1>}
 deform_upperarm_up_r {parent=deform_upperarm_up_base_r, loc=<0 0 0>, rot=<0 0 0 -1>}
 upperarm_twist_02_r {parent=upperarm_r, loc=<-13.7941 0 0>, rot=<0 0 0 -1>}
 deform_back_MSTR {parent=spine_05, loc=<2.2599 -11.3533 0>, rot=<0.5196 -0.4795 0.4795 -0.5196>}
 deform_back {parent=deform_back_MSTR, loc=<0 0 0>, rot=<0 0 0 1>}
 dyn_hood_bk_drape_MSTR {parent=spine_05, loc=<9.2443 -12.8204 -0.0356>, rot=<0.086 -0.704 0.6997 -0.0862>}
 dyn_hood_bk_drape_01 {parent=dyn_hood_bk_drape_MSTR, loc=<0 0 0>, rot=<0 0 0 1>}
 dyn_hood_bk_drape_02 {parent=dyn_hood_bk_drape_01, loc=<10.3621 0 0>, rot=<0 -0.0321 0 -0.9995>}
 dyn_hood_bk_drape_end {parent=dyn_hood_bk_drape_02, loc=<10.5656 0 0>, rot=<0 -0.0181 0 -0.9998>}
 deform_pec_aim_base_l {parent=spine_04, loc=<7.8732 12.6388 -2.4123>, rot=<0.0102 -0.4602 0.3979 -0.7936>}
 deform_pec_aim_l {parent=deform_pec_aim_base_l, loc=<0 0 0>, rot=<0 0 0 1>}
 deform_pec_base_l {parent=deform_pec_aim_l, loc=<10.766 0 0>, rot=<0 0 0 1>}
 deform_pec_l {parent=deform_pec_base_l, loc=<0.055 0 0>, rot=<0 0 0 1>}
 deform_pec_aim_base_r {parent=spine_04, loc=<7.8727 12.6388 2.4123>, rot=<-0.4602 -0.0102 0.7936 0.3979>}
 deform_pec_aim_r {parent=deform_pec_aim_base_r, loc=<0 0 0>, rot=<0 0 0 1>}
 deform_pec_base_r {parent=deform_pec_aim_r, loc=<-10.7662 -0.0006 -0.0007>, rot=<0 0 0 1>}
 deform_pec_r {parent=deform_pec_base_r, loc=<0 0 0>, rot=<0 0 0 1>}
 thigh_l {parent=pelvis, loc=<-2.5693 -0.0109 -7.7501>, rot=<0.0012 0.0572 0.0155 -0.9982>}
 calf_l {parent=thigh_l, loc=<-39.0484 0 0>, rot=<0.0006 -0.0161 -0.0391 -0.9991>}
 foot_l {parent=calf_l, loc=<-42.561 0.1077 0.4292>, rot=<0.0418 -0.0455 0.0343 -0.9975>}
 ball_l {parent=foot_l, loc=<-6.1599 13.8264 0.6806>, rot=<0.0001 -0.0001 -0.7063 -0.7079>}
 ball_l_ctrl {controlType=EULER_TRANSFORM, parent=ball_l, type=Control, loc=<0 0 0>, rot=<0 0 0 1>}
 deform_toe_l_outer {parent=foot_l, loc=<-7.0967 7.4935 -4.9834>, rot=<0.1492 -0.1495 -0.6904 -0.692>}
 deform_toe_l_mid {parent=foot_l, loc=<-7.0963 9.0184 0.3426>, rot=<-0.0248 0.0249 -0.7058 -0.7075>}
 deform_toe_l_inner {parent=foot_l, loc=<-7.0962 6.2076 6.101>, rot=<-0.2049 0.2053 -0.6759 -0.6775>}
 calf_twist_02_l {parent=calf_l, loc=<-18.8064 -0.0967 0.0397>, rot=<0.0237 0.0095 -0.0035 -0.9997>}
 calf_twist_01_l {parent=calf_l, loc=<-35.813 -0.0967 0.0397>, rot=<0.0237 0.0095 -0.0035 -0.9997>}
 deform_knee_in_base_l {parent=calf_l, loc=<1.298 -5.2957 -0.0496>, rot=<0 0 0 -1>}
 deform_knee_in_l {parent=deform_knee_in_base_l, loc=<0 0 0>, rot=<0 0 0 1>}
 deform_knee_out_base_l {parent=calf_l, loc=<0 4 0>, rot=<0 0 0 -1>}
 deform_knee_out_l {parent=deform_knee_out_base_l, loc=<0 0 0>, rot=<0 0 0 1>}
 thigh_twist_01_l {parent=thigh_l, loc=<-21.6303 0 0>, rot=<0.0013 0 0 -1>}
 deform_groin_base_l {parent=thigh_l, loc=<-2.1908 8.1142 -1.7442>, rot=<0 0 0 -1>}
 deform_groin_l {parent=deform_groin_base_l, loc=<0 0 0>, rot=<0 0 0 1>}
 deform_hip_base_l {parent=thigh_l, loc=<3.3614 0.2865 -10.12>, rot=<0 0 0 -1>}
 deform_hip_l {parent=deform_hip_base_l, loc=<0 0 0>, rot=<0 0 0 1>}
 thigh_r {parent=pelvis, loc=<-2.5693 -0.0109 7.7501>, rot=<0.0572 -0.0012 0.9982 0.0155>}
 calf_r {parent=thigh_r, loc=<39.0484 0 0>, rot=<0.0006 -0.0161 -0.0391 -0.9991>}
 foot_r {parent=calf_r, loc=<42.561 -0.1077 -0.4292>, rot=<0.0418 -0.0455 0.0343 -0.9975>}
 ball_r {parent=foot_r, loc=<6.1599 -13.8264 -0.6806>, rot=<0.0001 -0.0001 -0.7063 -0.7079>}
 ball_r_ctrl {controlType=EULER_TRANSFORM, parent=ball_r, type=Control, loc=<0 0 0>, rot=<0 0 0 1>}
 deform_toe_r_outer {parent=foot_r, loc=<7.0967 -7.4935 4.9834>, rot=<0.1492 -0.1495 -0.6904 -0.692>}
 deform_toe_r_mid {parent=foot_r, loc=<7.0963 -9.0184 -0.3427>, rot=<-0.0248 0.0249 -0.7058 -0.7075>}
 deform_toe_r_inner {parent=foot_r, loc=<7.0962 -6.2076 -6.101>, rot=<-0.2049 0.2053 -0.6759 -0.6775>}
 calf_twist_02_r {parent=calf_r, loc=<18.8064 0.0967 -0.0397>, rot=<0.0237 0.0095 -0.0035 -0.9997>}
 calf_twist_01_r {parent=calf_r, loc=<35.813 0.0967 -0.0397>, rot=<0.0237 0.0095 -0.0035 -0.9997>}
 deform_knee_in_base_r {parent=calf_r, loc=<-1.298 5.2957 0.0496>, rot=<0 0 0 -1>}
 deform_knee_in_r {parent=deform_knee_in_base_r, loc=<0 0 0>, rot=<0 0 0 1>}
 deform_knee_out_base_r {parent=calf_r, loc=<0 -4 0>, rot=<0 0 0 -1>}
 deform_knee_out_r {parent=deform_knee_out_base_r, loc=<0 0 0>, rot=<0 0 0 1>}
 thigh_twist_01_r {parent=thigh_r, loc=<21.6303 0 0>, rot=<0.0013 0 0 -1>}
 deform_groin_base_r {parent=thigh_r, loc=<2.1908 -8.1142 1.7442>, rot=<0 0 0 -1>}
 deform_groin_r {parent=deform_groin_base_r, loc=<0 0 0>, rot=<0 0 0 1>}
 deform_hip_base_r {parent=thigh_r, loc=<-3.3613 -0.2865 10.12>, rot=<0 0 0 -1>}
 deform_hip_r {parent=deform_hip_base_r, loc=<0 0 0>, rot=<0 0 0 1>}
 deform_glute_base_l {parent=pelvis, loc=<-8.8064 -8.8222 -8.4287>, rot=<0.0275 0.0501 -0.4472 -0.8926>}
 deform_glute_l {parent=deform_glute_base_l, loc=<0 0 0>, rot=<0 0 0 1>}
 deform_glute_base_r {parent=pelvis, loc=<-8.8062 -8.8222 8.4287>, rot=<0.0501 -0.0275 0.8926 -0.4472>}
 deform_glute_r {parent=deform_glute_base_r, loc=<0 0 0>, rot=<0 0 0 1>}
 dyn_loincloth_fr_l_MSTR {parent=pelvis, loc=<10.6507 10.1399 -3.9221>, rot=<-0.0127 -0.707 0.0127 -0.707>}
 dyn_loincloth_fr_l_01 {parent=dyn_loincloth_fr_l_MSTR, loc=<0 0 0>, rot=<0.5 -0.5 -0.5 -0.5>}
 dyn_loincloth_fr_l_02 {parent=dyn_loincloth_fr_l_01, loc=<17.5221 0 0>, rot=<0 0 0 1>}
 dyn_loincloth_fr_l_03 {parent=dyn_loincloth_fr_l_02, loc=<17.6421 0 0>, rot=<0 0 0 1>}
 dyn_loincloth_bk_l_MSTR {parent=pelvis, loc=<5.5007 -14.5778 -4.0882>, rot=<0.0102 -0.7372 0.6748 0.0307>}
 dyn_loincloth_bk_l_01 {parent=dyn_loincloth_bk_l_MSTR, loc=<0 0 0>, rot=<0 0 0 1>}
 dyn_loincloth_bk_l_02 {parent=dyn_loincloth_bk_l_01, loc=<8.3878 0 0>, rot=<-0.0013 -0.0526 -0.021 -0.9984>}
 dyn_loincloth_bk_l_03 {parent=dyn_loincloth_bk_l_02, loc=<8.1727 0 0>, rot=<0 0 0 1>}
 dyn_loincloth_side_fr_l_MSTR {parent=pelvis, loc=<11.9717 4.3659 -12.9073>, rot=<-0.0068 -0.989 0.044 -0.1408>}
 dyn_loincloth_side_fr_l_01 {parent=dyn_loincloth_side_fr_l_MSTR, loc=<0 0 0>, rot=<0 0 0 1>}
 dyn_loincloth_side_fr_l_02 {parent=dyn_loincloth_side_fr_l_01, loc=<12.3464 0 0>, rot=<0.0005 -0.0486 -0.0101 -0.9988>}
 dyn_loincloth_side_fr_l_03 {parent=dyn_loincloth_side_fr_l_02, loc=<10.4238 0 0>, rot=<0 0 0 1>}
 dyn_loincloth_side_bk_l_MSTR {parent=pelvis, loc=<10.6 -7.4594 -12.873>, rot=<0.008 -0.9863 0.0628 -0.1524>}
 dyn_loincloth_side_bk_l_01 {parent=dyn_loincloth_side_bk_l_MSTR, loc=<0 0 0>, rot=<0 0 0 1>}
 dyn_loincloth_side_bk_l_02 {parent=dyn_loincloth_side_bk_l_01, loc=<11.3994 0 0>, rot=<0.0024 -0.0947 -0.0253 -0.9952>}
 dyn_loincloth_side_bk_l_03 {parent=dyn_loincloth_side_bk_l_02, loc=<10.7491 0 0>, rot=<0 0 0 1>}
 dyn_loincloth_fr_r_MSTR {parent=pelvis, loc=<10.6509 10.1399 3.922>, rot=<0.707 -0.0127 -0.707 -0.0127>}
 dyn_loincloth_fr_r_01 {parent=dyn_loincloth_fr_r_MSTR, loc=<0 0 0>, rot=<0.5 -0.5 -0.5 -0.5>}
 dyn_loincloth_fr_r_02 {parent=dyn_loincloth_fr_r_01, loc=<-17.5224 0 0>, rot=<0 0 0 1>}
 dyn_loincloth_fr_r_03 {parent=dyn_loincloth_fr_r_02, loc=<-17.6421 0 0>, rot=<0 0 0 1>}
 dyn_loincloth_bk_r_MSTR {parent=pelvis, loc=<5.5007 -14.5778 4.0882>, rot=<0.7372 0.0102 0.0307 -0.6748>}
 dyn_loincloth_bk_r_01 {parent=dyn_loincloth_bk_r_MSTR, loc=<0 0 0>, rot=<0 0 0 1>}
 dyn_loincloth_bk_r_02 {parent=dyn_loincloth_bk_r_01, loc=<-8.3879 0 0>, rot=<-0.0013 -0.0526 -0.021 -0.9984>}
 dyn_loincloth_bk_r_03 {parent=dyn_loincloth_bk_r_02, loc=<-8.1728 0 0>, rot=<0 0 0 1>}
 dyn_loincloth_side_fr_r_MSTR {parent=pelvis, loc=<11.9719 4.3659 12.9073>, rot=<0.989 -0.0068 -0.1408 -0.044>}
 dyn_loincloth_side_fr_r_01 {parent=dyn_loincloth_side_fr_r_MSTR, loc=<0 0 0>, rot=<0 0 0 1>}
 dyn_loincloth_side_fr_r_02 {parent=dyn_loincloth_side_fr_r_01, loc=<-12.3467 0 0.0001>, rot=<0.0005 -0.0486 -0.0101 -0.9988>}
 dyn_loincloth_side_fr_r_03 {parent=dyn_loincloth_side_fr_r_02, loc=<-10.4238 0 0>, rot=<0 0 0 1>}
 dyn_loincloth_side_bk_r_MSTR {parent=pelvis, loc=<10.6005 -7.4594 12.873>, rot=<0.9795 -0.0027 -0.1526 -0.1314>}
 dyn_loincloth_side_bk_r_01 {parent=dyn_loincloth_side_bk_r_MSTR, loc=<0 0 0>, rot=<0 0 0 1>}
 dyn_loincloth_side_bk_r_02 {parent=dyn_loincloth_side_bk_r_01, loc=<-11.3998 0 0.0001>, rot=<0.0011 -0.0973 -0.0117 -0.9952>}
 dyn_loincloth_side_bk_r_03 {parent=dyn_loincloth_side_bk_r_02, loc=<-10.7492 0 0>, rot=<0 0 0 1>}
 cape_aim {parent=pelvis, loc=<-0.5593 -21.824 21.6265>, rot=<-0.0689 0.0216 -0.0155 0.9973>}
 dyn_cape_ground_collision {parent=root, loc=<0 0 0>, rot=<0 0 0 1>}
]
variables [
 Disable Floor Slope {type=bool, default=False}
 ItemToModify_0_Transform_Translation_Z {type=double, default=1}
 Push Down Weight {type=double, default=0}
 Source Middle Fingers {
  type="TArray<FRigElementKey>"
  default=((Type=Bone,Name="middle_01_l"),(Type=Bone,Name="middle_02_l"),(Type=Bone,Name="middle_03_l"),(Type=Bone,Name="middle_01_r"),(Type=Bone,Name="middle_02_r"),(Type=Bone,Name="middle_03_r"))
 }
 Target Middle Fingers {
  type="TArray<FRigElementKey>"
  default=((Type=Bone,Name="deform_middle_01_l"),(Type=Bone,Name="deform_middle_02_l"),(Type=Bone,Name="deform_middle_03_l"),(Type=Bone,Name="deform_middle_01_r"),(Type=Bone,Name="deform_middle_02_r"),(Type=Bone,Name="deform_middle_03_r"))
 }
]
