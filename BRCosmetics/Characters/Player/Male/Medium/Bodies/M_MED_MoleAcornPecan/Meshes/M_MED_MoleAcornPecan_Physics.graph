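# Physics asset for the M_MED_MoleAcornPecan male player body cosmetic, set up for the motion of its segmented loincloth.
# Eight three-link capsule chains (front, back, side-front, side-back, left and right) hang from MSTR bodies with tight swing and twist limits.
# Angular-free cross constraints tie front left to right at every level and each side's front to back at the middle link.
# Pelvis, thigh and calf capsules plus a large ground box provide collision shapes.
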PhysicsAsset /BRCosmetics/Characters/Player/Male/Medium/Bodies/M_MED_MoleAcornPecan/Meshes/M_MED_MoleAcornPecan_Physics
bodies {
 calf_l [{shapeType=Capsule, length=41.895, radius=9.7752, shapeName=calf_l_capsule, center=<-21.6964 -2.0859 0>, rotation=<rot 90 0 0>}]
 calf_r [{
   shapeType=Capsule
   length=41.895
   radius=9.7752
   shapeName=calf_r_capsule
   center=<21.436 1.9886 0>
   rotation=<rot 87.9278 179.9999 179.9999>
  }]
 dyn_cape_ground_collision [{shapeType=Box, extentX=318.8156, extentY=318.8156, extentZ=283.5656, shapeName=dyn_skirt_ground_collision_box, center=<0 0 -140>}]
 dyn_loincloth_bk_l_01 [{shapeType=Capsule, length=12, radius=6, shapeName=dyn_loincloth_bk_l_01_capsule, center=<3.7593 0 4.8>, rotation=<rot 90 0 0>}]
 dyn_loincloth_bk_l_02 [{shapeType=Capsule, length=12, radius=6, shapeName=dyn_loincloth_bk_l_01_capsule, center=<3.7593 0 4.8>, rotation=<rot 90 0 0>}]
 dyn_loincloth_bk_l_03 [{shapeType=Capsule, length=12, radius=6, shapeName=dyn_loincloth_bk_l_01_capsule, center=<3.7593 0 4.8>, rotation=<rot 90 0 0>}]
 dyn_loincloth_bk_l_MSTR [{shapeType=Capsule, length=1, radius=1, shapeName=dyn_loincloth_bk_l_MSTR_capsule}]
 dyn_loincloth_bk_r_01 [{shapeType=Capsule, length=12, radius=6, shapeName=dyn_loincloth_bk_r_01_capsule, center=<-3.6255 0 -4.8>, rotation=<rot 90 0 0>}]
 dyn_loincloth_bk_r_02 [{shapeType=Capsule, length=12, radius=6, shapeName=dyn_loincloth_bk_r_01_capsule, center=<-3.6255 0 -4.8>, rotation=<rot 90 0 0>}]
 dyn_loincloth_bk_r_03 [{shapeType=Capsule, length=12, radius=6, shapeName=dyn_loincloth_bk_r_01_capsule, center=<-3.6255 0 -4.8>, rotation=<rot 90 0 0>}]
 dyn_loincloth_bk_r_MSTR [{shapeType=Capsule, length=1, radius=1, shapeName=dyn_loincloth_bk_r_MSTR_capsule}]
 dyn_loincloth_fr_l_01 [{shapeType=Capsule, length=9.2621, radius=8.1278, shapeName=dyn_loincloth_fr_l_01_capsule, center=<12.4768 -2 12>, rotation=<rot 90 0 0>}]
 dyn_loincloth_fr_l_02 [{shapeType=Capsule, length=10.8243, radius=7.0145, shapeName=dyn_loincloth_fr_l_02_capsule, center=<9.2198 -2 6.6>, rotation=<rot 90 0 0>}]
 dyn_loincloth_fr_l_03 [{shapeType=Capsule, length=12.2433, radius=7.0145, shapeName=dyn_loincloth_fr_l_03_capsule, center=<7.4281 -2 6.6>, rotation=<rot 90 0 0>}]
 dyn_loincloth_fr_l_MSTR [{shapeType=Capsule, length=1, radius=1, shapeName=dyn_loincloth_fr_l_MSTR_capsule, rotation=<rot 0 -90 0>}]
 dyn_loincloth_fr_r_01 [{
   shapeType=Capsule
   length=9.2621
   radius=8.1278
   shapeName=dyn_loincloth_fr_r_01_capsule
   center=<-12.4768 2 -12>
   rotation=<rot -90 180 180>
  }]
 dyn_loincloth_fr_r_02 [{shapeType=Capsule, length=10.8243, radius=7.0145, shapeName=dyn_loincloth_fr_r_02_capsule, center=<-9.2198 2 -6.6>, rotation=<rot -90 0 0>}]
 dyn_loincloth_fr_r_03 [{shapeType=Capsule, length=12.2433, radius=7.0145, shapeName=dyn_loincloth_fr_r_03_capsule, center=<-7.4281 2 -6.6>, rotation=<rot -90 0 0>}]
 dyn_loincloth_fr_r_MSTR [{shapeType=Capsule, length=1, radius=1, shapeName=dyn_loincloth_fr_r_MSTR_capsule, rotation=<rot 0 -90 0>}]
 dyn_loincloth_side_bk_l_01 [{
   shapeType=Capsule
   length=11.8675
   radius=6
   shapeName=dyn_loincloth_side_bk_l_01_capsule
   center=<4.3664 0 7.4>
   rotation=<rot 0 0.0003 69.1437>
  }]
 dyn_loincloth_side_bk_l_02 [{
   shapeType=Capsule
   length=10.6619
   radius=7.2056
   shapeName=dyn_loincloth_side_bk_l_02_capsule
   center=<3.7884 0 7.4>
   rotation=<rot 0 0.0003 69.1437>
  }]
 dyn_loincloth_side_bk_l_03 [{
   shapeType=Capsule
   length=10.0359
   radius=7.9924
   shapeName=dyn_loincloth_side_bk_l_03_capsule
   center=<4.3664 0 7.4>
   rotation=<rot 0 0.0003 69.1437>
  }]
 dyn_loincloth_side_bk_l_MSTR [{shapeType=Capsule, length=1, radius=1, shapeName=dyn_loincloth_side_bk_l_MSTR_capsule}]
 dyn_loincloth_side_bk_r_01 [{
   shapeType=Capsule
   length=12.3324
   radius=6
   shapeName=dyn_loincloth_side_bk_r_01_capsule
   center=<-4.4112 -0.0603 -6.3>
   rotation=<rot -0.0013 0.0045 -108.206>
  }]
 dyn_loincloth_side_bk_r_02 [{
   shapeType=Capsule
   length=12.3324
   radius=6
   shapeName=dyn_loincloth_side_bk_r_02_capsule
   center=<-3.4194 0 -5.6>
   rotation=<rot -0.0012 0.0047 -108.2051>
  }]
 dyn_loincloth_side_bk_r_03 [{
   shapeType=Capsule
   length=11.2417
   radius=7.0907
   shapeName=dyn_loincloth_side_bk_r_03_capsule
   center=<-4.4112 0 -5.6>
   rotation=<rot -0.0012 0.0047 -108.2051>
  }]
 dyn_loincloth_side_bk_r_MSTR [{shapeType=Capsule, length=1, radius=1, shapeName=dyn_loincloth_side_bk_r_MSTR_capsule}]
 dyn_loincloth_side_fr_l_01 [{
   shapeType=Capsule
   length=11.8675
   radius=6
   shapeName=dyn_loincloth_side_fr_l_01_capsule
   center=<5.4546 0 7.4>
   rotation=<rot 0 0.0001 117.7886>
  }]
 dyn_loincloth_side_fr_l_02 [{
   shapeType=Capsule
   length=11.1376
   radius=6.7299
   shapeName=dyn_loincloth_side_fr_l_02_capsule
   center=<4.1037 0 7.4>
   rotation=<rot 0 0.0001 117.7886>
  }]
 dyn_loincloth_side_fr_l_03 [{
   shapeType=Capsule
   length=10.5161
   radius=7.3514
   shapeName=dyn_loincloth_side_fr_l_03_capsule
   center=<5.4546 -0.1113 6.4>
   rotation=<rot 0 0.0001 117.7886>
  }]
 dyn_loincloth_side_fr_l_MSTR [{shapeType=Capsule, length=1, radius=1, shapeName=dyn_loincloth_side_fr_l_MSTR_capsule}]
 dyn_loincloth_side_fr_r_01 [{
   shapeType=Capsule
   length=12.3324
   radius=6
   shapeName=dyn_loincloth_side_fr_r_01_capsule
   center=<-5.4595 -0.0603 -6.3>
   rotation=<rot 0 0.0002 -67.8307>
  }]
 dyn_loincloth_side_fr_r_02 [{
   shapeType=Capsule
   length=12.3324
   radius=6
   shapeName=dyn_loincloth_side_fr_r_02_capsule
   center=<-4.4677 0 -5.6>
   rotation=<rot 0 0.0002 -67.8307>
  }]
 dyn_loincloth_side_fr_r_03 [{
   shapeType=Capsule
   length=11.2417
   radius=7.0907
   shapeName=dyn_loincloth_side_fr_r_03_capsule
   center=<-5.4595 0 -5.6>
   rotation=<rot 0 0.0002 -67.8307>
  }]
 dyn_loincloth_side_fr_r_MSTR [{shapeType=Capsule, length=1, radius=1, shapeName=dyn_loincloth_side_fr_r_MSTR_capsule}]
 pelvis [{shapeType=Capsule, length=6.3293, radius=14.7656, shapeName=pelvis_capsule}]
 thigh_l [{
   shapeType=Capsule
   length=36.4123
   radius=10.2937
   shapeName=thigh_l_capsule
   center=<-17.949 -0.9337 1.1321>
   rotation=<rot 85.5135 0.0001 0.0002>
  }]
 thigh_r [{
   shapeType=Capsule
   length=35.535
   radius=10.2289
   shapeName=thigh_r_capsule
   center=<17.4891 0 -1.7182>
   rotation=<rot 84.4456 0.0001 0.0009>
  }]
}
constraints [
 {
  bone1Name=dyn_loincloth_side_bk_l_01
  bone2Name=dyn_loincloth_side_bk_l_MSTR
  swing1LimitDegrees=2
  swing2LimitDegrees=20
  twistLimitDegrees=4
 }
 {bone1Name=dyn_loincloth_side_bk_l_02, bone2Name=dyn_loincloth_side_bk_l_01, swing1LimitDegrees=2, swing2LimitDegrees=20, twistLimitDegrees=4}
 {bone1Name=dyn_loincloth_side_bk_l_03, bone2Name=dyn_loincloth_side_bk_l_02, swing1LimitDegrees=2, swing2LimitDegrees=20, twistLimitDegrees=4}
 {
  bone1Name=dyn_loincloth_side_bk_r_01
  bone2Name=dyn_loincloth_side_bk_r_MSTR
  swing1LimitDegrees=2
  swing2LimitDegrees=20
  twistLimitDegrees=4
 }
 {bone1Name=dyn_loincloth_side_bk_r_02, bone2Name=dyn_loincloth_side_bk_r_01, swing1LimitDegrees=2, swing2LimitDegrees=20, twistLimitDegrees=4}
 {bone1Name=dyn_loincloth_side_bk_r_03, bone2Name=dyn_loincloth_side_bk_r_02, swing1LimitDegrees=2, swing2LimitDegrees=20, twistLimitDegrees=4}
 {
  bone1Name=dyn_loincloth_side_fr_l_01
  bone2Name=dyn_loincloth_side_fr_l_MSTR
  swing1LimitDegrees=2
  swing2LimitDegrees=20
  twistLimitDegrees=4
 }
 {bone1Name=dyn_loincloth_side_fr_l_02, bone2Name=dyn_loincloth_side_fr_l_01, swing1LimitDegrees=2, swing2LimitDegrees=20, twistLimitDegrees=4}
 {bone1Name=dyn_loincloth_side_fr_l_03, bone2Name=dyn_loincloth_side_fr_l_02, swing1LimitDegrees=2, swing2LimitDegrees=20, twistLimitDegrees=4}
 {
  bone1Name=dyn_loincloth_side_fr_r_01
  bone2Name=dyn_loincloth_side_fr_r_MSTR
  swing1LimitDegrees=2
  swing2LimitDegrees=20
  twistLimitDegrees=4
 }
 {bone1Name=dyn_loincloth_side_fr_r_02, bone2Name=dyn_loincloth_side_fr_r_01, swing1LimitDegrees=2, swing2LimitDegrees=20, twistLimitDegrees=4}
 {bone1Name=dyn_loincloth_side_fr_r_03, bone2Name=dyn_loincloth_side_fr_r_02, swing1LimitDegrees=2, swing2LimitDegrees=20, twistLimitDegrees=4}
 {bone1Name=dyn_loincloth_fr_l_01, bone2Name=dyn_loincloth_fr_l_MSTR, swing1LimitDegrees=4, swing2LimitDegrees=30, twistLimitDegrees=6}
 {bone1Name=dyn_loincloth_fr_l_02, bone2Name=dyn_loincloth_fr_l_01, swing1LimitDegrees=5, swing2LimitDegrees=30, twistLimitDegrees=6}
 {bone1Name=dyn_loincloth_fr_l_03, bone2Name=dyn_loincloth_fr_l_02, swing1LimitDegrees=10, swing2LimitDegrees=30, twistLimitDegrees=6}
 {bone1Name=dyn_loincloth_fr_r_01, bone2Name=dyn_loincloth_fr_r_MSTR, swing1LimitDegrees=4, swing2LimitDegrees=30, twistLimitDegrees=6}
 {bone1Name=dyn_loincloth_fr_r_02, bone2Name=dyn_loincloth_fr_r_01, swing1LimitDegrees=5, swing2LimitDegrees=30, twistLimitDegrees=6}
 {bone1Name=dyn_loincloth_fr_r_03, bone2Name=dyn_loincloth_fr_r_02, swing1LimitDegrees=10, swing2LimitDegrees=30, twistLimitDegrees=6}
 {bone1Name=dyn_loincloth_bk_l_01, bone2Name=dyn_loincloth_bk_l_MSTR, swing1LimitDegrees=10, swing2LimitDegrees=10, twistLimitDegrees=4}
 {bone1Name=dyn_loincloth_bk_l_03, bone2Name=dyn_loincloth_bk_l_02, swing1LimitDegrees=10, swing2LimitDegrees=10, twistLimitDegrees=4}
 {bone1Name=dyn_loincloth_bk_l_02, bone2Name=dyn_loincloth_bk_l_01, swing1LimitDegrees=10, swing2LimitDegrees=10, twistLimitDegrees=4}
 {bone1Name=dyn_loincloth_bk_r_01, bone2Name=dyn_loincloth_bk_r_MSTR, swing1LimitDegrees=10, swing2LimitDegrees=10, twistLimitDegrees=4}
 {bone1Name=dyn_loincloth_bk_r_02, bone2Name=dyn_loincloth_bk_r_01, swing1LimitDegrees=10, swing2LimitDegrees=10, twistLimitDegrees=4}
 {bone1Name=dyn_loincloth_bk_r_03, bone2Name=dyn_loincloth_bk_r_02, swing1LimitDegrees=10, swing2LimitDegrees=10, twistLimitDegrees=4}
 {
  bone1Name=dyn_loincloth_side_fr_l_02
  bone2Name=dyn_loincloth_side_bk_l_02
  swing1LimitDegrees=45
  swing1Motion=Free
  swing2LimitDegrees=45
  swing2Motion=Free
  twistLimitDegrees=45
  twistMotion=Free
 }
 {
  bone1Name=dyn_loincloth_side_fr_r_02
  bone2Name=dyn_loincloth_side_bk_r_02
  swing1LimitDegrees=45
  swing1Motion=Free
  swing2LimitDegrees=45
  swing2Motion=Free
  twistLimitDegrees=45
  twistMotion=Free
 }
 {
  bone1Name=dyn_loincloth_fr_r_01
  bone2Name=dyn_loincloth_fr_l_01
  swing1LimitDegrees=45
  swing1Motion=Free
  swing2LimitDegrees=45
  swing2Motion=Free
  twistLimitDegrees=45
  twistMotion=Free
 }
 {
  bone1Name=dyn_loincloth_fr_r_02
  bone2Name=dyn_loincloth_fr_l_02
  swing1LimitDegrees=45
  swing1Motion=Free
  swing2LimitDegrees=45
  swing2Motion=Free
  twistLimitDegrees=45
  twistMotion=Free
 }
 {
  bone1Name=dyn_loincloth_fr_r_03
  bone2Name=dyn_loincloth_fr_l_03
  swing1LimitDegrees=45
  swing1Motion=Free
  swing2LimitDegrees=45
  swing2Motion=Free
  twistLimitDegrees=45
  twistMotion=Free
 }
]
