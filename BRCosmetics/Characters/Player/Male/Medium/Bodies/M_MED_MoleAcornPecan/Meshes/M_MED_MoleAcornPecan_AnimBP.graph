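# Animation Blueprint for the M_MED_MoleAcornPecan player body cosmetic, adding secondary motion to cape and loincloth.
# It copies the parent mesh pose, layers in face-accessory hood bones, then runs a Control Rig, slope-aligned cape ground collision and RigidBody.
# Cached cape RigidBody variants handle backpack, front-end and DBNO states, using constraints and bone offsets.
# Output selects a simple cloth pose, the RigidBody cape, or per-bone AnimDynamics via cloth and rigid-body flags.
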AnimBlueprint /BRCosmetics/Characters/Player/Male/Medium/Bodies/M_MED_MoleAcornPecan/Meshes/M_MED_MoleAcornPecan_AnimBP
target_skeleton=&/M_MED_MoleAcornPecan_Skeleton
graphs [
 AnimGraph {nodes [
   n1 = "Constraint - dyn_hood_bk_drape_01 (PP)"!(ComponentPose, Alpha=1, ConstraintWeights_0=0.5, ConstraintWeights_1:float=0.5) -> (Pose) {
    alphaBoolBlend {blendOption=Linear}
    boneToModify {boneName=dyn_hood_bk_drape_01}
    constraintSetup [{targetBone {boneName=dyn_cape_mid_l_01}}, {targetBone {boneName=dyn_cape_mid_r_01}}]
    constraintWeights [1,1]
   }
   n2 = "Constraint - dyn_hood_bk_drape_02 (PP)"!(ComponentPose=@n1, Alpha=1, ConstraintWeights_0=0.5, ConstraintWeights_1:float=0.5) -> (Pose) {
    alphaBoolBlend {blendOption=Linear}
    boneToModify {boneName=dyn_hood_bk_drape_02}
    constraintSetup [{targetBone {boneName=dyn_cape_mid_l_01}}, {targetBone {boneName=dyn_cape_mid_r_01}}]
    constraintWeights [1,1]
   }
   "Constraint - dyn_hood_bk_drape_end (PP)"!(ComponentPose=@n2, Alpha=1, ConstraintWeights_0=0.5, ConstraintWeights_1:float=0.5) -> (Pose) {
    alphaBoolBlend {blendOption=Linear}
    boneToModify {boneName=dyn_hood_bk_drape_end}
    constraintSetup [{targetBone {boneName=dyn_cape_mid_l_02}}, {targetBone {boneName=dyn_cape_mid_r_02}}]
    constraintWeights [1,1]
   }
   "Transform (Modify) Bone - Bone: attach_cape"!(ComponentPose, bAlphaBoolEnabled:bool=True, Translation="0,0,0", Rotation="1.804,0.592,4.387", Scale="1,1,1") -> (Pose) {
    alphaInputType=Bool
    rotationMode=BMM_Additive
    rotationSpace=BCS_BoneSpace
    translationMode=BMM_Additive
    translationSpace=BCS_BoneSpace
    alphaBoolBlend {blendOption=Linear}
    boneToModify {boneName=attach_cape}
    rotation=<rot 1.804 0.592 4.387>
   }
   "Transform (Modify) Bone - Bone: attach_backpack"!(ComponentPose, bAlphaBoolEnabled:bool=True, Translation="0,-1.5,0.745", Rotation="0,0,0", Scale="1,1,1") -> (Pose) {
    alphaInputType=Bool
    translationMode=BMM_Additive
    translationSpace=BCS_BoneSpace
    alphaBoolBlend {blendOption=Linear}
    boneToModify {boneName=attach_backpack}
    translation {y=-1.5, z=0.745}
   }
   n3 = "Transform (Modify) Bone - Bone: dyn_loincloth_bk_l_MSTR"(
    ComponentPose=Local To Component(LocalPose=Use cached pose 'Cape Final'())
    Alpha=1
    Translation="0,0,0"
    Rotation="-12.948,0.057,-0.055"
    Scale="1,1,1"
   ) {
    rotationMode=BMM_Additive
    rotationSpace=BCS_BoneSpace
    alphaBoolBlend {blendOption=Linear}
    boneToModify {boneName=dyn_loincloth_bk_l_MSTR}
    rotation=<rot -12.948 0.057 -0.055>
   }
   n4 = "Transform (Modify) Bone - Bone: dyn_loincloth_bk_r_MSTR"(ComponentPose=@n3, Alpha=1, Translation="0,0,0", Rotation="-12.491,0.025,-0.007", Scale="1,1,1") {
    rotationMode=BMM_Additive
    rotationSpace=BCS_BoneSpace
    alphaBoolBlend {blendOption=Linear}
    boneToModify {boneName=dyn_loincloth_bk_r_MSTR}
    rotation=<rot -12.491 0.025 -0.007>
   }
   n5 = Blend Poses by bool(
    BlendPose_0=Component To Local(ComponentPose=@n4)
    BlendPose_1=Use cached pose 'Post Deform Rig'()
    bActiveValue=Get CapeOn() -> (CapeOn:bool)
   ) {blendType=Linear}
   n6 = Use cached pose 'Post Anim Dynamics'()
   n7 = Get bRigidBodyEnabled()
   Output Pose(Result=Blend Poses by bool(
     BlendPose_0=Use cached pose 'Simple Cape for Cloth Final'()
     BlendPose_1=Blend Poses by bool(BlendPose_0=@n5, BlendPose_1=@n6, BlendTime_0=0, BlendTime_1=0, bActiveValue=@n7) {blendType=Linear}
     bActiveValue=Get bClothEnabled()
    ) {blendType=Linear}) {layerGroup=None}
   n8 = Control Rig!(Source=Use cached pose 'Pre CR Cache'(), Alpha, Disable Floor Slope:bool=NOT Boolean(A=Get bIsOnGround() -> (bIsOnGround:bool))) -> (Pose) {
    alphaBoolBlend {blendOption=Linear}
    controlRigAssetReference {blueprintRigClass=&/M_MED_MoleAcornPecan_CtrlRig.M_MED_MoleAcornPecan_CtrlRig_C}
   }
   n9 = "Transform (Modify) Bone - Bone: dyn_cape_ground_collision"(
    ComponentPose=Local To Component(LocalPose=@n8)
    Alpha=1
    Translation="0,0,0"
    Rotation=Get Slope Rotation() -> (Slope Rotation:Rotator)
    Scale="1,1,1"
   ) {
    rotationMode=BMM_Replace
    rotationSpace=BCS_WorldSpace
    alphaBoolBlend {blendOption=Linear}
    boneToModify {boneName=dyn_cape_ground_collision}
   }
   n10 = RigidBody!(
    ComponentPose=@n9
    Alpha=Get RBAN Alpha() -> (RBAN Alpha:double)
    OverrideWorldGravity:Vector="0,0,-980"
    ExternalForce
    SimSpaceSettings:Sim Space Settings Structure=(WorldAlpha=1,VelocityScaleZ=1,DampingAlpha=0,MaxLinearVelocity=800,MaxAngularVelocity=6,MaxLinearAcceleration=4000,MaxAngularAcceleration=30,ExternalLinearDragV=(X=0.5,Y=0.5,Z=0.5),ExternalLinearVelocity=(X=0,Y=0,Z=0),ExternalAngularVelocity=(X=0,Y=0,Z=0))
   ) -> (Pose) {
    bOverrideWorldGravity=true
    overridePhysicsAsset=&/M_MED_MoleAcornPecan_Physics
    simulationSpace=BaseBoneSpace
    alphaBoolBlend {blendOption=Linear}
    baseBoneRef {boneName=pelvis}
    overrideWorldGravity {z=-980}
    simSpaceSettings {
     dampingAlpha=0
     maxAngularAcceleration=30
     maxAngularVelocity=6
     maxLinearAcceleration=4000
     maxLinearVelocity=800
     worldAlpha=1
     externalLinearDragV=<0.5 0.5 0.5>
    }
   }
   n11 = "Transform (Modify) Bone - Bone: dyn_loincloth_fr_l_MSTR"!(ComponentPose=@n10, bAlphaBoolEnabled:bool=True, Translation="0,0,0", Rotation="0,0,-15", Scale="1,1,1") -> (Pose) {
    alphaInputType=Bool
    rotationMode=BMM_Additive
    rotationSpace=BCS_BoneSpace
    alphaBoolBlend {blendOption=Linear}
    boneToModify {boneName=dyn_loincloth_fr_l_MSTR}
    rotation {roll=-15}
   }
   n12 = "Transform (Modify) Bone - Bone: dyn_loincloth_fr_r_MSTR"!(ComponentPose=@n11, bAlphaBoolEnabled:bool=True, Translation="0,0,0", Rotation="0,0,-15", Scale="1,1,1") -> (Pose) {
    alphaInputType=Bool
    rotationMode=BMM_Additive
    rotationSpace=BCS_BoneSpace
    alphaBoolBlend {blendOption=Linear}
    boneToModify {boneName=dyn_loincloth_fr_r_MSTR}
    rotation {roll=-15}
   }
   Save cached pose 'Post Deform Rig'(Pose=Component To Local(ComponentPose=@n12)) {cacheName=Post Deform Rig}
   n13 = RigidBody(
    ComponentPose=Local To Component(LocalPose=Use cached pose 'Post Deform Rig'())
    OverrideWorldGravity="0,0,-980"
    SimSpaceSettings=(WorldAlpha=1,VelocityScaleZ=1,DampingAlpha=0,MaxLinearVelocity=800,MaxAngularVelocity=6,MaxLinearAcceleration=4000,MaxAngularAcceleration=30,ExternalLinearDragV=(X=0.5,Y=0.5,Z=0.5),ExternalLinearVelocity=(X=0,Y=0,Z=0),ExternalAngularVelocity=(X=0,Y=0,Z=0))
   ) {
    bOverrideWorldGravity=true
    overridePhysicsAsset=&/M_MED_MoleAcornPecan_CapePhysics
    simulationSpace=BaseBoneSpace
    alphaBoolBlend {blendOption=Linear}
    baseBoneRef {boneName=root}
    overrideWorldGravity {z=-980}
    simSpaceSettings {
     dampingAlpha=0
     maxAngularAcceleration=30
     maxAngularVelocity=6
     maxLinearAcceleration=4000
     maxLinearVelocity=800
     worldAlpha=1
     externalLinearDragV=<0.5 0.5 0.5>
    }
   }
   n14 = "Transform (Modify) Bone - Bone: deform_back"!(ComponentPose=@n13, bAlphaBoolEnabled:bool=True, Translation="-2,0,0", Rotation="0,0,0", Scale="1,1,1") -> (Pose) {
    alphaInputType=Bool
    translationMode=BMM_Additive
    translationSpace=BCS_BoneSpace
    alphaBoolBlend {blendOption=Linear}
    boneToModify {boneName=deform_back}
    translation {x=-2}
   }
   Save cached pose 'Cape Normal'(Pose=Component To Local(ComponentPose=@n14)) {cacheName=Cape Normal}
   n15 = RigidBody(
    ComponentPose=Local To Component(LocalPose=Use cached pose 'Post Deform Rig'())
    OverrideWorldGravity="0,0,-980"
    SimSpaceSettings=(WorldAlpha=1,VelocityScaleZ=1,DampingAlpha=0,MaxLinearVelocity=800,MaxAngularVelocity=6,MaxLinearAcceleration=4000,MaxAngularAcceleration=30,ExternalLinearDragV=(X=0.5,Y=0.5,Z=0.5),ExternalLinearVelocity=(X=0,Y=0,Z=0),ExternalAngularVelocity=(X=0,Y=0,Z=0))
   ) {
    bOverrideWorldGravity=true
    overridePhysicsAsset=&/M_MED_MoleAcornPecan_FE_Pack_CapePhysics
    simulationSpace=BaseBoneSpace
    alphaBoolBlend {blendOption=Linear}
    baseBoneRef {boneName=root}
    overrideWorldGravity {z=-980}
    simSpaceSettings {
     dampingAlpha=0
     maxAngularAcceleration=30
     maxAngularVelocity=6
     maxLinearAcceleration=4000
     maxLinearVelocity=800
     worldAlpha=1
     externalLinearDragV=<0.5 0.5 0.5>
    }
   }
   n16 = "Transform (Modify) Bone - Bone: deform_back"(ComponentPose=@n15, bAlphaBoolEnabled=True, Translation="-1.5,0,0", Rotation="0,0,0", Scale="1,1,1") {
    alphaInputType=Bool
    translationMode=BMM_Additive
    translationSpace=BCS_BoneSpace
    alphaBoolBlend {blendOption=Linear}
    boneToModify {boneName=deform_back}
    translation {x=-1.5}
   }
   n17 = "Constraint - dyn_cape_l_01 (P)"(ComponentPose=@n16, Alpha=1, ConstraintWeights_0=1) {
    alphaBoolBlend {blendOption=Linear}
    boneToModify {boneName=dyn_cape_l_01}
    constraintSetup [{perAxis {bX=false, bZ=false}, targetBone {boneName=spine_05}}]
    constraintWeights [1]
   }
   n18 = "Constraint - dyn_cape_mid_l_01 (P)"(ComponentPose=@n17, Alpha=1, ConstraintWeights_0=1) {
    alphaBoolBlend {blendOption=Linear}
    boneToModify {boneName=dyn_cape_mid_l_01}
    constraintSetup [{perAxis {bX=false, bZ=false}, targetBone {boneName=spine_05}}]
    constraintWeights [1]
   }
   n19 = "Constraint - dyn_cape_mid_r_01 (P)"(ComponentPose=@n18, Alpha=1, ConstraintWeights_0=1) {
    alphaBoolBlend {blendOption=Linear}
    boneToModify {boneName=dyn_cape_mid_r_01}
    constraintSetup [{perAxis {bX=false, bZ=false}, targetBone {boneName=spine_05}}]
    constraintWeights [1]
   }
   n20 = "Constraint - dyn_cape_r_01 (P)"(ComponentPose=@n19, Alpha=1, ConstraintWeights_0=1) {
    alphaBoolBlend {blendOption=Linear}
    boneToModify {boneName=dyn_cape_r_01}
    constraintSetup [{perAxis {bX=false, bZ=false}, targetBone {boneName=spine_05}}]
    constraintWeights [1]
   }
   n21 = "Constraint - dyn_cape_l_02 (P)"(ComponentPose=@n20, Alpha=0.5, ConstraintWeights_0=1) {
    alphaBoolBlend {blendOption=Linear}
    boneToModify {boneName=dyn_cape_l_02}
    constraintSetup [{perAxis {bX=false, bZ=false}, targetBone {boneName=spine_04}}]
    constraintWeights [1]
   }
   n22 = "Constraint - dyn_cape_mid_l_02 (P)"(ComponentPose=@n21, Alpha=0.5, ConstraintWeights_0=1) {
    alphaBoolBlend {blendOption=Linear}
    boneToModify {boneName=dyn_cape_mid_l_02}
    constraintSetup [{perAxis {bX=false, bZ=false}, targetBone {boneName=spine_04}}]
    constraintWeights [1]
   }
   n23 = "Constraint - dyn_cape_mid_r_02 (P)"(ComponentPose=@n22, Alpha=0.5, ConstraintWeights_0=1) {
    alphaBoolBlend {blendOption=Linear}
    boneToModify {boneName=dyn_cape_mid_r_02}
    constraintSetup [{perAxis {bX=false, bZ=false}, targetBone {boneName=spine_04}}]
    constraintWeights [1]
   }
   n24 = "Constraint - dyn_cape_r_02 (P)"(ComponentPose=@n23, Alpha=0.5, ConstraintWeights_0=1) {
    alphaBoolBlend {blendOption=Linear}
    boneToModify {boneName=dyn_cape_r_02}
    constraintSetup [{perAxis {bX=false, bZ=false}, targetBone {boneName=spine_04}}]
    constraintWeights [1]
   }
   n25 = Component To Local(ComponentPose=@n24)
   n26 = RigidBody(
    ComponentPose=Local To Component(LocalPose=Use cached pose 'Post Deform Rig'())
    OverrideWorldGravity="0,0,-980"
    SimSpaceSettings=(WorldAlpha=1,VelocityScaleZ=1,DampingAlpha=0,MaxLinearVelocity=800,MaxAngularVelocity=6,MaxLinearAcceleration=4000,MaxAngularAcceleration=30,ExternalLinearDragV=(X=0.5,Y=0.5,Z=0.5),ExternalLinearVelocity=(X=0,Y=0,Z=0),ExternalAngularVelocity=(X=0,Y=0,Z=0))
   ) {
    bOverrideWorldGravity=true
    overridePhysicsAsset=&/M_MED_MoleAcornPecan_Pack_CapePhysics
    simulationSpace=BaseBoneSpace
    alphaBoolBlend {blendOption=Linear}
    baseBoneRef {boneName=root}
    overrideWorldGravity {z=-980}
    simSpaceSettings {
     dampingAlpha=0
     maxAngularAcceleration=30
     maxAngularVelocity=6
     maxLinearAcceleration=4000
     maxLinearVelocity=800
     worldAlpha=1
     externalLinearDragV=<0.5 0.5 0.5>
    }
   }
   n27 = "Transform (Modify) Bone - Bone: deform_back"(ComponentPose=@n26, bAlphaBoolEnabled=True, Translation="-2,0,0", Rotation="0,0,0", Scale="1,1,1") {
    alphaInputType=Bool
    translationMode=BMM_Additive
    translationSpace=BCS_BoneSpace
    alphaBoolBlend {blendOption=Linear}
    boneToModify {boneName=deform_back}
    translation {x=-2}
   }
   n28 = "Constraint - dyn_cape_l_01 (P)"(ComponentPose=@n27, Alpha=1, ConstraintWeights_0=1) {
    alphaBoolBlend {blendOption=Linear}
    boneToModify {boneName=dyn_cape_l_01}
    constraintSetup [{perAxis {bX=false, bZ=false}, targetBone {boneName=spine_05}}]
    constraintWeights [1]
   }
   n29 = "Constraint - dyn_cape_mid_l_01 (P)"(ComponentPose=@n28, Alpha=1, ConstraintWeights_0=1) {
    alphaBoolBlend {blendOption=Linear}
    boneToModify {boneName=dyn_cape_mid_l_01}
    constraintSetup [{perAxis {bX=false, bZ=false}, targetBone {boneName=spine_05}}]
    constraintWeights [1]
   }
   n30 = "Constraint - dyn_cape_mid_r_01 (P)"(ComponentPose=@n29, Alpha=1, ConstraintWeights_0=1) {
    alphaBoolBlend {blendOption=Linear}
    boneToModify {boneName=dyn_cape_mid_r_01}
    constraintSetup [{perAxis {bX=false, bZ=false}, targetBone {boneName=spine_05}}]
    constraintWeights [1]
   }
   n31 = "Constraint - dyn_cape_r_01 (P)"(ComponentPose=@n30, Alpha=1, ConstraintWeights_0=1) {
    alphaBoolBlend {blendOption=Linear}
    boneToModify {boneName=dyn_cape_r_01}
    constraintSetup [{perAxis {bX=false, bZ=false}, targetBone {boneName=spine_05}}]
    constraintWeights [1]
   }
   n32 = "Constraint - dyn_cape_l_02 (P)"(ComponentPose=@n31, Alpha=1, ConstraintWeights_0=1) {
    alphaBoolBlend {blendOption=Linear}
    boneToModify {boneName=dyn_cape_l_02}
    constraintSetup [{perAxis {bX=false, bZ=false}, targetBone {boneName=spine_04}}]
    constraintWeights [1]
   }
   n33 = "Constraint - dyn_cape_mid_l_02 (P)"(ComponentPose=@n32, Alpha=1, ConstraintWeights_0=1) {
    alphaBoolBlend {blendOption=Linear}
    boneToModify {boneName=dyn_cape_mid_l_02}
    constraintSetup [{perAxis {bX=false, bZ=false}, targetBone {boneName=spine_04}}]
    constraintWeights [1]
   }
   n34 = "Constraint - dyn_cape_mid_r_02 (P)"(ComponentPose=@n33, Alpha=1, ConstraintWeights_0=1) {
    alphaBoolBlend {blendOption=Linear}
    boneToModify {boneName=dyn_cape_mid_r_02}
    constraintSetup [{perAxis {bX=false, bZ=false}, targetBone {boneName=spine_04}}]
    constraintWeights [1]
   }
   n35 = "Constraint - dyn_cape_r_02 (P)"(ComponentPose=@n34, Alpha=1, ConstraintWeights_0=1) {
    alphaBoolBlend {blendOption=Linear}
    boneToModify {boneName=dyn_cape_r_02}
    constraintSetup [{perAxis {bX=false, bZ=false}, targetBone {boneName=spine_04}}]
    constraintWeights [1]
   }
   n36 = Component To Local(ComponentPose=@n35)
   n37 = Get bIsInFrontEnd()
   n38 = "Transform (Modify) Bone - Bone: dyn_cape_l_01"!(
    ComponentPose=Local To Component(LocalPose=Blend Poses by bool(BlendPose_0=@n25, BlendPose_1=@n36, bActiveValue=@n37) {blendType=Linear})
    bAlphaBoolEnabled:bool=True
    Translation="0,0,0"
    Rotation="4,0,0"
    Scale="1,1,1"
   ) -> (Pose) {
    alphaInputType=Bool
    rotationMode=BMM_Additive
    rotationSpace=BCS_BoneSpace
    translationMode=BMM_Additive
    translationSpace=BCS_BoneSpace
    alphaBoolBlend {blendOption=Linear}
    boneToModify {boneName=dyn_cape_l_01}
    rotation {pitch=4}
   }
   n39 = "Transform (Modify) Bone - Bone: dyn_cape_mid_l_01"!(ComponentPose=@n38, bAlphaBoolEnabled:bool=True, Translation="0,0,0", Rotation="4,0,0", Scale="1,1,1") -> (Pose) {
    alphaInputType=Bool
    rotationMode=BMM_Additive
    rotationSpace=BCS_BoneSpace
    translationMode=BMM_Additive
    translationSpace=BCS_BoneSpace
    alphaBoolBlend {blendOption=Linear}
    boneToModify {boneName=dyn_cape_mid_l_01}
    rotation {pitch=4}
   }
   n40 = "Transform (Modify) Bone - Bone: dyn_cape_mid_r_01"!(ComponentPose=@n39, bAlphaBoolEnabled:bool=True, Translation="0,0,0", Rotation="4,0,0", Scale="1,1,1") -> (Pose) {
    alphaInputType=Bool
    rotationMode=BMM_Additive
    rotationSpace=BCS_BoneSpace
    translationMode=BMM_Additive
    translationSpace=BCS_BoneSpace
    alphaBoolBlend {blendOption=Linear}
    boneToModify {boneName=dyn_cape_mid_r_01}
    rotation {pitch=4}
   }
   n41 = "Transform (Modify) Bone - Bone: dyn_cape_r_01"!(ComponentPose=@n40, bAlphaBoolEnabled:bool=True, Translation="0,0,0", Rotation="4,0,0", Scale="1,1,1") -> (Pose) {
    alphaInputType=Bool
    rotationMode=BMM_Additive
    rotationSpace=BCS_BoneSpace
    translationMode=BMM_Additive
    translationSpace=BCS_BoneSpace
    alphaBoolBlend {blendOption=Linear}
    boneToModify {boneName=dyn_cape_r_01}
    rotation {pitch=4}
   }
   n42 = "Transform (Modify) Bone - Bone: dyn_cape_l_03"(ComponentPose=@n41, Alpha=1, Translation="0,0,6", Rotation="10,0,0", Scale="1,1,1") {
    rotationMode=BMM_Additive
    rotationSpace=BCS_BoneSpace
    translationMode=BMM_Additive
    translationSpace=BCS_BoneSpace
    alphaBoolBlend {blendOption=Linear}
    boneToModify {boneName=dyn_cape_l_03}
    rotation {pitch=10}
    translation {z=6}
   }
   n43 = "Transform (Modify) Bone - Bone: dyn_cape_mid_l_03"(ComponentPose=@n42, Alpha=1, Translation="0,0,6", Rotation="10,0,0", Scale="1,1,1") {
    rotationMode=BMM_Additive
    rotationSpace=BCS_BoneSpace
    translationMode=BMM_Additive
    translationSpace=BCS_BoneSpace
    alphaBoolBlend {blendOption=Linear}
    boneToModify {boneName=dyn_cape_mid_l_03}
    rotation {pitch=10}
    translation {z=6}
   }
   n44 = "Transform (Modify) Bone - Bone: dyn_cape_r_03"(ComponentPose=@n43, Alpha=1, Translation="0,0,6", Rotation="10,0,0", Scale="1,1,1") {
    rotationMode=BMM_Additive
    rotationSpace=BCS_BoneSpace
    translationMode=BMM_Additive
    translationSpace=BCS_BoneSpace
    alphaBoolBlend {blendOption=Linear}
    boneToModify {boneName=dyn_cape_r_03}
    rotation {pitch=10}
    translation {z=6}
   }
   n45 = "Transform (Modify) Bone - Bone: dyn_cape_mid_r_03"(ComponentPose=@n44, Alpha=1, Translation="0,0,6", Rotation="10,0,0", Scale="1,1,1") {
    rotationMode=BMM_Additive
    rotationSpace=BCS_BoneSpace
    translationMode=BMM_Additive
    translationSpace=BCS_BoneSpace
    alphaBoolBlend {blendOption=Linear}
    boneToModify {boneName=dyn_cape_mid_r_03}
    rotation {pitch=10}
    translation {z=6}
   }
   n46 = "Transform (Modify) Bone - Bone: dyn_cape_l_01"!(ComponentPose=@n45, Alpha:float=1, Translation="0,0,6", Rotation="0,0,0", Scale="1,1,1") -> (Pose) {
    rotationMode=BMM_Additive
    rotationSpace=BCS_BoneSpace
    translationMode=BMM_Additive
    translationSpace=BCS_BoneSpace
    alphaBoolBlend {blendOption=Linear}
    boneToModify {boneName=dyn_cape_l_01}
    translation {z=6}
   }
   n47 = "Transform (Modify) Bone - Bone: dyn_cape_mid_l_01"!(ComponentPose=@n46, Alpha:float=1, Translation="0,0,6", Rotation="0,0,0", Scale="1,1,1") -> (Pose) {
    rotationMode=BMM_Additive
    rotationSpace=BCS_BoneSpace
    translationMode=BMM_Additive
    translationSpace=BCS_BoneSpace
    alphaBoolBlend {blendOption=Linear}
    boneToModify {boneName=dyn_cape_mid_l_01}
    translation {z=6}
   }
   n48 = "Transform (Modify) Bone - Bone: dyn_cape_r_01"!(ComponentPose=@n47, Alpha:float=1, Translation="0,0,6", Rotation="0,0,0", Scale="1,1,1") -> (Pose) {
    rotationMode=BMM_Additive
    rotationSpace=BCS_BoneSpace
    translationMode=BMM_Additive
    translationSpace=BCS_BoneSpace
    alphaBoolBlend {blendOption=Linear}
    boneToModify {boneName=dyn_cape_r_01}
    translation {z=6}
   }
   n49 = "Transform (Modify) Bone - Bone: dyn_cape_mid_r_01"!(ComponentPose=@n48, Alpha:float=1, Translation="0,0,6", Rotation="0,0,0", Scale="1,1,1") -> (Pose) {
    rotationMode=BMM_Additive
    rotationSpace=BCS_BoneSpace
    translationMode=BMM_Additive
    translationSpace=BCS_BoneSpace
    alphaBoolBlend {blendOption=Linear}
    boneToModify {boneName=dyn_cape_mid_r_01}
    translation {z=6}
   }
   n50 = "Transform (Modify) Bone - Bone: attach_backpack"(ComponentPose=@n49, bAlphaBoolEnabled=True, Translation="0,-2.5,0", Rotation="0,0,0", Scale="1,1,1") {
    alphaInputType=Bool
    translationMode=BMM_Additive
    translationSpace=BCS_BoneSpace
    alphaBoolBlend {blendOption=Linear}
    boneToModify {boneName=attach_backpack}
    translation {y=-2.5}
   }
   Save cached pose 'Cape With Backpack'(Pose=Component To Local(ComponentPose=@n50)) {cacheName=Cape With Backpack}
   Save cached pose 'Cape Handle Pack'(Pose=Blend Poses by bool(BlendPose_0=Use cached pose 'Cape Normal'(), BlendPose_1=Use cached pose 'Cape With Backpack'(), bActiveValue=Get bIsDBNO()) {blendType=Linear}) {cacheName=Cape Handle Pack}
   Save cached pose 'Cape Final'(Pose=Blend Poses by bool(
     BlendPose_0=Use cached pose 'Cape Handle Pack'()
     BlendPose_1=Use cached pose 'Cape Normal'()
     bActiveValue=Get isWearingPack() -> (isWearingPack:bool)
    ) {blendType=Linear}) {cacheName=Cape Final}
   n51 = Get Face Acc Component() -> (Face Acc Component:Skeletal Mesh Component Budgeted Object Reference)
   Save cached pose 'Pre CR Cache'(Pose=Layered blend per bone(
     BasePose=Copy Pose From Mesh() {bUseAttachedParent=true}
     BlendPoses_0=Copy Pose From Mesh(SourceMeshComponent=@n51) {bUseAttachedParent=true}
    ) {layerSetup [{branchFilters [{boneName=dyn_hood_bk_drape_01}]}]}) {cacheName=Pre CR Cache}
   n52 = Anim Dynamics - Bone: dyn_cape_l_01(ComponentPose=Local To Component(LocalPose=Use cached pose 'Post Deform Rig'()), Alpha=1, ExternalForce="0,0,0") {
    angularDampingOverride=1
    angularSpringConstant=10000
    bAngularSpring=true
    bOverrideAngularDamping=true
    bUseGravityOverride=true
    gravityScale=0.35
    simulationSpace=BoneRelative
    alphaBoolBlend {blendOption=Linear}
    boundBone {boneName=dyn_cape_l_01}
    componentAppliedLinearAccClamp=<3000 3000 3000>
    componentLinearAccScale=<0.7 0.7 0.7>
    componentLinearVelScale=<0.7 0.7 0.7>
    gravityOverride {z=-980}
    physicsBodyDefinitions [{
      boundBone {boneName=dyn_cape_l_01}
      boxExtents=<10 40 40>
      constraintSetup {angularLimitsMax=<0 5 15>, angularLimitsMin=<0 -25 -15>, angularTarget=<1 0 0>}
      localJointOffset=<5 0 0>
     }]
    relativeSpaceBone {boneName=attach_cape}
   }
   n53 = Anim Dynamics - Bone: dyn_cape_l_02(ComponentPose=@n52, Alpha=1, ExternalForce="0,0,0") {
    angularDampingOverride=1
    angularSpringConstant=10000
    bAngularSpring=true
    bOverrideAngularDamping=true
    bUseGravityOverride=true
    gravityScale=0.35
    simulationSpace=BoneRelative
    alphaBoolBlend {blendOption=Linear}
    alphaScaleBiasClamp {outRange {min=0.6}}
    boundBone {boneName=dyn_cape_l_02}
    componentAppliedLinearAccClamp=<3000 3000 3000>
    componentLinearAccScale=<0.7 0.7 0.7>
    componentLinearVelScale=<0.7 0.7 0.7>
    gravityOverride {z=-980}
    physicsBodyDefinitions [{
      boundBone {boneName=dyn_cape_l_02}
      boxExtents=<10 40 40>
      constraintSetup {angularLimitsMax=<0 5 15>, angularLimitsMin=<0 -25 -15>, angularTarget=<1 0 0>}
      localJointOffset=<5 0 0>
     }]
    relativeSpaceBone {boneName=attach_cape}
   }
   n54 = Anim Dynamics - Bone: dyn_cape_l_03(ComponentPose=@n53, Alpha=1, ExternalForce="0,0,0") {
    angularDampingOverride=1
    angularSpringConstant=10000
    bAngularSpring=true
    bOverrideAngularDamping=true
    bUseGravityOverride=true
    gravityScale=0.35
    simulationSpace=BoneRelative
    alphaBoolBlend {blendOption=Linear}
    boundBone {boneName=dyn_cape_l_03}
    componentAppliedLinearAccClamp=<3000 3000 3000>
    componentLinearAccScale=<0.7 0.7 0.7>
    componentLinearVelScale=<0.7 0.7 0.7>
    gravityOverride {z=-980}
    physicsBodyDefinitions [{
      boundBone {boneName=dyn_cape_l_03}
      boxExtents=<10 40 40>
      constraintSetup {angularLimitsMax=<0 5 15>, angularLimitsMin=<0 -25 -15>, angularTarget=<1 0 0>}
      localJointOffset=<5 0 0>
     }]
    relativeSpaceBone {boneName=attach_cape}
   }
   n55 = Anim Dynamics - Bone: dyn_cape_l_04(ComponentPose=@n54, Alpha=1, ExternalForce="0,0,0") {
    angularDampingOverride=1
    angularSpringConstant=10000
    bAngularSpring=true
    bOverrideAngularDamping=true
    bUseGravityOverride=true
    gravityScale=0.35
    simulationSpace=BoneRelative
    alphaBoolBlend {blendOption=Linear}
    boundBone {boneName=dyn_cape_l_04}
    componentAppliedLinearAccClamp=<3000 3000 3000>
    componentLinearAccScale=<0.7 0.7 0.7>
    componentLinearVelScale=<0.7 0.7 0.7>
    gravityOverride {z=-980}
    physicsBodyDefinitions [{
      boundBone {boneName=dyn_cape_l_04}
      boxExtents=<10 40 40>
      constraintSetup {angularLimitsMax=<0 5 15>, angularLimitsMin=<0 -25 -15>, angularTarget=<1 0 0>}
      localJointOffset=<5 0 0>
     }]
    relativeSpaceBone {boneName=attach_cape}
   }
   n56 = Anim Dynamics - Bone: dyn_cape_l_05(ComponentPose=@n55, Alpha=1, ExternalForce="0,0,0") {
    angularDampingOverride=1
    angularSpringConstant=10000
    bAngularSpring=true
    bOverrideAngularDamping=true
    bUseGravityOverride=true
    gravityScale=0.35
    simulationSpace=BoneRelative
    alphaBoolBlend {blendOption=Linear}
    boundBone {boneName=dyn_cape_l_05}
    componentAppliedLinearAccClamp=<3000 3000 3000>
    componentLinearAccScale=<0.7 0.7 0.7>
    componentLinearVelScale=<0.7 0.7 0.7>
    gravityOverride {z=-980}
    physicsBodyDefinitions [{
      boundBone {boneName=dyn_cape_l_05}
      boxExtents=<10 40 40>
      constraintSetup {angularLimitsMax=<0 5 15>, angularLimitsMin=<0 -25 -15>, angularTarget=<1 0 0>}
      localJointOffset=<5 0 0>
     }]
    relativeSpaceBone {boneName=attach_cape}
   }
   n57 = Anim Dynamics - Bone: dyn_cape_l_06(ComponentPose=@n56, Alpha=1, ExternalForce="0,0,0") {
    angularDampingOverride=1
    angularSpringConstant=10000
    bAngularSpring=true
    bOverrideAngularDamping=true
    bUseGravityOverride=true
    gravityScale=0.35
    simulationSpace=BoneRelative
    alphaBoolBlend {blendOption=Linear}
    boundBone {boneName=dyn_cape_l_06}
    componentAppliedLinearAccClamp=<3000 3000 3000>
    componentLinearAccScale=<0.7 0.7 0.7>
    componentLinearVelScale=<0.7 0.7 0.7>
    gravityOverride {z=-980}
    physicsBodyDefinitions [{
      boundBone {boneName=dyn_cape_l_06}
      boxExtents=<10 40 40>
      constraintSetup {angularLimitsMax=<0 5 15>, angularLimitsMin=<0 -25 -15>, angularTarget=<1 0 0>}
      localJointOffset=<5 0 0>
     }]
    relativeSpaceBone {boneName=attach_cape}
   }
   n58 = Anim Dynamics - Bone: dyn_cape_mid_l_01(ComponentPose=@n57, Alpha=1, ExternalForce="0,0,0") {
    angularDampingOverride=1
    angularSpringConstant=10000
    bAngularSpring=true
    bOverrideAngularDamping=true
    bUseGravityOverride=true
    gravityScale=0.35
    simulationSpace=BoneRelative
    alphaBoolBlend {blendOption=Linear}
    boundBone {boneName=dyn_cape_mid_l_01}
    componentAppliedLinearAccClamp=<3000 3000 3000>
    componentLinearAccScale=<0.7 0.7 0.7>
    componentLinearVelScale=<0.7 0.7 0.7>
    gravityOverride {z=-980}
    physicsBodyDefinitions [{
      boundBone {boneName=dyn_cape_mid_l_01}
      boxExtents=<10 40 40>
      constraintSetup {angularLimitsMax=<0 5 15>, angularLimitsMin=<0 -25 -15>, angularTarget=<1 0 0>}
      localJointOffset=<5 0 0>
     }]
    relativeSpaceBone {boneName=attach_cape}
   }
   n59 = Anim Dynamics - Bone: dyn_cape_mid_l_02(ComponentPose=@n58, Alpha=1, ExternalForce="0,0,0") {
    angularDampingOverride=1
    angularSpringConstant=10000
    bAngularSpring=true
    bOverrideAngularDamping=true
    bUseGravityOverride=true
    gravityScale=0.35
    simulationSpace=BoneRelative
    alphaBoolBlend {blendOption=Linear}
    alphaScaleBiasClamp {outRange {min=0.6}}
    boundBone {boneName=dyn_cape_mid_l_02}
    componentAppliedLinearAccClamp=<3000 3000 3000>
    componentLinearAccScale=<0.7 0.7 0.7>
    componentLinearVelScale=<0.7 0.7 0.7>
    gravityOverride {z=-980}
    physicsBodyDefinitions [{
      boundBone {boneName=dyn_cape_mid_l_02}
      boxExtents=<10 40 40>
      constraintSetup {angularLimitsMax=<0 5 15>, angularLimitsMin=<0 -25 -15>, angularTarget=<1 0 0>}
      localJointOffset=<5 0 0>
     }]
    relativeSpaceBone {boneName=attach_cape}
   }
   n60 = Anim Dynamics - Bone: dyn_cape_mid_l_03(ComponentPose=@n59, Alpha=1, ExternalForce="0,0,0") {
    angularDampingOverride=1
    angularSpringConstant=10000
    bAngularSpring=true
    bOverrideAngularDamping=true
    bUseGravityOverride=true
    gravityScale=0.35
    simulationSpace=BoneRelative
    alphaBoolBlend {blendOption=Linear}
    boundBone {boneName=dyn_cape_mid_l_03}
    componentAppliedLinearAccClamp=<3000 3000 3000>
    componentLinearAccScale=<0.7 0.7 0.7>
    componentLinearVelScale=<0.7 0.7 0.7>
    gravityOverride {z=-980}
    physicsBodyDefinitions [{
      boundBone {boneName=dyn_cape_mid_l_03}
      boxExtents=<10 40 40>
      constraintSetup {angularLimitsMax=<0 5 15>, angularLimitsMin=<0 -25 -15>, angularTarget=<1 0 0>}
      localJointOffset=<5 0 0>
     }]
    relativeSpaceBone {boneName=attach_cape}
   }
   n61 = Anim Dynamics - Bone: dyn_cape_mid_l_04(ComponentPose=@n60, Alpha=1, ExternalForce="0,0,0") {
    angularDampingOverride=1
    angularSpringConstant=10000
    bAngularSpring=true
    bOverrideAngularDamping=true
    bUseGravityOverride=true
    gravityScale=0.35
    simulationSpace=BoneRelative
    alphaBoolBlend {blendOption=Linear}
    boundBone {boneName=dyn_cape_mid_l_04}
    componentAppliedLinearAccClamp=<3000 3000 3000>
    componentLinearAccScale=<0.7 0.7 0.7>
    componentLinearVelScale=<0.7 0.7 0.7>
    gravityOverride {z=-980}
    physicsBodyDefinitions [{
      boundBone {boneName=dyn_cape_mid_l_04}
      boxExtents=<10 40 40>
      constraintSetup {angularLimitsMax=<0 5 15>, angularLimitsMin=<0 -25 -15>, angularTarget=<1 0 0>}
      localJointOffset=<5 0 0>
     }]
    relativeSpaceBone {boneName=attach_cape}
   }
   n62 = Anim Dynamics - Bone: dyn_cape_mid_l_05(ComponentPose=@n61, Alpha=1, ExternalForce="0,0,0") {
    angularDampingOverride=1
    angularSpringConstant=10000
    bAngularSpring=true
    bOverrideAngularDamping=true
    bUseGravityOverride=true
    gravityScale=0.35
    simulationSpace=BoneRelative
    alphaBoolBlend {blendOption=Linear}
    boundBone {boneName=dyn_cape_mid_l_05}
    componentAppliedLinearAccClamp=<3000 3000 3000>
    componentLinearAccScale=<0.7 0.7 0.7>
    componentLinearVelScale=<0.7 0.7 0.7>
    gravityOverride {z=-980}
    physicsBodyDefinitions [{
      boundBone {boneName=dyn_cape_mid_l_05}
      boxExtents=<10 40 40>
      constraintSetup {angularLimitsMax=<0 5 15>, angularLimitsMin=<0 -25 -15>, angularTarget=<1 0 0>}
      localJointOffset=<5 0 0>
     }]
    relativeSpaceBone {boneName=attach_cape}
   }
   n63 = Anim Dynamics - Bone: dyn_cape_mid_l_06(ComponentPose=@n62, Alpha=1, ExternalForce="0,0,0") {
    angularDampingOverride=1
    angularSpringConstant=10000
    bAngularSpring=true
    bOverrideAngularDamping=true
    bUseGravityOverride=true
    gravityScale=0.35
    simulationSpace=BoneRelative
    alphaBoolBlend {blendOption=Linear}
    boundBone {boneName=dyn_cape_mid_l_06}
    componentAppliedLinearAccClamp=<3000 3000 3000>
    componentLinearAccScale=<0.7 0.7 0.7>
    componentLinearVelScale=<0.7 0.7 0.7>
    gravityOverride {z=-980}
    physicsBodyDefinitions [{
      boundBone {boneName=dyn_cape_mid_l_06}
      boxExtents=<10 40 40>
      constraintSetup {angularLimitsMax=<0 5 15>, angularLimitsMin=<0 -25 -15>, angularTarget=<1 0 0>}
      localJointOffset=<5 0 0>
     }]
    relativeSpaceBone {boneName=attach_cape}
   }
   n64 = Anim Dynamics - Bone: dyn_cape_mid_r_01(ComponentPose=@n63, Alpha=1, ExternalForce="0,0,0") {
    angularDampingOverride=1
    angularSpringConstant=10000
    bAngularSpring=true
    bOverrideAngularDamping=true
    bUseGravityOverride=true
    gravityScale=0.35
    simulationSpace=BoneRelative
    alphaBoolBlend {blendOption=Linear}
    boundBone {boneName=dyn_cape_mid_r_01}
    componentAppliedLinearAccClamp=<3000 3000 3000>
    componentLinearAccScale=<0.7 0.7 0.7>
    componentLinearVelScale=<0.7 0.7 0.7>
    gravityOverride {z=-980}
    physicsBodyDefinitions [{
      boundBone {boneName=dyn_cape_mid_r_01}
      boxExtents=<10 40 40>
      constraintSetup {angularLimitsMax=<0 5 15>, angularLimitsMin=<0 -25 -15>, angularTarget=<1 0 0>}
      localJointOffset=<5 0 0>
     }]
    relativeSpaceBone {boneName=attach_cape}
   }
   n65 = Anim Dynamics - Bone: dyn_cape_mid_r_02(ComponentPose=@n64, Alpha=1, ExternalForce="0,0,0") {
    angularDampingOverride=1
    angularSpringConstant=10000
    bAngularSpring=true
    bOverrideAngularDamping=true
    bUseGravityOverride=true
    gravityScale=0.35
    simulationSpace=BoneRelative
    alphaBoolBlend {blendOption=Linear}
    alphaScaleBiasClamp {outRange {min=0.6}}
    boundBone {boneName=dyn_cape_mid_r_02}
    componentAppliedLinearAccClamp=<3000 3000 3000>
    componentLinearAccScale=<0.7 0.7 0.7>
    componentLinearVelScale=<0.7 0.7 0.7>
    gravityOverride {z=-980}
    physicsBodyDefinitions [{
      boundBone {boneName=dyn_cape_mid_r_02}
      boxExtents=<10 40 40>
      constraintSetup {angularLimitsMax=<0 5 15>, angularLimitsMin=<0 -25 -15>, angularTarget=<1 0 0>}
      localJointOffset=<5 0 0>
     }]
    relativeSpaceBone {boneName=attach_cape}
   }
   n66 = Anim Dynamics - Bone: dyn_cape_mid_r_03(ComponentPose=@n65, Alpha=1, ExternalForce="0,0,0") {
    angularDampingOverride=1
    angularSpringConstant=10000
    bAngularSpring=true
    bOverrideAngularDamping=true
    bUseGravityOverride=true
    gravityScale=0.35
    simulationSpace=BoneRelative
    alphaBoolBlend {blendOption=Linear}
    boundBone {boneName=dyn_cape_mid_r_03}
    componentAppliedLinearAccClamp=<3000 3000 3000>
    componentLinearAccScale=<0.7 0.7 0.7>
    componentLinearVelScale=<0.7 0.7 0.7>
    gravityOverride {z=-980}
    physicsBodyDefinitions [{
      boundBone {boneName=dyn_cape_mid_r_03}
      boxExtents=<10 40 40>
      constraintSetup {angularLimitsMax=<0 5 15>, angularLimitsMin=<0 -25 -15>, angularTarget=<1 0 0>}
      localJointOffset=<5 0 0>
     }]
    relativeSpaceBone {boneName=attach_cape}
   }
   n67 = Anim Dynamics - Bone: dyn_cape_mid_r_04(ComponentPose=@n66, Alpha=1, ExternalForce="0,0,0") {
    angularDampingOverride=1
    angularSpringConstant=10000
    bAngularSpring=true
    bOverrideAngularDamping=true
    bUseGravityOverride=true
    gravityScale=0.35
    simulationSpace=BoneRelative
    alphaBoolBlend {blendOption=Linear}
    boundBone {boneName=dyn_cape_mid_r_04}
    componentAppliedLinearAccClamp=<3000 3000 3000>
    componentLinearAccScale=<0.7 0.7 0.7>
    componentLinearVelScale=<0.7 0.7 0.7>
    gravityOverride {z=-980}
    physicsBodyDefinitions [{
      boundBone {boneName=dyn_cape_mid_r_04}
      boxExtents=<10 40 40>
      constraintSetup {angularLimitsMax=<0 5 15>, angularLimitsMin=<0 -25 -15>, angularTarget=<1 0 0>}
      localJointOffset=<5 0 0>
     }]
    relativeSpaceBone {boneName=attach_cape}
   }
   n68 = Anim Dynamics - Bone: dyn_cape_mid_r_05(ComponentPose=@n67, Alpha=1, ExternalForce="0,0,0") {
    angularDampingOverride=1
    angularSpringConstant=10000
    bAngularSpring=true
    bOverrideAngularDamping=true
    bUseGravityOverride=true
    gravityScale=0.35
    simulationSpace=BoneRelative
    alphaBoolBlend {blendOption=Linear}
    boundBone {boneName=dyn_cape_mid_r_05}
    componentAppliedLinearAccClamp=<3000 3000 3000>
    componentLinearAccScale=<0.7 0.7 0.7>
    componentLinearVelScale=<0.7 0.7 0.7>
    gravityOverride {z=-980}
    physicsBodyDefinitions [{
      boundBone {boneName=dyn_cape_mid_r_05}
      boxExtents=<10 40 40>
      constraintSetup {angularLimitsMax=<0 5 15>, angularLimitsMin=<0 -25 -15>, angularTarget=<1 0 0>}
      localJointOffset=<5 0 0>
     }]
    relativeSpaceBone {boneName=attach_cape}
   }
   n69 = Anim Dynamics - Bone: dyn_cape_mid_r_06(ComponentPose=@n68, Alpha=1, ExternalForce="0,0,0") {
    angularDampingOverride=1
    angularSpringConstant=10000
    bAngularSpring=true
    bOverrideAngularDamping=true
    bUseGravityOverride=true
    gravityScale=0.35
    simulationSpace=BoneRelative
    alphaBoolBlend {blendOption=Linear}
    boundBone {boneName=dyn_cape_mid_r_06}
    componentAppliedLinearAccClamp=<3000 3000 3000>
    componentLinearAccScale=<0.7 0.7 0.7>
    componentLinearVelScale=<0.7 0.7 0.7>
    gravityOverride {z=-980}
    physicsBodyDefinitions [{
      boundBone {boneName=dyn_cape_mid_r_06}
      boxExtents=<10 40 40>
      constraintSetup {angularLimitsMax=<0 5 15>, angularLimitsMin=<0 -25 -15>, angularTarget=<1 0 0>}
      localJointOffset=<5 0 0>
     }]
    relativeSpaceBone {boneName=attach_cape}
   }
   n70 = Anim Dynamics - Bone: dyn_cape_r_01(ComponentPose=@n69, Alpha=1, ExternalForce="0,0,0") {
    angularDampingOverride=1
    angularSpringConstant=10000
    bAngularSpring=true
    bOverrideAngularDamping=true
    bUseGravityOverride=true
    gravityScale=0.35
    simulationSpace=BoneRelative
    alphaBoolBlend {blendOption=Linear}
    boundBone {boneName=dyn_cape_r_01}
    componentAppliedLinearAccClamp=<3000 3000 3000>
    componentLinearAccScale=<0.7 0.7 0.7>
    componentLinearVelScale=<0.7 0.7 0.7>
    gravityOverride {z=-980}
    physicsBodyDefinitions [{
      boundBone {boneName=dyn_cape_r_01}
      boxExtents=<10 40 40>
      constraintSetup {angularLimitsMax=<0 5 15>, angularLimitsMin=<0 -25 -15>, angularTarget=<1 0 0>}
      localJointOffset=<5 0 0>
     }]
    relativeSpaceBone {boneName=attach_cape}
   }
   n71 = Anim Dynamics - Bone: dyn_cape_r_02(ComponentPose=@n70, Alpha=1, ExternalForce="0,0,0") {
    angularDampingOverride=1
    angularSpringConstant=10000
    bAngularSpring=true
    bOverrideAngularDamping=true
    bUseGravityOverride=true
    gravityScale=0.35
    simulationSpace=BoneRelative
    alphaBoolBlend {blendOption=Linear}
    alphaScaleBiasClamp {outRange {min=0.6}}
    boundBone {boneName=dyn_cape_r_02}
    componentAppliedLinearAccClamp=<3000 3000 3000>
    componentLinearAccScale=<0.7 0.7 0.7>
    componentLinearVelScale=<0.7 0.7 0.7>
    gravityOverride {z=-980}
    physicsBodyDefinitions [{
      boundBone {boneName=dyn_cape_r_02}
      boxExtents=<10 40 40>
      constraintSetup {angularLimitsMax=<0 5 15>, angularLimitsMin=<0 -25 -15>, angularTarget=<1 0 0>}
      localJointOffset=<5 0 0>
     }]
    relativeSpaceBone {boneName=attach_cape}
   }
   n72 = Anim Dynamics - Bone: dyn_cape_r_03(ComponentPose=@n71, Alpha=1, ExternalForce="0,0,0") {
    angularDampingOverride=1
    angularSpringConstant=10000
    bAngularSpring=true
    bOverrideAngularDamping=true
    bUseGravityOverride=true
    gravityScale=0.35
    simulationSpace=BoneRelative
    alphaBoolBlend {blendOption=Linear}
    boundBone {boneName=dyn_cape_r_03}
    componentAppliedLinearAccClamp=<3000 3000 3000>
    componentLinearAccScale=<0.7 0.7 0.7>
    componentLinearVelScale=<0.7 0.7 0.7>
    gravityOverride {z=-980}
    physicsBodyDefinitions [{
      boundBone {boneName=dyn_cape_r_03}
      boxExtents=<10 40 40>
      constraintSetup {angularLimitsMax=<0 5 15>, angularLimitsMin=<0 -25 -15>, angularTarget=<1 0 0>}
      localJointOffset=<5 0 0>
     }]
    relativeSpaceBone {boneName=attach_cape}
   }
   n73 = Anim Dynamics - Bone: dyn_cape_r_04(ComponentPose=@n72, Alpha=1, ExternalForce="0,0,0") {
    angularDampingOverride=1
    angularSpringConstant=10000
    bAngularSpring=true
    bOverrideAngularDamping=true
    bUseGravityOverride=true
    gravityScale=0.35
    simulationSpace=BoneRelative
    alphaBoolBlend {blendOption=Linear}
    boundBone {boneName=dyn_cape_r_04}
    componentAppliedLinearAccClamp=<3000 3000 3000>
    componentLinearAccScale=<0.7 0.7 0.7>
    componentLinearVelScale=<0.7 0.7 0.7>
    gravityOverride {z=-980}
    physicsBodyDefinitions [{
      boundBone {boneName=dyn_cape_r_04}
      boxExtents=<10 40 40>
      constraintSetup {angularLimitsMax=<0 5 15>, angularLimitsMin=<0 -25 -15>, angularTarget=<1 0 0>}
      localJointOffset=<5 0 0>
     }]
    relativeSpaceBone {boneName=attach_cape}
   }
   n74 = Anim Dynamics - Bone: dyn_cape_r_05(ComponentPose=@n73, Alpha=1, ExternalForce="0,0,0") {
    angularDampingOverride=1
    angularSpringConstant=10000
    bAngularSpring=true
    bOverrideAngularDamping=true
    bUseGravityOverride=true
    gravityScale=0.35
    simulationSpace=BoneRelative
    alphaBoolBlend {blendOption=Linear}
    boundBone {boneName=dyn_cape_r_05}
    componentAppliedLinearAccClamp=<3000 3000 3000>
    componentLinearAccScale=<0.7 0.7 0.7>
    componentLinearVelScale=<0.7 0.7 0.7>
    gravityOverride {z=-980}
    physicsBodyDefinitions [{
      boundBone {boneName=dyn_cape_r_05}
      boxExtents=<10 40 40>
      constraintSetup {angularLimitsMax=<0 5 15>, angularLimitsMin=<0 -25 -15>, angularTarget=<1 0 0>}
      localJointOffset=<5 0 0>
     }]
    relativeSpaceBone {boneName=attach_cape}
   }
   n75 = Anim Dynamics - Bone: dyn_cape_r_06(ComponentPose=@n74, Alpha=1, ExternalForce="0,0,0") {
    angularDampingOverride=1
    angularSpringConstant=10000
    bAngularSpring=true
    bOverrideAngularDamping=true
    bUseGravityOverride=true
    gravityScale=0.35
    simulationSpace=BoneRelative
    alphaBoolBlend {blendOption=Linear}
    boundBone {boneName=dyn_cape_r_06}
    componentAppliedLinearAccClamp=<3000 3000 3000>
    componentLinearAccScale=<0.7 0.7 0.7>
    componentLinearVelScale=<0.7 0.7 0.7>
    gravityOverride {z=-980}
    physicsBodyDefinitions [{
      boundBone {boneName=dyn_cape_r_06}
      boxExtents=<10 40 40>
      constraintSetup {angularLimitsMax=<0 5 15>, angularLimitsMin=<0 -25 -15>, angularTarget=<1 0 0>}
      localJointOffset=<5 0 0>
     }]
    relativeSpaceBone {boneName=attach_cape}
   }
   Save cached pose 'Post Anim Dynamics'(Pose=Component To Local(ComponentPose=@n75)) {cacheName=Post Anim Dynamics}
   n76 = RigidBody(
    ComponentPose=Local To Component(LocalPose=Use cached pose 'Post Deform Rig'())
    OverrideWorldGravity="0,0,-980"
    SimSpaceSettings=(WorldAlpha=1,VelocityScaleZ=1,DampingAlpha=0,MaxLinearVelocity=800,MaxAngularVelocity=6,MaxLinearAcceleration=4000,MaxAngularAcceleration=30,ExternalLinearDragV=(X=0.5,Y=0.5,Z=0.5),ExternalLinearVelocity=(X=0,Y=0,Z=0),ExternalAngularVelocity=(X=0,Y=0,Z=0))
   ) {
    bOverrideWorldGravity=true
    overridePhysicsAsset=&/M_MED_MoleAcornPecan_Pack_CapePhysics_Simple
    simulationSpace=BaseBoneSpace
    alphaBoolBlend {blendOption=Linear}
    baseBoneRef {boneName=root}
    overrideWorldGravity {z=-980}
    simSpaceSettings {
     dampingAlpha=0
     maxAngularAcceleration=30
     maxAngularVelocity=6
     maxLinearAcceleration=4000
     maxLinearVelocity=800
     worldAlpha=1
     externalLinearDragV=<0.5 0.5 0.5>
    }
   }
   n77 = "Transform (Modify) Bone - Bone: deform_back"(ComponentPose=@n76, bAlphaBoolEnabled=True, Translation="-1.5,0,0", Rotation="0,0,0", Scale="1,1,1") {
    alphaInputType=Bool
    translationMode=BMM_Additive
    translationSpace=BCS_BoneSpace
    alphaBoolBlend {blendOption=Linear}
    boneToModify {boneName=deform_back}
    translation {x=-1.5}
   }
   n78 = "Constraint - dyn_hood_bk_drape_01 (PP)"(ComponentPose=@n77, Alpha=1, ConstraintWeights_0=0.5, ConstraintWeights_1=0.5) {
    alphaBoolBlend {blendOption=Linear}
    boneToModify {boneName=dyn_hood_bk_drape_01}
    constraintSetup [{targetBone {boneName=dyn_cape_mid_l_01}}, {targetBone {boneName=dyn_cape_mid_r_01}}]
    constraintWeights [1,1]
   }
   n79 = "Constraint - dyn_hood_bk_drape_02 (PP)"(ComponentPose=@n78, Alpha=1, ConstraintWeights_0=0.5, ConstraintWeights_1=0.5) {
    alphaBoolBlend {blendOption=Linear}
    boneToModify {boneName=dyn_hood_bk_drape_02}
    constraintSetup [{targetBone {boneName=dyn_cape_mid_l_01}}, {targetBone {boneName=dyn_cape_mid_r_01}}]
    constraintWeights [1,1]
   }
   n80 = "Constraint - dyn_hood_bk_drape_end (PP)"(ComponentPose=@n79, Alpha=1, ConstraintWeights_0=0.5, ConstraintWeights_1=0.5) {
    alphaBoolBlend {blendOption=Linear}
    boneToModify {boneName=dyn_hood_bk_drape_end}
    constraintSetup [{targetBone {boneName=dyn_cape_mid_l_02}}, {targetBone {boneName=dyn_cape_mid_r_02}}]
    constraintWeights [1,1]
   }
   n81 = "Constraint - dyn_cape_l_01 (P)"(ComponentPose=@n80, Alpha=1, ConstraintWeights_0=1) {
    alphaBoolBlend {blendOption=Linear}
    boneToModify {boneName=dyn_cape_l_01}
    constraintSetup [{perAxis {bX=false, bZ=false}, targetBone {boneName=spine_05}}]
    constraintWeights [1]
   }
   n82 = "Constraint - dyn_cape_mid_l_01 (P)"(ComponentPose=@n81, Alpha=1, ConstraintWeights_0=1) {
    alphaBoolBlend {blendOption=Linear}
    boneToModify {boneName=dyn_cape_mid_l_01}
    constraintSetup [{perAxis {bX=false, bZ=false}, targetBone {boneName=spine_05}}]
    constraintWeights [1]
   }
   n83 = "Constraint - dyn_cape_mid_r_01 (P)"(ComponentPose=@n82, Alpha=1, ConstraintWeights_0=1) {
    alphaBoolBlend {blendOption=Linear}
    boneToModify {boneName=dyn_cape_mid_r_01}
    constraintSetup [{perAxis {bX=false, bZ=false}, targetBone {boneName=spine_05}}]
    constraintWeights [1]
   }
   n84 = "Constraint - dyn_cape_r_01 (P)"(ComponentPose=@n83, Alpha=1, ConstraintWeights_0=1) {
    alphaBoolBlend {blendOption=Linear}
    boneToModify {boneName=dyn_cape_r_01}
    constraintSetup [{perAxis {bX=false, bZ=false}, targetBone {boneName=spine_05}}]
    constraintWeights [1]
   }
   n85 = "Constraint - dyn_cape_l_02 (P)"(ComponentPose=@n84, Alpha=1, ConstraintWeights_0=1) {
    alphaBoolBlend {blendOption=Linear}
    boneToModify {boneName=dyn_cape_l_02}
    constraintSetup [{perAxis {bX=false, bZ=false}, targetBone {boneName=spine_04}}]
    constraintWeights [1]
   }
   n86 = "Constraint - dyn_cape_mid_l_02 (P)"(ComponentPose=@n85, Alpha=1, ConstraintWeights_0=1) {
    alphaBoolBlend {blendOption=Linear}
    boneToModify {boneName=dyn_cape_mid_l_02}
    constraintSetup [{perAxis {bX=false, bZ=false}, targetBone {boneName=spine_04}}]
    constraintWeights [1]
   }
   n87 = "Constraint - dyn_cape_mid_r_02 (P)"(ComponentPose=@n86, Alpha=1, ConstraintWeights_0=1) {
    alphaBoolBlend {blendOption=Linear}
    boneToModify {boneName=dyn_cape_mid_r_02}
    constraintSetup [{perAxis {bX=false, bZ=false}, targetBone {boneName=spine_04}}]
    constraintWeights [1]
   }
   n88 = "Constraint - dyn_cape_r_02 (P)"(ComponentPose=@n87, Alpha=1, ConstraintWeights_0=1) {
    alphaBoolBlend {blendOption=Linear}
    boneToModify {boneName=dyn_cape_r_02}
    constraintSetup [{perAxis {bX=false, bZ=false}, targetBone {boneName=spine_04}}]
    constraintWeights [1]
   }
   Save cached pose 'Simple Cape with Pack for Cloth'(Pose=Component To Local(ComponentPose=@n88)) {cacheName=Simple Cape with Pack for Cloth}
   n89 = RigidBody(
    ComponentPose=Local To Component(LocalPose=Use cached pose 'Post Deform Rig'())
    OverrideWorldGravity="0,0,-980"
    SimSpaceSettings=(WorldAlpha=1,VelocityScaleZ=1,DampingAlpha=0,MaxLinearVelocity=800,MaxAngularVelocity=6,MaxLinearAcceleration=4000,MaxAngularAcceleration=30,ExternalLinearDragV=(X=0.5,Y=0.5,Z=0.5),ExternalLinearVelocity=(X=0,Y=0,Z=0),ExternalAngularVelocity=(X=0,Y=0,Z=0))
   ) {
    bOverrideWorldGravity=true
    overridePhysicsAsset=&/M_MED_MoleAcornPecan_CapePhysics_Simple
    simulationSpace=BaseBoneSpace
    alphaBoolBlend {blendOption=Linear}
    baseBoneRef {boneName=root}
    overrideWorldGravity {z=-980}
    simSpaceSettings {
     dampingAlpha=0
     maxAngularAcceleration=30
     maxAngularVelocity=6
     maxLinearAcceleration=4000
     maxLinearVelocity=800
     worldAlpha=1
     externalLinearDragV=<0.5 0.5 0.5>
    }
   }
   n90 = "Transform (Modify) Bone - Bone: deform_back"(ComponentPose=@n89, bAlphaBoolEnabled=True, Translation="-1.5,0,0", Rotation="0,0,0", Scale="1,1,1") {
    alphaInputType=Bool
    translationMode=BMM_Additive
    translationSpace=BCS_BoneSpace
    alphaBoolBlend {blendOption=Linear}
    boneToModify {boneName=deform_back}
    translation {x=-1.5}
   }
   n91 = "Constraint - dyn_hood_bk_drape_01 (PP)"(ComponentPose=@n90, Alpha=1, ConstraintWeights_0=0.5, ConstraintWeights_1=0.5) {
    alphaBoolBlend {blendOption=Linear}
    boneToModify {boneName=dyn_hood_bk_drape_01}
    constraintSetup [{targetBone {boneName=dyn_cape_mid_l_01}}, {targetBone {boneName=dyn_cape_mid_r_01}}]
    constraintWeights [1,1]
   }
   n92 = "Constraint - dyn_hood_bk_drape_02 (PP)"(ComponentPose=@n91, Alpha=1, ConstraintWeights_0=0.5, ConstraintWeights_1=0.5) {
    alphaBoolBlend {blendOption=Linear}
    boneToModify {boneName=dyn_hood_bk_drape_02}
    constraintSetup [{targetBone {boneName=dyn_cape_mid_l_01}}, {targetBone {boneName=dyn_cape_mid_r_01}}]
    constraintWeights [1,1]
   }
   n93 = "Constraint - dyn_hood_bk_drape_end (PP)"(ComponentPose=@n92, Alpha=1, ConstraintWeights_0=0.5, ConstraintWeights_1=0.5) {
    alphaBoolBlend {blendOption=Linear}
    boneToModify {boneName=dyn_hood_bk_drape_end}
    constraintSetup [{targetBone {boneName=dyn_cape_mid_l_02}}, {targetBone {boneName=dyn_cape_mid_r_02}}]
    constraintWeights [1,1]
   }
   Save cached pose 'Simple Cape for Cloth'(Pose=Component To Local(ComponentPose=@n93)) {cacheName=Simple Cape for Cloth}
   Save cached pose 'Simple Cape for Cloth Final'(Pose=Blend Poses by bool(
     BlendPose_0=Use cached pose 'Simple Cape with Pack for Cloth'()
     BlendPose_1=Use cached pose 'Simple Cape for Cloth'()
     bActiveValue=Get isWearingPack()
    ) {blendType=Linear}) {cacheName=Simple Cape for Cloth Final}
   Use cached pose 'Cape Final'()
   Use cached pose 'Post Deform Rig'()
   Reroute Node(InputPin:Wildcard) -> (OutputPin:Wildcard)
   Reroute Node()
  ]}
 EventGraph {nodes [
   n1 = Event BlueprintUpdateAnimation()
   > n2 = Sequence!(execute) -> (then_0, then_1, then_2:Exec)
   n3 = Get bClothEnabled()
   > Branch(Condition=@n3)
   > SetClothingInteractorIfNotSet()
   > n4 = Cloth Controller Function DEPRECATED!(
    execute
    CharacterPartAnimInstance:Custom Character Part Anim Instance Object Reference=Self-Reference()
    ClothingAssetName:String=M_MED_MoleAcornPecan_Clothing_0
    Cloth Paramaters:Cloth Params Struct Structure=Get Cloth Paramaters() -> (Cloth Paramaters:Cloth Params Struct Structure)
    Cloth Core Settings:Cloth Core Setting Struct Structure=Get Cloth Core Settings() -> (Cloth Core Settings:Cloth Core Setting Struct Structure)
    ClothUpAxis:Vector="0,0,1"
    OnGround_Moving_SpeedLimit:double=2
    OnGround_Moving_AngularVel_Max:double=0.05
    OnGround_Moving_FictitiousVel_Max:double=0.6
    LinVelScale_FwdBwd_Clamp:double=0
    LinVelScale_UpDwn_Clamp:double=0.7
    DeltaTime:double=@n1.DeltaTimeX
    GameTimeInSeconds:double=GetGameTimeInSeconds()
    WindGustsParams:Wind Gust Params Struct Structure=Get Wind Gusts Params() -> (Wind Gusts Params:Wind Gust Params Struct Structure)
    WindGustLocalSettings:Wind Gust Local Setting Struct Structure=Get Wind Gust Local Settings() -> (Wind Gust Local Settings:Wind Gust Local Setting Struct Structure)
   ) -> (
    then
    CurrentStateOut:String
    Cloth Settings Out:Cloth Setting Struct Structure
    Cloth Core Settings Out:Cloth Core Setting Struct Structure
    WindGustSettings Out:Wind Gust Setting Struct Structure
    WindGustLocalSettings Out:Wind Gust Local Setting Struct Structure
    GustStrength - WindVelocityOut:Vector
    LinearVelVector-DELETE LATER:Vector
   )
   > Set Wind Gust Local Settings(Wind Gust Local Settings:Wind Gust Local Setting Struct Structure=@n4.WindGustLocalSettings Out) -> (Output_Get:Wind Gust Local Setting Struct Structure)
   n5 = BreakRotator(InRot=Get PawnRotationRate() -> (PawnRotationRate:Rotator))
   > Branch(Condition="float > float"(A="Absolute (Float)"(A=@n5.Yaw), B=5) {is=title K2Node_CallFunction})
   > SetClothingInteractorIfNotSet()
   > SetVelocityScale!(
    execute
    self:Chaos Clothing Interactor Object Reference=Get ChaosClothingInteractor() -> (ChaosClothingInteractor:Chaos Clothing Interactor Object Reference)
    LinearVelocityScale:Vector="0.05,0.05,0.05"
    AngularVelocityScale:float=0
    FictitiousAngularScale:float=0
   ) -> (then)
   Set Slope Rotation(
    execute=@n2.then_1
    Slope Rotation:Rotator=Break Fort Anim Input Slope Info!(FortAnimInput_SlopeInfo:"Fort Anim Input Slope Info Structure (by ref)"=Get SlopeInfoInput() -> (SlopeInfoInput:Fort Anim Input Slope Info Structure)) -> (RootSlopeTranslation:Vector=~, LeftFootSlopeTranslation:Vector=~, RightFootSlopeTranslation:Vector=~, SlopeRotation:Rotator=~) {is=title K2Node_BreakStruct}.SlopeRotation
   ) -> (Output_Get:Rotator)
   > n6 = "Gravity Override State Function (Parameter Asset) (Nativized)"(
    CharacterPartAnimInstance=Self-Reference()
    GameTimeInSeconds=GetGameTimeInSeconds()
    Parameters=&/M_MED_MoleAcornPecan_Parameters
   )
   > Set Gravity Override(Gravity Override=@n6)
   > Set RBAN Settings Out(RBAN Settings Out=@n6.OutRigidBodyAnimNodeParameters)
   n7 = Branch(execute=@n2.then_2, Condition=AND Boolean!(A=Get CapeOn(), B=Get isWearingPack(), C:bool=Get bIsSkyDiving()) -> (ReturnValue))
   > Set RBAN Alpha(RBAN Alpha:double=0.25) -> (Output_Get:double)
   Set RBAN Alpha(execute=@n7.else, RBAN Alpha=1)
   n8 = Event OnCharacterCustomizationLoaded()
   n9 = CheckForBackbling()
   > Set isWearingPack(isWearingPack:bool=OR Boolean!(A=@n9, B=@n9.bIsBackblingACape, C:bool=@n9.bIsBackblingLong) -> (ReturnValue)) -> (Output_Get:bool)
   > Set isWearingCape(isWearingCape=@n9.bIsBackblingLong)
   > n10 = Sequence()
   > n11 = Branch(Condition=CharacterItemDefinitionHasAnyMetaTag(Tags=Make Array("[0]"=(TagName="Cosmetics.Variant.Property.Stage2"))))
   > Set AdjustBackpackForEars(AdjustBackpackForEars:bool=true) -> (Output_Get:bool)
   > Set HoodClothCollidersConstAlpha(HoodClothCollidersConstAlpha:double=1) -> (Output_Get:double)
   Set AdjustBackpackForEars(execute=@n11.else, AdjustBackpackForEars=false)
   > Set HoodClothCollidersConstAlpha(HoodClothCollidersConstAlpha=0)
   Set Face Acc Component(
    execute=@n10.then_2
    Face Acc Component:Skeletal Mesh Component Budgeted Object Reference=Get Array Item with EFort Custom Part Type(Array=@n8.CustomMeshParts, PartType=Face)
   ) -> (Output_Get:Skeletal Mesh Component Budgeted Object Reference)
   n12 = Branch(execute=@n10.then_1, Condition=CharacterItemDefinitionHasAnyMetaTag(Tags=Make Array("[0]"=(TagName="Cosmetics.Variant.Property.Stage3"))))
   > Set CapeOn(CapeOn:bool=false) -> (Output_Get:bool)
   Set CapeOn(execute=@n12.else, CapeOn=true)
   Event OnEndFalling()
   > SetClothingInteractorIfNotSet()
   > Branch(Condition=Get resetClothLandingFromZilpine() -> (resetClothLandingFromZilpine:bool))
   n13 = Get ChaosClothingInteractor()
   > ResetAndTeleport!(execute, self:Chaos Clothing Interactor Object Reference=@n13, bReset:bool=true, bTeleport:bool=true) -> (then)
   > Set resetClothLandingFromZilpine(resetClothLandingFromZilpine:bool=false) -> (Output_Get:bool)
   n14 = Event BlueprintProcessAnimInputEvent()
   n15 = IsNamed(self=@n14.AnimInputEvent, InEventName=aie.Player.Zipline.Entered)
   > n16 = Branch(Condition=@n15)
   > Set IsOnZipline(IsOnZipline:bool=true) -> (Output_Get:bool)
   > Branch(Condition=Get bClothEnabled())
   > SetClothingInteractorIfNotSet()
   > n17 = Branch(Condition=@n15)
   > Delay(Duration=0.2)
   > ResetAndTeleport(self=@n13, bReset=true, bTeleport=true)
   n18 = IsNamed(self=@n14.AnimInputEvent, InEventName=aie.Player.Zipline.Exited)
   Branch(execute=@n17.else, Condition=@n18)
   > Set resetClothLandingFromZilpine(resetClothLandingFromZilpine=true)
   > Delay(Duration=1.1)
   Branch(execute=@n16.else, Condition=@n18)
   > Set IsOnZipline(IsOnZipline=false)
   AND Boolean!(A=@n3, B=Get CapeOn()) -> (ReturnValue)
  ]}
 SetClothingInteractorIfNotSet {nodes [
   SetClothingInteractorIfNotSet() {is=title K2Node_FunctionEntry}
   n1 = IsValid(Object=Get ChaosClothingInteractor())
   > Branch(Condition=AND Boolean(A=Get bClothEnabled(), B=NOT Boolean(A=@n1)))
   n2 = MakeLiteralString(Value=M_MED_MoleAcornPecan_Clothing_0)
   n3 = GetOwningComponent()
   > n4 = Cast To ChaosClothingInteractor!(
    execute
    Object=GetClothingInteractor!(
     self:Clothing Simulation Interactor Object Reference=GetClothingSimulationInteractor!(self:Skeletal Mesh Component Object Reference=@n3, ClothingAssetName:Name=None) -> (ReturnValue:Clothing Simulation Interactor Object Reference)
     ClothingAssetName:Name=String To Name(InString=@n2)
    ) -> (ReturnValue:Clothing Interactor Object Reference)
   ) -> (then, CastFailed, AsChaos Clothing Interactor:Chaos Clothing Interactor Object Reference) {is=title K2Node_DynamicCast}
   > Set ChaosClothingInteractor(ChaosClothingInteractor:Chaos Clothing Interactor Object Reference=@n4) -> (Output_Get:Chaos Clothing Interactor Object Reference)
  ]}
]
variables [
 AdjustBackpackForEars {type=boolean}
 CapeOn {type=boolean}
 ChaosClothingInteractor {type=ChaosClothingInteractor, default=null}
 Cloth Core Settings {
  type=ClothCoreSettingStruct
  default {
   collisionThickness=0.2
   dampingCoefficient=0.1
   frictionCoefficient=0.8
   localDampingCoefficient=0
   numberIterations=2
   numberSubsteps=1
   selfCollisionThickness=0
   animDriveDamping=<0 1>
   animDriveStiffness=<0 1>
   areaStiffness=<0.5 0.5>
   bendingStiffness=<1 1>
   edgeStiffness=<1 1>
   tetherStiffness=<1 1>
  }
 }
 Cloth Paramaters {
  type=ClothParamsStruct
  default {
   dBNO {
    angularVelocityScale=0.2
    fictitiousAngularScale=0
    jointName=pelvis
    dragCoefficient=<0.2 1>
    gravityOverride=<-800 0 0>
    liftCoefficient=<0 0.5>
    linearVelocityScale=<0.3 0.3 0.1>
   }
   dBNO_Carried {
    angularVelocityScale=0.6
    fictitiousAngularScale=0.1
    jointName=pelvis
    dragCoefficient=<0.2 1>
    gravityOverride=<-980 -200 0>
    liftCoefficient=<0 0.5>
    linearVelocityScale=<0.3 0.3 0.1>
   }
   dBNO_Impulse {
    angularVelocityScale=0
    fictitiousAngularScale=0
    jointName=pelvis
    dragCoefficient=<0 0.2>
    gravityOverride=<-3000 -2000 0>
    liftCoefficient=<0 0.1>
    linearVelocityScale=<0.3 0.1 0.3>
   }
   dBNO_RollBwd {
    angularVelocityScale=0
    fictitiousAngularScale=0
    jointName=pelvis
    dragCoefficient=<0 0.2>
    gravityOverride=<-2000 0 0>
    liftCoefficient=<0 0.1>
    linearVelocityScale=<0.3 0.1 0.3>
   }
   dBNO_RollFwd {
    angularVelocityScale=0
    fictitiousAngularScale=0
    jointName=pelvis
    dragCoefficient=<0 0.2>
    gravityOverride=<-980 0 0>
    liftCoefficient=<0 0.1>
    linearVelocityScale=<0.3 0.1 0.3>
   }
   dBNO_Tumbling {
    angularVelocityScale=0
    fictitiousAngularScale=0
    jointName=pelvis
    dragCoefficient=<0 0.2>
    gravityOverride=<-2500 -300 0>
    liftCoefficient=<0 0.1>
    linearVelocityScale=<0.3 0.1 0.3>
   }
   emoteOrMelee {
    angularVelocityScale=0.6
    fictitiousAngularScale=0.1
    jointName=root
    dragCoefficient=<2.5 2.5>
    gravityOverride=<0 0 -980>
    liftCoefficient=<0.5 0.5>
    linearVelocityScale=<0.6 0.6 0.5>
   }
   front_End {
    angularVelocityScale=0.1
    fictitiousAngularScale=0
    jointName=root
    dragCoefficient=<0.1 0.1>
    gravityOverride=<0 -100 -980>
    liftCoefficient=<0.1 0.1>
    linearVelocityScale=<0.1 0.1 0.1>
   }
   onGround_Crouch_Moving {
    angularVelocityScale=0.1
    fictitiousAngularScale=0.3
    jointName=root
    dragCoefficient=<0.2 2>
    gravityOverride=<0 0 -980>
    liftCoefficient=<0 0.5>
    linearVelocityScale=<0.6 0.6 0.5>
   }
   onGround_Falling {
    angularVelocityScale=0.6
    fictitiousAngularScale=0
    jointName=root
    dragCoefficient=<0 0.2>
    gravityOverride=<0 -500 -980>
    liftCoefficient=<0 0.1>
    linearVelocityScale=<2 1 0.3>
   }
   onGround_Moving {
    angularVelocityScale=0.6
    fictitiousAngularScale=0.1
    jointName=root
    dragCoefficient=<0.5 0.75>
    gravityOverride=<0 -200 -980>
    liftCoefficient=<0.5 0.5>
    linearVelocityScale=<0.6 0.6 0.5>
   }
   onGround_Standing {
    angularVelocityScale=0.3
    fictitiousAngularScale=0.1
    jointName=root
    dragCoefficient=<0.2 0.5>
    gravityOverride=<50 -200 -980>
    liftCoefficient=<0 0.5>
    linearVelocityScale=<0.6 0.6 0.1>
   }
   rail_Grinding {
    angularVelocityScale=0.2
    fictitiousAngularScale=0
    jointName=root
    dragCoefficient=<0.5 0.5>
    gravityOverride=<50 -500 -400>
    liftCoefficient=<0.25 0.25>
    linearVelocityScale=<0.1 0.1 0.1>
   }
   riding_Moving {
    angularVelocityScale=0.05
    fictitiousAngularScale=1
    jointName=root
    dragCoefficient=<0.5 0.5>
    gravityOverride=<0 0 -980>
    liftCoefficient=<0.5 0.5>
    linearVelocityScale=<0.4 0.6 0.3>
   }
   riding_Moving_Targeting {
    angularVelocityScale=0.05
    fictitiousAngularScale=0.7
    jointName=root
    dragCoefficient=<0.5 0.5>
    gravityOverride=<0 0 -980>
    liftCoefficient=<0.5 0.5>
    linearVelocityScale=<0.2 0.6 0.2>
   }
   riding_Standing {
    angularVelocityScale=0.05
    fictitiousAngularScale=1
    jointName=root
    dragCoefficient=<0.5 0.5>
    gravityOverride=<0 -450 -980>
    liftCoefficient=<0.5 0.5>
    linearVelocityScale=<0.6 0.6 0.3>
   }
   skydive_DivingDown {
    angularVelocityScale=0.1
    fictitiousAngularScale=0
    jointName=pelvis
    dragCoefficient=<0.4 0.4>
    gravityOverride=<-1000 0 1000>
    liftCoefficient=<0.3 0.3>
    linearVelocityScale=<0.2 0.2 0.1>
   }
   skydive_FloatingDown {
    angularVelocityScale=0.1
    fictitiousAngularScale=0.1
    jointName=pelvis
    dragCoefficient=<0.75 0.75>
    gravityOverride=<-1000 -500 0>
    liftCoefficient=<0.5 0.5>
    linearVelocityScale=<0.2 0.2 0.2>
   }
   skydive_GoingUp {
    angularVelocityScale=0.1
    fictitiousAngularScale=0
    jointName=root
    dragCoefficient=<0.01 0.01>
    gravityOverride=<0 -500 1000>
    liftCoefficient=<0.05 0.05>
    linearVelocityScale=<0.1 0.1 0.01>
   }
   skydive_Parachuting {
    angularVelocityScale=0.1
    fictitiousAngularScale=0
    jointName=root
    dragCoefficient=<0.75 0.75>
    gravityOverride=<0 -200 -800>
    liftCoefficient=<0.5 0.5>
    linearVelocityScale=<0.2 0.2 0.2>
   }
   skydive_Surfing {
    angularVelocityScale=0.1
    fictitiousAngularScale=0.1
    jointName=root
    dragCoefficient=<1 1>
    gravityOverride=<0 -400 -980>
    liftCoefficient=<0.5 0.5>
    linearVelocityScale=<0.2 0.2 0.1>
   }
   swimming {
    angularVelocityScale=0.6
    fictitiousAngularScale=0
    jointName=root
    dragCoefficient=<2.5 2.5>
    gravityOverride=<0 0 0>
    liftCoefficient=<0.1 0.1>
    linearVelocityScale=<0.6 0.6 0.5>
   }
   tactical_Sprint {
    angularVelocityScale=0.6
    fictitiousAngularScale=0.1
    jointName=root
    dragCoefficient=<2 2>
    gravityOverride=<0 0 -980>
    liftCoefficient=<0.5 0.5>
    linearVelocityScale=<0.6 0.6 0.5>
   }
   ziplining {
    angularVelocityScale=0.05
    fictitiousAngularScale=0
    jointName=root
    dragCoefficient=<1 1>
    gravityOverride=<0 0 -980>
    liftCoefficient=<0.6 0.6>
    linearVelocityScale=<0.2 0.2 0.1>
   }
  }
 }
 Face Acc Component {type=SkeletalMeshComponentBudgeted}
 Gravity Override {type=Vector}
 Gravity Override Params {
  type=FortGravityOverrideParameterSet
  default=(Skydive_GoingUp=(GravityOverride=(WindFrequency=(X=1.500000,Y=7.000000,Z=8.000000),WindAmplitude=(X=200.000000,Y=-500.000000,Z=400.000000),GravityOverride=(X=0.000000,Y=-300.000000,Z=1500.000000),JointName="root")),Skydive_FloatingDown=(GravityOverride=(WindFrequency=(X=1.500000,Y=7.000000,Z=8.000000),WindAmplitude=(X=200.000000,Y=-500.000000,Z=400.000000),WindOffset=(X=0.000000,Y=-600.000000,Z=400.000000),GravityOverride=(X=0.000000,Y=-700.000000,Z=600.000000),JointName="root")),Skydive_DivingDown=(GravityOverride=(WindFrequency=(X=1.500000,Y=12.000000,Z=10.000000),JointName="pelvis")),Skydive_Parachuting=(GravityOverride=(WindFrequency=(X=1.500000,Y=7.000000,Z=8.000000),WindAmplitude=(X=200.000000,Y=-500.000000,Z=400.000000),GravityOverride=(X=0.000000,Y=-800.000000,Z=-200.000000),JointName="pelvis")),OnGround_Standing=(GravityOverride=(GravityOverride=(X=0.000000,Y=0.000000,Z=-980.000000),JointName="root")),OnGround_Moving=(GravityOverride=(GravityOverride=(X=0.000000,Y=-400.000000,Z=-980.000000),JointName="root")),OnGround_Moving_HighRotationRate=(GravityOverride=(GravityOverride=(X=0.000000,Y=0.000000,Z=-980.000000),JointName="root")),OnGround_Crouch_Moving=(GravityOverride=(GravityOverride=(X=0.000000,Y=0.000000,Z=-980.000000),JointName="root")),Emote_or_Melee=(GravityOverride=(GravityOverride=(X=0.000000,Y=0.000000,Z=-980.000000),JointName="root")),Swimming=(GravityOverride=(GravityOverride=(X=-800.000000,Y=0.000000,Z=0.000000),JointName="pelvis")),Front_End=(GravityOverride=(GravityOverride=(X=0.000000,Y=0.000000,Z=-980.000000),JointName="root")),DBNO=(GravityOverride=(GravityOverride=(X=-980.000000,Y=400.000000,Z=0.000000),JointName="pelvis")),DBNO_Carried=(GravityOverride=(GravityOverride=(X=-980.000000,Y=500.000000,Z=-500.000000),JointName="pelvis")),OnGround_Falling=(GravityOverride=(GravityOverride=(X=0.000000,Y=0.000000,Z=-980.000000),JointName="root")),Ziplining=(GravityOverride=(GravityOverride=(X=0.000000,Y=0.000000,Z=-980.000000),JointName="root")),Skydive_Surfing=(GravityOverride=(WindFrequency=(X=1.500000,Y=7.000000,Z=8.000000),WindAmplitude=(X=200.000000,Y=-500.000000,Z=400.000000),GravityOverride=(X=500.000000,Y=-300.000000,Z=-980.000000),JointName="root")),Riding_Standing=(GravityOverride=(GravityOverride=(X=0.000000,Y=-350.000000,Z=-980.000000),JointName="root")),Riding_Moving=(GravityOverride=(GravityOverride=(X=0.000000,Y=-450.000000,Z=-450.000000),JointName="root")),Riding_Moving_Targeting=(GravityOverride=(GravityOverride=(X=0.000000,Y=-200.000000,Z=-980.000000),JointName="root")),Tactical_Sprint=(GravityOverride=(GravityOverride=(X=0.000000,Y=-200.000000,Z=-600.000000),JointName="root")),Rail_Grinding=(GravityOverride=(GravityOverride=(X=0.000000,Y=-300.000000,Z=-600.000000),JointName="root")),)
 }
 HoodClothCollidersConstAlpha {type=number}
 IsOnZipline {type=boolean}
 isWearingCape {type=boolean}
 isWearingPack {type=boolean}
 RBAN Alpha {type=number}
 RBAN Params {
  type=FortRigidBodyAnimNodeParameterSet
  default=(Skydive_GoingUp=(RigidBodyAnimNode=(SimSpaceSettings=(WorldAlpha=1.000000,DampingAlpha=0.000000,MaxLinearVelocity=200.000000,MaxAngularVelocity=2.000000,MaxLinearAcceleration=1000.000000,MaxAngularAcceleration=10.000000,ExternalLinearDragV=(X=0.500000,Y=0.500000,Z=0.500000)))),Skydive_FloatingDown=(RigidBodyAnimNode=(SimSpaceSettings=(WorldAlpha=1.000000,DampingAlpha=0.000000,MaxLinearVelocity=200.000000,MaxAngularVelocity=2.000000,MaxLinearAcceleration=1000.000000,MaxAngularAcceleration=10.000000,ExternalLinearDragV=(X=0.500000,Y=0.500000,Z=0.500000)))),Skydive_DivingDown=(RigidBodyAnimNode=(SimSpaceSettings=(WorldAlpha=1.000000,DampingAlpha=0.000000,MaxLinearVelocity=200.000000,MaxAngularVelocity=2.000000,MaxLinearAcceleration=1000.000000,MaxAngularAcceleration=10.000000,ExternalLinearDragV=(X=0.500000,Y=0.500000,Z=0.500000)))),Skydive_Parachuting=(RigidBodyAnimNode=(SimSpaceSettings=(WorldAlpha=1.000000,DampingAlpha=0.000000,MaxLinearVelocity=200.000000,MaxAngularVelocity=2.000000,MaxLinearAcceleration=1000.000000,MaxAngularAcceleration=10.000000,ExternalLinearDragV=(X=0.500000,Y=0.500000,Z=0.500000)))),OnGround_Standing=(RigidBodyAnimNode=(SimSpaceSettings=(WorldAlpha=1.000000,VelocityScaleZ=0.200000,DampingAlpha=0.000000,MaxLinearVelocity=400.000000,MaxAngularVelocity=2.000000,MaxLinearAcceleration=3000.000000,MaxAngularAcceleration=10.000000,ExternalLinearDragV=(X=0.500000,Y=0.500000,Z=0.500000)))),OnGround_Moving=(RigidBodyAnimNode=(SimSpaceSettings=(WorldAlpha=1.000000,VelocityScaleZ=0.600000,DampingAlpha=0.000000,MaxLinearVelocity=500.000000,MaxAngularVelocity=4.000000,MaxLinearAcceleration=3000.000000,MaxAngularAcceleration=20.000000,ExternalLinearDragV=(X=0.500000,Y=0.500000,Z=0.750000)))),OnGround_Moving_HighRotationRate=(RigidBodyAnimNode=(SimSpaceSettings=(WorldAlpha=1.000000,VelocityScaleZ=0.200000,DampingAlpha=0.000000,MaxLinearVelocity=400.000000,MaxAngularVelocity=2.000000,MaxLinearAcceleration=2000.000000,MaxAngularAcceleration=10.000000,ExternalLinearDragV=(X=0.500000,Y=0.500000,Z=0.500000)))),OnGround_Crouch_Moving=(RigidBodyAnimNode=(SimSpaceSettings=(WorldAlpha=1.000000,VelocityScaleZ=0.400000,DampingAlpha=0.000000,MaxLinearVelocity=400.000000,MaxAngularVelocity=7.500000,MaxLinearAcceleration=3000.000000,MaxAngularAcceleration=30.000000,ExternalLinearDragV=(X=0.500000,Y=0.500000,Z=0.250000)))),Emote_or_Melee=(RigidBodyAnimNode=(SimSpaceSettings=(WorldAlpha=1.000000,VelocityScaleZ=0.400000,DampingAlpha=0.000000,MaxLinearVelocity=400.000000,MaxAngularVelocity=7.500000,MaxLinearAcceleration=3000.000000,MaxAngularAcceleration=30.000000,ExternalLinearDragV=(X=0.500000,Y=0.500000,Z=0.250000)))),Swimming=(RigidBodyAnimNode=(SimSpaceSettings=(WorldAlpha=1.000000,VelocityScaleZ=0.400000,DampingAlpha=0.000000,MaxLinearVelocity=400.000000,MaxAngularVelocity=7.500000,MaxLinearAcceleration=3000.000000,MaxAngularAcceleration=30.000000,ExternalLinearDragV=(X=0.500000,Y=0.500000,Z=0.250000)))),Front_End=(RigidBodyAnimNode=(SimSpaceSettings=(WorldAlpha=1.000000,VelocityScaleZ=0.400000,DampingAlpha=0.000000,MaxLinearVelocity=400.000000,MaxAngularVelocity=7.500000,MaxLinearAcceleration=3000.000000,MaxAngularAcceleration=30.000000,ExternalLinearDragV=(X=0.500000,Y=0.500000,Z=0.250000)))),DBNO=(RigidBodyAnimNode=(SimSpaceSettings=(WorldAlpha=1.000000,VelocityScaleZ=0.400000,DampingAlpha=0.000000,MaxLinearVelocity=400.000000,MaxAngularVelocity=7.500000,MaxLinearAcceleration=3000.000000,MaxAngularAcceleration=30.000000,ExternalLinearDragV=(X=0.500000,Y=0.500000,Z=0.250000)))),DBNO_Carried=(RigidBodyAnimNode=(SimSpaceSettings=(WorldAlpha=1.000000,VelocityScaleZ=0.400000,DampingAlpha=0.000000,MaxLinearVelocity=400.000000,MaxAngularVelocity=7.500000,MaxLinearAcceleration=3000.000000,MaxAngularAcceleration=30.000000,ExternalLinearDragV=(X=0.500000,Y=0.500000,Z=0.250000)))),OnGround_Falling=(RigidBodyAnimNode=(SimSpaceSettings=(WorldAlpha=1.000000,VelocityScaleZ=0.400000,DampingAlpha=0.000000,MaxLinearVelocity=400.000000,MaxAngularVelocity=7.500000,MaxLinearAcceleration=3000.000000,MaxAngularAcceleration=30.000000,ExternalLinearDragV=(X=0.500000,Y=0.500000,Z=0.250000)))),Ziplining=(RigidBodyAnimNode=(SimSpaceSettings=(WorldAlpha=1.000000,VelocityScaleZ=0.400000,DampingAlpha=0.000000,MaxLinearVelocity=400.000000,MaxAngularVelocity=7.500000,MaxLinearAcceleration=3000.000000,MaxAngularAcceleration=30.000000,ExternalLinearDragV=(X=0.500000,Y=0.500000,Z=0.250000)))),Skydive_Surfing=(RigidBodyAnimNode=(SimSpaceSettings=(WorldAlpha=0.700000,VelocityScaleZ=0.400000,DampingAlpha=0.000000,MaxLinearVelocity=400.000000,MaxAngularVelocity=7.500000,MaxLinearAcceleration=3000.000000,MaxAngularAcceleration=30.000000,ExternalLinearDragV=(X=0.500000,Y=0.500000,Z=0.250000)))),Riding_Standing=(RigidBodyAnimNode=(SimSpaceSettings=(WorldAlpha=1.000000,VelocityScaleZ=0.400000,DampingAlpha=0.000000,MaxLinearVelocity=400.000000,MaxAngularVelocity=7.500000,MaxLinearAcceleration=3000.000000,MaxAngularAcceleration=30.000000,ExternalLinearDragV=(X=0.500000,Y=0.500000,Z=0.250000)))),Riding_Moving=(RigidBodyAnimNode=(SimSpaceSettings=(WorldAlpha=1.000000,VelocityScaleZ=0.400000,DampingAlpha=0.000000,MaxLinearVelocity=400.000000,MaxAngularVelocity=7.500000,MaxLinearAcceleration=3000.000000,MaxAngularAcceleration=30.000000,ExternalLinearDragV=(X=0.500000,Y=0.500000,Z=0.250000)))),Riding_Moving_Targeting=(RigidBodyAnimNode=(SimSpaceSettings=(WorldAlpha=1.000000,VelocityScaleZ=0.400000,DampingAlpha=0.000000,MaxLinearVelocity=400.000000,MaxAngularVelocity=7.500000,MaxLinearAcceleration=3000.000000,MaxAngularAcceleration=30.000000,ExternalLinearDragV=(X=0.500000,Y=0.500000,Z=0.250000)))),Tactical_Sprint=(RigidBodyAnimNode=(SimSpaceSettings=(WorldAlpha=1.000000,VelocityScaleZ=0.200000,DampingAlpha=0.000000,MaxLinearVelocity=500.000000,MaxAngularVelocity=5.000000,MaxLinearAcceleration=2500.000000,MaxAngularAcceleration=25.000000,ExternalLinearDragV=(X=0.500000,Y=0.500000,Z=0.500000)))),Rail_Grinding=(RigidBodyAnimNode=(SimSpaceSettings=(WorldAlpha=1.000000,VelocityScaleZ=0.400000,DampingAlpha=0.000000,MaxLinearVelocity=200.000000,MaxAngularVelocity=2.000000,MaxLinearAcceleration=1000.000000,MaxAngularAcceleration=10.000000,ExternalLinearDragV=(X=0.500000,Y=0.500000,Z=0.500000)))),)
 }
 RBAN Settings Out {type=FortRigidBodyAnimNodeParameters}
 resetClothLandingFromZilpine {type=boolean}
 Slope Rotation {type=Rotator}
 Wind Gust Local Settings {type=WindGustLocalSettingStruct, default {elapsedTime=0, finalGustStrengthLocal=0, gustDuration=0, gustStrength=0, timeBetweenGusts=0}}
 Wind Gusts Params {
  type=WindGustParamsStruct
  default {
   dBNO {
    maxDownInterpSpeed=5
    maxGustDuration=2
    maxGustStrength=7
    maxTimeBetweenGusts=7
    maxUpInterpSpeed=5
    minDownInterpSpeed=2
    minGustDuration=0.5
    minGustStrength=3
    minTimeBetweenGusts=0
    minUpInterpSpeed=2
    windTurbulenceScalarA=0.1
    windTurbulenceScalarB=1
    windAmplitude=<0 0 0>
    windFrequency=<0 0 0>
    windOffset=<0 0 0>
   }
   dBNO_Carried {
    maxDownInterpSpeed=5
    maxGustDuration=2
    maxGustStrength=7
    maxTimeBetweenGusts=7
    maxUpInterpSpeed=5
    minDownInterpSpeed=2
    minGustDuration=0.5
    minGustStrength=3
    minTimeBetweenGusts=0
    minUpInterpSpeed=2
    windTurbulenceScalarA=0.1
    windTurbulenceScalarB=1
    windAmplitude=<0 0 0>
    windFrequency=<0 0 0>
    windOffset=<0 0 0>
   }
   dBNO_Impulse {
    maxDownInterpSpeed=5
    maxGustDuration=2
    maxGustStrength=7
    maxTimeBetweenGusts=7
    maxUpInterpSpeed=5
    minDownInterpSpeed=2
    minGustDuration=0.5
    minGustStrength=3
    minTimeBetweenGusts=0
    minUpInterpSpeed=2
    windTurbulenceScalarA=0.1
    windTurbulenceScalarB=1
    windAmplitude=<0 0 0>
    windFrequency=<0 0 0>
    windOffset=<0 0 0>
   }
   dBNO_RollBwd {
    maxDownInterpSpeed=5
    maxGustDuration=2
    maxGustStrength=7
    maxTimeBetweenGusts=7
    maxUpInterpSpeed=5
    minDownInterpSpeed=2
    minGustDuration=0.5
    minGustStrength=3
    minTimeBetweenGusts=0
    minUpInterpSpeed=2
    windTurbulenceScalarA=0.1
    windTurbulenceScalarB=1
    windAmplitude=<0 0 0>
    windFrequency=<0 0 0>
    windOffset=<0 0 0>
   }
   dBNO_RollFwd {
    maxDownInterpSpeed=5
    maxGustDuration=2
    maxGustStrength=7
    maxTimeBetweenGusts=7
    maxUpInterpSpeed=5
    minDownInterpSpeed=2
    minGustDuration=0.5
    minGustStrength=3
    minTimeBetweenGusts=0
    minUpInterpSpeed=2
    windTurbulenceScalarA=0.1
    windTurbulenceScalarB=1
    windAmplitude=<0 0 0>
    windFrequency=<0 0 0>
    windOffset=<0 0 0>
   }
   dBNO_Tumbling {
    maxDownInterpSpeed=5
    maxGustDuration=2
    maxGustStrength=7
    maxTimeBetweenGusts=7
    maxUpInterpSpeed=5
    minDownInterpSpeed=2
    minGustDuration=0.5
    minGustStrength=3
    minTimeBetweenGusts=0
    minUpInterpSpeed=2
    windTurbulenceScalarA=0.1
    windTurbulenceScalarB=1
    windAmplitude=<0 0 0>
    windFrequency=<0 0 0>
    windOffset=<0 0 0>
   }
   emoteOrMelee {
    maxDownInterpSpeed=5
    maxGustDuration=2
    maxGustStrength=7
    maxTimeBetweenGusts=7
    maxUpInterpSpeed=5
    minDownInterpSpeed=2
    minGustDuration=0.5
    minGustStrength=3
    minTimeBetweenGusts=0
    minUpInterpSpeed=2
    windTurbulenceScalarA=0.1
    windTurbulenceScalarB=1
    windAmplitude=<0 0 0>
    windFrequency=<0 0 0>
    windOffset=<0 0 0>
   }
   front_End {
    maxDownInterpSpeed=5
    maxGustDuration=2
    maxGustStrength=4
    maxTimeBetweenGusts=5
    maxUpInterpSpeed=2
    minDownInterpSpeed=2
    minGustDuration=0.5
    minGustStrength=2
    minTimeBetweenGusts=0
    minUpInterpSpeed=2
    windTurbulenceScalarA=0.1
    windTurbulenceScalarB=0.75
    windAmplitude=<0.5 0.3 0.25>
    windFrequency=<5 3 7>
    windOffset=<0 0 -250>
   }
   onGround_Crouch_Moving {
    maxDownInterpSpeed=5
    maxGustDuration=2
    maxGustStrength=7
    maxTimeBetweenGusts=7
    maxUpInterpSpeed=5
    minDownInterpSpeed=2
    minGustDuration=0.5
    minGustStrength=3
    minTimeBetweenGusts=0
    minUpInterpSpeed=2
    windTurbulenceScalarA=0.1
    windTurbulenceScalarB=1
    windAmplitude=<0 0 0>
    windFrequency=<0 0 0>
    windOffset=<0 0 0>
   }
   onGround_Falling {
    maxDownInterpSpeed=5
    maxGustDuration=2
    maxGustStrength=7
    maxTimeBetweenGusts=7
    maxUpInterpSpeed=5
    minDownInterpSpeed=2
    minGustDuration=0.5
    minGustStrength=3
    minTimeBetweenGusts=0
    minUpInterpSpeed=2
    windTurbulenceScalarA=0.1
    windTurbulenceScalarB=1
    windAmplitude=<0 0 0>
    windFrequency=<0 0 0>
    windOffset=<0 0 0>
   }
   onGround_Moving {
    maxDownInterpSpeed=5
    maxGustDuration=2
    maxGustStrength=7
    maxTimeBetweenGusts=7
    maxUpInterpSpeed=5
    minDownInterpSpeed=2
    minGustDuration=0.5
    minGustStrength=3
    minTimeBetweenGusts=0
    minUpInterpSpeed=2
    windTurbulenceScalarA=0.1
    windTurbulenceScalarB=1
    windAmplitude=<0 0 0>
    windFrequency=<0 0 0>
    windOffset=<0 0 0>
   }
   onGround_Standing {
    maxDownInterpSpeed=5
    maxGustDuration=2
    maxGustStrength=7
    maxTimeBetweenGusts=7
    maxUpInterpSpeed=5
    minDownInterpSpeed=2
    minGustDuration=0.5
    minGustStrength=3
    minTimeBetweenGusts=0
    minUpInterpSpeed=2
    windTurbulenceScalarA=0.1
    windTurbulenceScalarB=1
    windAmplitude=<0 0 0>
    windFrequency=<0 0 0>
    windOffset=<0 0 0>
   }
   rail_Grinding {
    maxDownInterpSpeed=5
    maxGustDuration=8
    maxGustStrength=7
    maxTimeBetweenGusts=8
    maxUpInterpSpeed=5
    minDownInterpSpeed=2
    minGustDuration=0.2
    minGustStrength=3
    minTimeBetweenGusts=1
    minUpInterpSpeed=2
    windTurbulenceScalarA=0.1
    windTurbulenceScalarB=0.8
    windAmplitude=<20 65 42>
    windFrequency=<5 6 8.6>
    windOffset=<50 -500 -400>
   }
   riding_Moving {
    maxDownInterpSpeed=5
    maxGustDuration=2
    maxGustStrength=7
    maxTimeBetweenGusts=7
    maxUpInterpSpeed=5
    minDownInterpSpeed=2
    minGustDuration=0.5
    minGustStrength=3
    minTimeBetweenGusts=0
    minUpInterpSpeed=2
    windTurbulenceScalarA=0.1
    windTurbulenceScalarB=1
    windAmplitude=<0 0 0>
    windFrequency=<0 0 0>
    windOffset=<0 0 0>
   }
   riding_Moving_Targeting {
    maxDownInterpSpeed=5
    maxGustDuration=2
    maxGustStrength=7
    maxTimeBetweenGusts=7
    maxUpInterpSpeed=5
    minDownInterpSpeed=2
    minGustDuration=0.5
    minGustStrength=3
    minTimeBetweenGusts=0
    minUpInterpSpeed=2
    windTurbulenceScalarA=0.1
    windTurbulenceScalarB=1
    windAmplitude=<0 0 0>
    windFrequency=<0 0 0>
    windOffset=<0 0 0>
   }
   riding_Standing {
    maxDownInterpSpeed=5
    maxGustDuration=2
    maxGustStrength=7
    maxTimeBetweenGusts=7
    maxUpInterpSpeed=5
    minDownInterpSpeed=2
    minGustDuration=0.5
    minGustStrength=3
    minTimeBetweenGusts=0
    minUpInterpSpeed=2
    windTurbulenceScalarA=0.1
    windTurbulenceScalarB=1
    windAmplitude=<0 0 0>
    windFrequency=<0 0 0>
    windOffset=<0 0 0>
   }
   skydive_DivingDown {
    maxDownInterpSpeed=5
    maxGustDuration=2
    maxGustStrength=10
    maxTimeBetweenGusts=1
    maxUpInterpSpeed=2
    minDownInterpSpeed=2
    minGustDuration=0.5
    minGustStrength=2
    minTimeBetweenGusts=0
    minUpInterpSpeed=2
    windTurbulenceScalarA=0.1
    windTurbulenceScalarB=1
    windAmplitude=<0.5 1 0.75>
    windFrequency=<100 150 125>
    windOffset=<0 0 -300>
   }
   skydive_FloatingDown {
    maxDownInterpSpeed=5
    maxGustDuration=2
    maxGustStrength=7
    maxTimeBetweenGusts=7
    maxUpInterpSpeed=5
    minDownInterpSpeed=2
    minGustDuration=0.5
    minGustStrength=3
    minTimeBetweenGusts=0
    minUpInterpSpeed=2
    windTurbulenceScalarA=0.1
    windTurbulenceScalarB=1
    windAmplitude=<0.5 0.5 0.5>
    windFrequency=<50 50 50>
    windOffset=<0 -50 0>
   }
   skydive_GoingUp {
    maxDownInterpSpeed=5
    maxGustDuration=2
    maxGustStrength=7
    maxTimeBetweenGusts=7
    maxUpInterpSpeed=5
    minDownInterpSpeed=2
    minGustDuration=0.5
    minGustStrength=3
    minTimeBetweenGusts=0
    minUpInterpSpeed=2
    windTurbulenceScalarA=0.1
    windTurbulenceScalarB=1
    windAmplitude=<0 0 0>
    windFrequency=<0 0 0>
    windOffset=<0 0 0>
   }
   skydive_Parachuting {
    maxDownInterpSpeed=5
    maxGustDuration=2
    maxGustStrength=7
    maxTimeBetweenGusts=7
    maxUpInterpSpeed=5
    minDownInterpSpeed=2
    minGustDuration=0.5
    minGustStrength=3
    minTimeBetweenGusts=0
    minUpInterpSpeed=2
    windTurbulenceScalarA=0.1
    windTurbulenceScalarB=1
    windAmplitude=<0.5 0.5 0.5>
    windFrequency=<50 50 50>
    windOffset=<0 0 0>
   }
   skydive_Surfing {
    maxDownInterpSpeed=5
    maxGustDuration=2
    maxGustStrength=7
    maxTimeBetweenGusts=7
    maxUpInterpSpeed=5
    minDownInterpSpeed=2
    minGustDuration=0.5
    minGustStrength=3
    minTimeBetweenGusts=0
    minUpInterpSpeed=2
    windTurbulenceScalarA=0.1
    windTurbulenceScalarB=1
    windAmplitude=<0 0 0>
    windFrequency=<0 0 0>
    windOffset=<0 0 0>
   }
   swimming {
    maxDownInterpSpeed=5
    maxGustDuration=2
    maxGustStrength=7
    maxTimeBetweenGusts=7
    maxUpInterpSpeed=5
    minDownInterpSpeed=2
    minGustDuration=0.5
    minGustStrength=3
    minTimeBetweenGusts=0
    minUpInterpSpeed=2
    windTurbulenceScalarA=0.1
    windTurbulenceScalarB=1
    windAmplitude=<0 0 0>
    windFrequency=<0 0 0>
    windOffset=<0 0 0>
   }
   tactical_Sprint {
    maxDownInterpSpeed=5
    maxGustDuration=2
    maxGustStrength=7
    maxTimeBetweenGusts=7
    maxUpInterpSpeed=5
    minDownInterpSpeed=2
    minGustDuration=0.5
    minGustStrength=3
    minTimeBetweenGusts=0
    minUpInterpSpeed=2
    windTurbulenceScalarA=0.1
    windTurbulenceScalarB=1
    windAmplitude=<0 0 0>
    windFrequency=<0 0 0>
    windOffset=<0 0 0>
   }
   ziplining {
    maxDownInterpSpeed=5
    maxGustDuration=2
    maxGustStrength=7
    maxTimeBetweenGusts=7
    maxUpInterpSpeed=5
    minDownInterpSpeed=2
    minGustDuration=0.5
    minGustStrength=3
    minTimeBetweenGusts=0
    minUpInterpSpeed=2
    windTurbulenceScalarA=0.1
    windTurbulenceScalarB=1
    windAmplitude=<0 0 0>
    windFrequency=<0 0 0>
    windOffset=<0 0 0>
   }
  }
 }
]
