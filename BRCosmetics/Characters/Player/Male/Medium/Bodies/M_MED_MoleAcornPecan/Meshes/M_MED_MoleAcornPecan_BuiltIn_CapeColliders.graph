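# Physics asset for the M_MED_MoleAcornPecan medium male player body, supplying collision shapes that, by its name, serve the built-in cape simulation.
# Capsules cover the pelvis, spine, neck and clavicles, with a sphere on the head and a large box on dyn_cape_ground_collision acting as ground.
# Limb, hood drape and loincloth bodies are listed without shapes, and no constraints connect any bodies.
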PhysicsAsset /BRCosmetics/Characters/Player/Male/Medium/Bodies/M_MED_MoleAcornPecan/Meshes/M_MED_MoleAcornPecan_BuiltIn_CapeColliders
bodies {
 calf_l []
 calf_r []
 clavicle_l [{shapeType=Capsule, length=9.911, radius=8.9953, center=<9.4266 -1.1806 -2.4977>, rotation=<rot 69.3297 1.4476 176.9462>}]
 clavicle_r [{shapeType=Capsule, length=12.0955, radius=6.8108, center=<-10.3266 1.1579 0.1114>, rotation=<rot 78.1159 -0.8229 -5.2398>}]
 dyn_cape_ground_collision [{shapeType=Box, extentX=318.8156, extentY=318.8156, extentZ=283.5656, shapeName=dyn_skirt_ground_collision_box, center=<0 0 -140>}]
 dyn_hood_bk_drape_01 []
 dyn_hood_bk_drape_02 []
 dyn_hood_bk_drape_end []
 dyn_loincloth_bk_l_02 []
 dyn_loincloth_bk_r_02 []
 dyn_loincloth_side_bk_l_02 []
 dyn_loincloth_side_bk_r_02 []
 foot_l []
 foot_r []
 head [{shapeType=Sphere, radius=17.3125, center=<11.028 -0.174 0>}]
 lowerarm_l []
 lowerarm_r []
 neck_02 [{shapeType=Capsule, length=11.4137, radius=9.3413, center=<-0.2642 -0.3408 0>, rotation=<rot 90 -171.1193 -175.3978>}]
 pelvis [{shapeType=Capsule, length=6.2436, radius=16.9408, center=<-2.4887 0.1881 0>}]
 spine_01 [{shapeType=Capsule, length=5.1128, radius=16.3896, center=<-1.7965 0.1884 0>}]
 spine_02 [
  {shapeType=Capsule, length=4.4954, radius=16.4632, center=<-1.711 -0.1851 0.0001>}
  {shapeType=Capsule, length=29.3956, radius=11.5807, center=<2.5024 0.4992 3.9825>, rotation=<rot 90 0.0049 0.0049>}
  {shapeType=Capsule, length=29.875, radius=11.1013, center=<2.4704 0.4991 -3.9825>, rotation=<rot 90 -0.0396 -0.0396>}
 ]
 spine_03 [{shapeType=Capsule, length=5.89, radius=16.0238, center=<0.1445 0.2396 0.0001>}]
 spine_04 [{shapeType=Capsule, length=9.0936, radius=14.4793, center=<2.766 0.7765 0.0002>}]
 spine_05 [
  {shapeType=Capsule, length=8.8473, radius=14.8703, center=<-2.2795 1.6915 0>}
  {shapeType=Capsule, length=15.1119, radius=12.7824, center=<-0.2916 0.4452 0>}
 ]
 thigh_l []
 thigh_r []
 upperarm_l []
 upperarm_r []
}
constraints []
