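# Physics asset for the ears face accessory of the MoleAcornPecan medium male player body, defining bodies and joints for secondary ear motion.
# Each ear chains from a MSTR root through 01–08, with two spikes off 06.
# A 09–010 branch off 03 cross-links to 04 and 05, forming closed loops.
# Torso, arm, head and cape bodies have no constraints; left-ear limits at 04–05 are looser than right.
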PhysicsAsset /BRCosmetics/Characters/Player/Male/Medium/Bodies/M_MED_MoleAcornPecan/Meshes/Parts/M_MED_MoleAcornPecan_Ears_FaceAcc_Physics
bodies {
 clavicle_l [{
   shapeType=Capsule
   length=16.625
   radius=8.375
   shapeName=clavicle_l_capsule
   center=<5.8622 -0.4086 0>
   rotation=<rot -0.5519 -87.2511 101.3405>
  }]
 clavicle_r [{
   shapeType=Capsule
   length=16.625
   radius=8.375
   shapeName=clavicle_r_capsule
   center=<-5.8622 0.4086 0>
   rotation=<rot -0.5519 -87.2511 -78.6595>
  }]
 dyn_cape_l_02 [{
   shapeType=Capsule
   length=38.9797
   radius=27.9114
   shapeName=dyn_cape_mid_l_02_capsule
   center=<15.4394 0 -25.3595>
   rotation=<rot 85.38 -180 -180>
  }]
 dyn_cape_mid_l_02 [{
   shapeType=Capsule
   length=21.3543
   radius=45.5369
   shapeName=dyn_cape_mid_l_02_capsule
   center=<16.7225 4.265 -42.1807>
   rotation=<rot 85.38 -180 -180>
  }]
 dyn_cape_mid_r_02 [{
   shapeType=Capsule
   length=21.3543
   radius=45.5369
   shapeName=dyn_cape_mid_l_02_capsule
   center=<16.7473 -3.5021 -42.7439>
   rotation=<rot 85.38 -180 -180>
  }]
 dyn_cape_r_02 [{
   shapeType=Capsule
   length=38.9797
   radius=27.9114
   shapeName=dyn_cape_mid_l_02_capsule
   center=<15.4734 0 -25.6632>
   rotation=<rot 85.38 -180 -180>
  }]
 dyn_ear_l_01 [{shapeType=Capsule, length=2.6925, radius=3.1925, shapeName=dyn_ear_l_01_capsule, rotation=<rot 0 -90 0>}]
 dyn_ear_l_010 [{shapeType=Capsule, length=5.2417, radius=3.0458, shapeName=dyn_ear_l_010_capsule, center=<5 0 0>, rotation=<rot 90 0 0>}]
 dyn_ear_l_02 [{shapeType=Box, extentX=11.625, extentY=7.5, extentZ=9.75, shapeName=dyn_ear_l_02_box, center=<2.853 0 0>}]
 dyn_ear_l_03 [{shapeType=Box, extentX=10.875, extentY=7.5, extentZ=16.25, shapeName=dyn_ear_l_03_box, center=<4.3238 0 -1.448>}]
 dyn_ear_l_04 [{shapeType=Box, extentX=13.125, extentY=9.25, extentZ=18.125, shapeName=dyn_ear_l_04_box, center=<2.8901 0 -1.5109>}]
 dyn_ear_l_05 [{shapeType=Box, extentX=19.25, extentY=7.5, extentZ=21.875, shapeName=dyn_ear_l_05_box, center=<4.3268 0 -1.8409>}]
 dyn_ear_l_06 [{shapeType=Box, extentX=19.75, extentY=7.5, extentZ=18.875, shapeName=dyn_ear_l_06_box, center=<10.412 0 0>}]
 dyn_ear_l_06_spike_a [{shapeType=Capsule, length=7.3958, radius=2.1331, shapeName=dyn_ear_l_06_spike_a_capsule, center=<5 0 0>, rotation=<rot 90 0 0>}]
 dyn_ear_l_06_spike_b [{shapeType=Capsule, length=6.2589, radius=2.4459, shapeName=dyn_ear_l_06_spike_b_capsule, center=<5 0 0>, rotation=<rot 90 0 0>}]
 dyn_ear_l_07 [{shapeType=Capsule, length=5.2372, radius=4.3561, shapeName=dyn_ear_l_07_capsule, center=<5 0 0>, rotation=<rot 90 0 0>}]
 dyn_ear_l_08 [{shapeType=Capsule, length=4.4516, radius=1.5974, shapeName=dyn_ear_l_08_capsule, center=<5 0 0>, rotation=<rot 90 0 0>}]
 dyn_ear_l_09 [{shapeType=Capsule, length=7.2183, radius=3.5569, shapeName=dyn_ear_l_09_capsule, center=<5 0 0>, rotation=<rot 90 0 0>}]
 dyn_ear_l_MSTR [{shapeType=Capsule, length=0.505, radius=0.505, shapeName=dyn_ear_l_MSTR_capsule, rotation=<rot 0 -90 0>}]
 dyn_ear_r_01 [{shapeType=Capsule, length=2.6925, radius=3.1925, shapeName=dyn_ear_r_01_capsule, rotation=<rot 0 -90 180>}]
 dyn_ear_r_010 [{shapeType=Capsule, length=5.2417, radius=3.0458, shapeName=dyn_ear_r_010_capsule, center=<-5 0 0>, rotation=<rot 90 180 0>}]
 dyn_ear_r_02 [{shapeType=Box, extentX=11.625, extentY=7.5, extentZ=9.75, shapeName=dyn_ear_r_02_box, center=<-2.853 0 0>, rotation=<rot 0 0 -180>}]
 dyn_ear_r_03 [{shapeType=Box, extentX=10.875, extentY=7.5, extentZ=16.25, shapeName=dyn_ear_r_03_box, center=<-4.3238 0 1.448>, rotation=<rot 0 0 -180>}]
 dyn_ear_r_04 [{shapeType=Box, extentX=13.125, extentY=9.25, extentZ=18.125, shapeName=dyn_ear_r_04_box, center=<-2.8901 0 1.5109>, rotation=<rot 0 0 -180>}]
 dyn_ear_r_05 [{shapeType=Box, extentX=19.25, extentY=7.5, extentZ=21.875, shapeName=dyn_ear_r_05_box, center=<-4.3268 0 1.8409>, rotation=<rot 0 0 -180>}]
 dyn_ear_r_06 [{shapeType=Box, extentX=19.75, extentY=7.5, extentZ=18.875, shapeName=dyn_ear_r_06_box, center=<-10.412 0 0>, rotation=<rot 0 0 -180>}]
 dyn_ear_r_06_spike_a [{shapeType=Capsule, length=7.3958, radius=2.1331, shapeName=dyn_ear_r_06_spike_a_capsule, center=<-5 0 0>, rotation=<rot 90 180 0>}]
 dyn_ear_r_06_spike_b [{shapeType=Capsule, length=6.2589, radius=2.4459, shapeName=dyn_ear_r_06_spike_b_capsule, center=<-5 0 0>, rotation=<rot 90 180 0>}]
 dyn_ear_r_07 [{shapeType=Capsule, length=5.2372, radius=4.3561, shapeName=dyn_ear_r_07_capsule, center=<-5 0 0>, rotation=<rot 90 180 0>}]
 dyn_ear_r_08 [{shapeType=Capsule, length=4.4516, radius=1.5974, shapeName=dyn_ear_r_08_capsule, center=<-5 0 0>, rotation=<rot 90 180 0>}]
 dyn_ear_r_09 [{shapeType=Capsule, length=7.2183, radius=3.5569, shapeName=dyn_ear_r_09_capsule, center=<-5 0 0>, rotation=<rot 90 180 0>}]
 dyn_ear_r_MSTR [{shapeType=Capsule, length=0.505, radius=0.505, shapeName=dyn_ear_r_MSTR_capsule, rotation=<rot 0 -90 180>}]
 head [{shapeType=Capsule, length=1, radius=1, rotation=<rot -0.0515 -90 -90.2064>}]
 lowerarm_l [{shapeType=Capsule, length=15, radius=8, shapeName=lowerarm_l_capsule, center=<10.2985 0 0>, rotation=<rot 0 -90 90>}]
 lowerarm_r [{shapeType=Capsule, length=15, radius=8, shapeName=lowerarm_r_capsule, center=<-10.2985 0 0>, rotation=<rot 0 -90 -90>}]
 pelvis [{shapeType=Capsule, length=1.505, radius=19.255, shapeName=pelvis_capsule, center=<0 0.9924 0>, rotation=<rot 0 -90 0>}]
 spine_02 [{shapeType=Capsule, length=11.125, radius=13.875, shapeName=spine_02_capsule, center=<0 0.4287 0>, rotation=<rot 0 -90 0>}]
 spine_03 [{shapeType=Capsule, length=12.75, radius=12.25, shapeName=spine_03_capsule, rotation=<rot 0 -90 0>}]
 spine_04 [{shapeType=Capsule, length=15.3125, radius=9.6875, shapeName=spine_04_capsule, center=<0 -1.4352 0>, rotation=<rot 0 -90 0>}]
 spine_05 [{shapeType=Capsule, length=13.6875, radius=11.3125, shapeName=spine_05_capsule, center=<1.594 0.7468 0>, rotation=<rot 0 -90 180>}]
 upperarm_l [{shapeType=Capsule, length=20.625, radius=8, shapeName=upperarm_l_capsule, center=<13.1737 0 0>, rotation=<rot 0 -90 90>}]
 upperarm_r [{shapeType=Capsule, length=20.625, radius=8, shapeName=upperarm_r_capsule, center=<-13.1737 0 0>, rotation=<rot 0 -90 -90>}]
}
constraints [
 {bone1Name=dyn_ear_l_01, bone2Name=dyn_ear_l_MSTR, swing1LimitDegrees=5, swing2LimitDegrees=5, twistLimitDegrees=5}
 {bone1Name=dyn_ear_l_02, bone2Name=dyn_ear_l_01, swing1LimitDegrees=5, swing2LimitDegrees=5, twistLimitDegrees=5}
 {bone1Name=dyn_ear_l_03, bone2Name=dyn_ear_l_02, swing1LimitDegrees=5, swing2LimitDegrees=5, twistLimitDegrees=5}
 {bone1Name=dyn_ear_l_04, bone2Name=dyn_ear_l_03, swing1LimitDegrees=30, swing2LimitDegrees=15, twistLimitDegrees=10}
 {bone1Name=dyn_ear_l_05, bone2Name=dyn_ear_l_04, swing1LimitDegrees=30, swing2LimitDegrees=15, twistLimitDegrees=10}
 {bone1Name=dyn_ear_l_06, bone2Name=dyn_ear_l_05, swing1LimitDegrees=20, swing2LimitDegrees=10, twistLimitDegrees=10}
 {bone1Name=dyn_ear_l_07, bone2Name=dyn_ear_l_06, swing1LimitDegrees=20, swing2LimitDegrees=10, twistLimitDegrees=10}
 {bone1Name=dyn_ear_l_08, bone2Name=dyn_ear_l_07, swing1LimitDegrees=40, swing2LimitDegrees=20, twistLimitDegrees=20}
 {bone1Name=dyn_ear_l_06_spike_a, bone2Name=dyn_ear_l_06, swing1LimitDegrees=40, swing2LimitDegrees=20, twistLimitDegrees=20}
 {bone1Name=dyn_ear_l_06_spike_b, bone2Name=dyn_ear_l_06, swing1LimitDegrees=40, swing2LimitDegrees=20, twistLimitDegrees=20}
 {bone1Name=dyn_ear_l_09, bone2Name=dyn_ear_l_03, swing1LimitDegrees=10, swing2LimitDegrees=5, twistLimitDegrees=5}
 {bone1Name=dyn_ear_l_010, bone2Name=dyn_ear_l_09, swing1LimitDegrees=10, swing2LimitDegrees=5, twistLimitDegrees=5}
 {bone1Name=dyn_ear_l_09, bone2Name=dyn_ear_l_04, swing1LimitDegrees=5, swing2LimitDegrees=5, twistLimitDegrees=5}
 {bone1Name=dyn_ear_l_010, bone2Name=dyn_ear_l_05, swing1LimitDegrees=5, swing2LimitDegrees=5, twistLimitDegrees=5}
 {bone1Name=dyn_ear_r_01, bone2Name=dyn_ear_r_MSTR, swing1LimitDegrees=5, swing2LimitDegrees=5, twistLimitDegrees=5}
 {bone1Name=dyn_ear_r_02, bone2Name=dyn_ear_r_01, swing1LimitDegrees=5, swing2LimitDegrees=5, twistLimitDegrees=5}
 {bone1Name=dyn_ear_r_03, bone2Name=dyn_ear_r_02, swing1LimitDegrees=5, swing2LimitDegrees=5, twistLimitDegrees=5}
 {bone1Name=dyn_ear_r_04, bone2Name=dyn_ear_r_03, swing1LimitDegrees=20, swing2LimitDegrees=10, twistLimitDegrees=10}
 {bone1Name=dyn_ear_r_05, bone2Name=dyn_ear_r_04, swing1LimitDegrees=20, swing2LimitDegrees=10, twistLimitDegrees=10}
 {bone1Name=dyn_ear_r_06, bone2Name=dyn_ear_r_05, swing1LimitDegrees=20, swing2LimitDegrees=10, twistLimitDegrees=10}
 {bone1Name=dyn_ear_r_07, bone2Name=dyn_ear_r_06, swing1LimitDegrees=20, swing2LimitDegrees=10, twistLimitDegrees=10}
 {bone1Name=dyn_ear_r_08, bone2Name=dyn_ear_r_07, swing1LimitDegrees=40, swing2LimitDegrees=20, twistLimitDegrees=20}
 {bone1Name=dyn_ear_r_06_spike_a, bone2Name=dyn_ear_r_06, swing1LimitDegrees=40, swing2LimitDegrees=20, twistLimitDegrees=20}
 {bone1Name=dyn_ear_r_06_spike_b, bone2Name=dyn_ear_r_06, swing1LimitDegrees=40, swing2LimitDegrees=20, twistLimitDegrees=20}
 {bone1Name=dyn_ear_r_09, bone2Name=dyn_ear_r_03, swing1LimitDegrees=10, swing2LimitDegrees=5, twistLimitDegrees=5}
 {bone1Name=dyn_ear_r_010, bone2Name=dyn_ear_r_09, swing1LimitDegrees=10, swing2LimitDegrees=5, twistLimitDegrees=5}
 {bone1Name=dyn_ear_r_09, bone2Name=dyn_ear_r_04, swing1LimitDegrees=5, swing2LimitDegrees=5, twistLimitDegrees=5}
 {bone1Name=dyn_ear_r_010, bone2Name=dyn_ear_r_05, swing1LimitDegrees=5, swing2LimitDegrees=5, twistLimitDegrees=5}
]
