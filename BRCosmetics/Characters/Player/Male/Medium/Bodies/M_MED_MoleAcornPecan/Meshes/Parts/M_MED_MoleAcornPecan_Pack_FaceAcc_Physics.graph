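# Physics asset for the MoleAcornPecan medium male face accessory pack, rigging hood and back-drape bones for physics with capsule bodies.
# Five short hood chains and a longer back drape each hang from a tiny MSTR root body through limited constraints.
# The drape bends mainly on one swing axis, with the other swing and twist nearly locked.
# Spine, clavicle, head and mid-cape capsules have no constraints here.
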PhysicsAsset /BRCosmetics/Characters/Player/Male/Medium/Bodies/M_MED_MoleAcornPecan/Meshes/Parts/M_MED_MoleAcornPecan_Pack_FaceAcc_Physics
bodies {
 clavicle_l [{shapeType=Capsule, length=8.939, radius=9.1853, shapeName=clavicle_l_capsule, center=<7.6816 0 0>, rotation=<rot 80.0985 -0.001 -0.001>}]
 clavicle_r [{shapeType=Capsule, length=8.637, radius=9.1853, shapeName=clavicle_r_capsule, center=<-7.535 0 0>, rotation=<rot 79.6256 -0.0003 -0.0003>}]
 dyn_cape_mid_l_01 [{
   shapeType=Capsule
   length=9.6676
   radius=20.0035
   shapeName=dyn_cape_mid_l_01_capsule
   center=<4.7641 0 -18.6507>
   rotation=<rot 84.7409 0 0>
  }]
 dyn_cape_mid_l_02 [{shapeType=Capsule, length=9.6676, radius=20.0035, shapeName=dyn_cape_mid_l_02_capsule, center=<4.3876 0 -19.961>, rotation=<rot 82.93 0 0>}]
 dyn_cape_mid_l_MSTR [{shapeType=Capsule, length=0.505, radius=0.505, shapeName=dyn_cape_mid_l_MSTR_capsule, rotation=<rot 0 -90 0>}]
 dyn_cape_mid_r_01 [{
   shapeType=Capsule
   length=9.6676
   radius=20.0035
   shapeName=dyn_cape_mid_r_01_capsule
   center=<4.7641 0 -18.6507>
   rotation=<rot 84.7408 0 0>
  }]
 dyn_cape_mid_r_02 [{shapeType=Capsule, length=10.0254, radius=20.0035, shapeName=dyn_cape_mid_r_02_capsule, center=<4.3265 0 -20.4539>, rotation=<rot 82.93 0 0>}]
 dyn_cape_mid_r_MSTR [{shapeType=Capsule, length=0.505, radius=0.505, shapeName=dyn_cape_mid_r_MSTR_capsule, rotation=<rot 0 -90 0>}]
 dyn_hood_bk [{shapeType=Capsule, length=6.3255, radius=2.6745, shapeName=dyn_hood_l_02_capsule, center=<4.4934 0 4.2>, rotation=<rot 0 0 -90>}]
 dyn_hood_bk_MSTR [{shapeType=Capsule, length=0.505, radius=0.505, shapeName=dyn_hood_bk_MSTR_capsule}]
 dyn_hood_bk_drape_01 [{shapeType=Capsule, length=26.2698, radius=7.3695, shapeName=dyn_hood_bk_drape_01_capsule, center=<4.0439 0 8.4>, rotation=<rot 0 0 90>}]
 dyn_hood_bk_drape_02 [{shapeType=Capsule, length=28.7463, radius=6.7288, shapeName=dyn_hood_bk_drape_02_capsule, center=<4.0439 0 5.2>, rotation=<rot 0 0 90>}]
 dyn_hood_bk_drape_MSTR [{shapeType=Capsule, length=0.505, radius=0.505, shapeName=dyn_hood_bk_drape_MSTR_capsule}]
 dyn_hood_bk_drape_end [{shapeType=Capsule, length=25.5127, radius=6.2208, shapeName=dyn_hood_bk_drape_end_capsule, center=<4.5421 0 3.3238>, rotation=<rot 0 0 90>}]
 dyn_hood_bk_end [{shapeType=Capsule, length=11.047, radius=3.328, shapeName=dyn_hood_l_02_end_capsule, center=<2.2327 0 5.5>, rotation=<rot 0 0 -90>}]
 dyn_hood_l_01 [{shapeType=Capsule, length=6.1454, radius=2.8546, shapeName=dyn_hood_l_01_capsule, center=<3.7131 0 2.7>, rotation=<rot 0 0 -90>}]
 dyn_hood_l_01_MSTR [{shapeType=Capsule, length=0.505, radius=0.505, shapeName=dyn_hood_l_01_MSTR_capsule}]
 dyn_hood_l_01_end [{
   shapeType=Capsule
   length=10.6416
   radius=3.1894
   shapeName=dyn_hood_l_01_end_capsule
   center=<0.1763 -0.2675 1.5>
   rotation=<rot 0.7143 -4.0359 -77.6012>
  }]
 dyn_hood_l_02 [{
   shapeType=Capsule
   length=6.2031
   radius=2.7969
   shapeName=dyn_hood_l_02_capsule
   center=<3.9049 -0.4694 3.4>
   rotation=<rot 0.0526 12.05 -107.3863>
  }]
 dyn_hood_l_02_MSTR [{shapeType=Capsule, length=0.505, radius=0.505, shapeName=dyn_hood_l_02_MSTR_capsule}]
 dyn_hood_l_02_end [{
   shapeType=Capsule
   length=11.0442
   radius=3.3308
   shapeName=dyn_hood_l_02_end_capsule
   center=<1.7539 -0.4687 4>
   rotation=<rot -0.0001 12.2143 -107.0431>
  }]
 dyn_hood_r_01 [{shapeType=Capsule, length=6.1571, radius=2.8429, shapeName=dyn_hood_r_02_capsule, center=<-3.7073 0.0583 -2.3>, rotation=<rot 0 0 97.3032>}]
 dyn_hood_r_01_MSTR [{shapeType=Capsule, length=0.505, radius=0.505, shapeName=dyn_hood_r_01_MSTR_capsule}]
 dyn_hood_r_01_end [{
   shapeType=Capsule
   length=11.2943
   radius=3.0807
   shapeName=dyn_hood_r_02_end_capsule
   center=<-0.7632 0.0904 -1.7>
   rotation=<rot 0 -6.2427 98.7207>
  }]
 dyn_hood_r_02 [{
   shapeType=Capsule
   length=6.346
   radius=2.654
   shapeName=dyn_hood_r_02_capsule
   center=<-4.4508 0.0841 -3.1>
   rotation=<rot 1.0071 12.813 83.3515>
  }]
 dyn_hood_r_02_MSTR [{shapeType=Capsule, length=0.505, radius=0.505, shapeName=dyn_hood_r_02_MSTR_capsule}]
 dyn_hood_r_02_end [{
   shapeType=Capsule
   length=11.3528
   radius=3.0222
   shapeName=dyn_hood_r_02_end_capsule
   center=<-2.3752 0.4286 -3.8>
   rotation=<rot 0 13.5724 86.0378>
  }]
 head [{shapeType=Capsule, length=14.7187, radius=9.0764, center=<4.7399 -3.0365 0>, rotation=<rot 90 0 0>}]
 spine_02 [{shapeType=Capsule, length=3.5595, radius=15.0468, shapeName=spine_05_capsule}]
 spine_03 [{shapeType=Capsule, length=3.5595, radius=15.0468, shapeName=spine_05_capsule}]
 spine_04 [{shapeType=Capsule, length=3.5595, radius=15.0468, shapeName=spine_05_capsule}]
 spine_05 [{shapeType=Capsule, length=4.5826, radius=14.4555, shapeName=spine_05_capsule, center=<2.7018 0 0>}]
}
constraints [
 {bone1Name=dyn_hood_bk_drape_01, bone2Name=dyn_hood_bk_drape_MSTR, swing1LimitDegrees=30, swing2LimitDegrees=1, twistLimitDegrees=2}
 {bone1Name=dyn_hood_bk_drape_02, bone2Name=dyn_hood_bk_drape_01, swing1LimitDegrees=30, swing2LimitDegrees=1, twistLimitDegrees=2}
 {bone1Name=dyn_hood_bk_drape_end, bone2Name=dyn_hood_bk_drape_02, swing1LimitDegrees=30, swing2LimitDegrees=1, twistLimitDegrees=2}
 {bone1Name=dyn_hood_bk, bone2Name=dyn_hood_bk_MSTR, swing1LimitDegrees=20, swing2LimitDegrees=20, twistLimitDegrees=4}
 {bone1Name=dyn_hood_bk_end, bone2Name=dyn_hood_bk, swing1LimitDegrees=20, swing2LimitDegrees=40, twistLimitDegrees=4}
 {bone1Name=dyn_hood_l_01, bone2Name=dyn_hood_l_01_MSTR, swing1LimitDegrees=20, swing2LimitDegrees=20, twistLimitDegrees=4}
 {bone1Name=dyn_hood_l_01_end, bone2Name=dyn_hood_l_01, swing1LimitDegrees=20, swing2LimitDegrees=30, twistLimitDegrees=4}
 {bone1Name=dyn_hood_l_02, bone2Name=dyn_hood_l_02_MSTR, swing1LimitDegrees=20, swing2LimitDegrees=20, twistLimitDegrees=4}
 {bone1Name=dyn_hood_l_02_end, bone2Name=dyn_hood_l_02, swing1LimitDegrees=20, swing2LimitDegrees=30, twistLimitDegrees=4}
 {bone1Name=dyn_hood_r_01, bone2Name=dyn_hood_r_01_MSTR, swing1LimitDegrees=20, swing2LimitDegrees=20, twistLimitDegrees=4}
 {bone1Name=dyn_hood_r_01_end, bone2Name=dyn_hood_r_01, swing1LimitDegrees=20, swing2LimitDegrees=30, twistLimitDegrees=4}
 {bone1Name=dyn_hood_r_02, bone2Name=dyn_hood_r_02_MSTR, swing1LimitDegrees=20, swing2LimitDegrees=20, twistLimitDegrees=4}
 {bone1Name=dyn_hood_r_02_end, bone2Name=dyn_hood_r_02, swing1LimitDegrees=20, swing2LimitDegrees=30, twistLimitDegrees=4}
]
