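# Animation Blueprint for the MoleAcornPecan player's face accessory, animating its hood as secondary motion.
# The attached parent's pose runs through a Control Rig and hood-lip AnimDynamics into a cached pose.
# That cache feeds two RigidBody simulations, with a pack-specific physics asset chosen when customization loading detects back bling.
# Without back bling, disabled rigid bodies fall back to a bone constraint tying the back drape to cape bones.
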AnimBlueprint /BRCosmetics/Characters/Player/Male/Medium/Bodies/M_MED_MoleAcornPecan/Meshes/Parts/M_MED_MoleAcornPecan_FaceAcc_AnimBP
target_skeleton=&/M_MED_MoleAcornPecan_FaceAcc_Skeleton
graphs [
 AnimGraph {nodes [
   Layered blend per bone(BlendPoses_0=Copy Pose From Mesh(SourceMeshComponent=Get CharacterMesh() -> (CharacterMesh:Skeletal Mesh Component Object Reference)) {bUseAttachedParent=true})
   n1 = "Constraint - dyn_hood_bk_drape_01 (RT)"!(ComponentPose=Local To Component(LocalPose=Use cached pose 'Initial Cache'()), Alpha=1, ConstraintWeights_0=1, ConstraintWeights_1:float=1) -> (Pose) {
    alphaBoolBlend {blendOption=Linear}
    boneToModify {boneName=dyn_hood_bk_drape_01}
    constraintSetup [{transformType=Rotation, targetBone {boneName=dyn_cape_mid_l_01}}, {transformType=Translation, targetBone {boneName=dyn_cape_mid_r_01}}]
    constraintWeights [1,1]
   }
   n2 = Component To Local(ComponentPose=@n1)
   n3 = Use cached pose 'Hood Normal'()
   n4 = Get bRigidBodyEnabled()
   Output Pose(Result=Blend Poses by bool(
     BlendPose_0=Use cached pose 'Hood With Pack'()
     BlendPose_1=Blend Poses by bool(BlendPose_0=@n3, BlendPose_1=@n2, bActiveValue=@n4) {blendType=Linear}
     bActiveValue=Get isWearingPack() -> (isWearingPack:bool)
    ) {blendType=Linear}) {layerGroup=None}
   n5 = Control Rig(Source=Copy Pose From Mesh() {bUseAttachedParent=true}) {
    alphaBoolBlend {blendOption=Linear}
    controlRigAssetReference {blueprintRigClass=&/M_MED_MoleAcornPecan_FaceAcc_CtrlRig.M_MED_MoleAcornPecan_FaceAcc_CtrlRig_C}
   }
   n6 = Anim Dynamics - Bone: dyn_hood_lip(ComponentPose=Local To Component(LocalPose=@n5), Alpha=1, ExternalForce="0,0,0") {
    angularDampingOverride=0.999
    angularSpringConstant=2000
    bAngularSpring=true
    bOverrideAngularDamping=true
    simulationSpace=RootRelative
    alphaBoolBlend {blendOption=Linear}
    boundBone {boneName=dyn_hood_lip}
    componentLinearAccScale=<0.3 0.3 0.3>
    componentLinearVelScale=<0.3 0.3 0.3>
    physicsBodyDefinitions [{
      boundBone {boneName=dyn_hood_lip}
      boxExtents=<14 14 14>
      constraintSetup {angularLimitsMax=<0 10 4>, angularLimitsMin=<0 -30 -4>, angularTarget=<1 0 0>}
      localJointOffset=<0 0 0>
     }]
   }
   Save cached pose 'Initial Cache'(Pose=Component To Local(ComponentPose=@n6)) {cacheName=Initial Cache}
   n7 = RigidBody(ComponentPose=Local To Component(LocalPose=Use cached pose 'Initial Cache'())) {
    bClampLinearTranslationLimitToRefPose=true
    bOverrideWorldGravity=true
    overridePhysicsAsset=&/M_MED_MoleAcornPecan_FaceAcc_Physics
    simulationSpace=BaseBoneSpace
    alphaBoolBlend {blendOption=Linear}
    baseBoneRef {boneName=spine_05}
    overrideWorldGravity {z=-980}
    simSpaceSettings {
     dampingAlpha=0
     maxAngularAcceleration=30
     maxAngularVelocity=6
     maxLinearAcceleration=4000
     maxLinearVelocity=800
     velocityScaleZ=0.6
     worldAlpha=1
     externalLinearDragV=<0.7 0.7 0.7>
    }
   }
   Save cached pose 'Hood Normal'(Pose=Component To Local(ComponentPose=@n7)) {cacheName=Hood Normal}
   n8 = RigidBody(ComponentPose=Local To Component(LocalPose=Use cached pose 'Initial Cache'())) {
    bClampLinearTranslationLimitToRefPose=true
    bOverrideWorldGravity=true
    overridePhysicsAsset=&/M_MED_MoleAcornPecan_Pack_FaceAcc_Physics
    simulationSpace=BaseBoneSpace
    alphaBoolBlend {blendOption=Linear}
    baseBoneRef {boneName=spine_05}
    overrideWorldGravity {z=-980}
    simSpaceSettings {
     dampingAlpha=0
     maxAngularAcceleration=30
     maxAngularVelocity=6
     maxLinearAcceleration=4000
     maxLinearVelocity=800
     velocityScaleZ=0.6
     worldAlpha=1
     externalLinearDragV=<0.7 0.7 0.7>
    }
   }
   Save cached pose 'Hood With Pack'(Pose=Component To Local(ComponentPose=@n8)) {cacheName=Hood With Pack}
  ]}
 EventGraph {nodes [
   n1 = Event OnCharacterCustomizationLoaded()
   n2 = CheckForBackbling()
   > Set isWearingPack(isWearingPack:bool=OR Boolean!(A=@n2, B=@n2.bIsBackblingACape, C:bool=@n2.bIsBackblingLong) -> (ReturnValue)) -> (Output_Get:bool)
   > Set isWearingCape(isWearingCape=@n2.bIsBackblingLong)
   > Set CharacterMesh(CharacterMesh:Skeletal Mesh Component Object Reference=@n1.CustomBaseMesh) -> (Output_Get:Skeletal Mesh Component Object Reference)
   Event BlueprintUpdateAnimation()
   Set Gravity Override(Gravity Override=Gravity Override State Function DEPRECATED(CharacterPartAnimInstance=Self-Reference(), GameTimeInSeconds=GetGameTimeInSeconds(), GravityOverrideParams=Get Gravity Override Params()))
  ]}
]
variables [
 CharacterMesh {type=SkeletalMeshComponent}
 Gravity Override {type=Vector}
 Gravity Override Params {type=GravityOverrideParamsStruct}
 isWearingCape {type=boolean}
 isWearingPack {type=boolean}
]
