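# Animation Blueprint for the head part of the MoleAcornPecan male character body, targeting the MoleAcornGloam head skeleton.
# When character customization loads, it gets a dynamic material instance for element 1 of the part's skeletal mesh and sets "Hide Element 01" to 1.
# The animation update event only calls the parent implementation, and no animation logic of its own is shown.
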AnimBlueprint /BRCosmetics/Characters/Player/Male/Medium/Bodies/M_MED_MoleAcornPecan/Meshes/Parts/M_MED_MoleAcornPecan_Head_AnimBP
target_skeleton=/BRCosmetics/Characters/Player/Male/Medium/Heads/M_MED_MoleAcornGloam_Head/Meshes/Fortnite_Base_Head_MoleAcornGloam_Skeleton
graphs [EventGraph {nodes [
   n1 = Event BlueprintUpdateAnimation()
   > Parent: BlueprintUpdateAnimation!(execute, DeltaTimeX:float=@n1.DeltaTimeX) -> (then) {is=title K2Node_CallParentFunction}
   Event OnCharacterCustomizationLoaded()
   > n2 = Get Dynamic Material Instance(SkeletalMeshComponent=GetPartSkeletalMeshComponent(), ElementIndex=1)
   > SetScalarParameterValue(self=@n2, ParameterName=Hide Element 01, Value=1)
   TryGetPawnOwner()
  ]}]
variables []
